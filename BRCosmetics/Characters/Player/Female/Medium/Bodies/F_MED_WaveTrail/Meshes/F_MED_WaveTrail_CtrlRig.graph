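# Control Rig for the F_MED_WaveTrail female player body that corrects limb twists and drives secondary deformation of its armour, belts and shoes.
# An optional debug pass copies controls to bones and solves twist bones; the main chain then runs twist fixes, a shared deform rig and per-part functions.
# Parts use weighted parent constraints plus offsets driven by a DeltaRotation helper and spherical pose readers, with arm and leg armour extras gated by emote-state booleans.
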ControlRigBlueprint /BRCosmetics/Characters/Player/Female/Medium/Bodies/F_MED_WaveTrail/Meshes/F_MED_WaveTrail_CtrlRig
graphs [
 New Function_ContainedGraph {nodes [
   n1 = RigVMFunctionEntryNode()
   RigVMFunction_ControlFlowBranch = &RigVMFunction_ControlFlowBranch(ExecuteContext=@n1.ExecuteContext, Condition=@n1.debug)
   RigVMFunctionReturnNode(ExecuteContext=@RigVMFunction_ControlFlowBranch.Completed)
   RigVMFunction_Sequence = RigVMAggregateNode(ExecuteContext=@RigVMFunction_ControlFlowBranch.True)
   RigVMFunction_MathQuaternionSwingTwist = &RigVMFunction_MathQuaternionSwingTwist(
    TwistAxis=(X=1,Y=0,Z=0)
    Input=&RigUnit_GetRelativeTransformForItem(Child=(Type=Bone,Name="thigh_l"), Parent=(Type=Null,Name="thigh_l_null"), bChildInitial=False, bParentInitial=False).RelativeTransform.Rotation
   )
   Inverse = /Script/RigVM.RigVMFunction_MathQuaternionInverse(Value=@RigVMFunction_MathQuaternionSwingTwist.Twist) {is=struct RigVMUnitNode, template="Inverse::Execute(in Value,out Result)"}
   &RigUnit_ModifyTransforms(
    ItemToModify=((Item=(Type=Bone,Name="thigh_twist_01_l"),Transform=(Rotation=(X=0,Y=0,Z=0,W=1),Translation=(X=0,Y=0,Z=0),Scale3D=(X=1,Y=1,Z=1))))
    Weight=1
    WeightMinimum=0
    ItemToModify.0.Transform.Rotation=@Inverse.Result
    ExecutePin=@RigVMFunction_Sequence.B
   )
   n2 = &RigUnit_GetRelativeTransformForItem(Child=(Type=Bone,Name="thigh_r"), Parent=(Type=Null,Name="thigh_r_null"), bChildInitial=False, bParentInitial=False)
   > &RigUnit_ModifyTransforms(
    ItemToModify=((Item=(Type=Bone,Name="thigh_twist_01_r"),Transform=(Rotation=(X=0,Y=0,Z=0,W=1),Translation=(X=0,Y=0,Z=0),Scale3D=(X=1,Y=1,Z=1))))
    Weight=1
    WeightMinimum=0
    ItemToModify.0.Transform.Rotation=&RigVMFunction_MathQuaternionInverse(Value=&RigVMFunction_MathQuaternionSwingTwist(TwistAxis=(X=1,Y=0,Z=0), Input=@n2.RelativeTransform.Rotation).Twist) {template="Inverse::Execute(in Value,out Result)"}.Result
   )
   n3 = &RigVMFunction_MathQuaternionSwingTwist(
    TwistAxis=(X=1,Y=0,Z=0)
    Input=&RigUnit_GetRelativeTransformForItem(Child=(Type=Bone,Name="foot_l"), Parent=(Type=Null,Name="foot_l_null"), bChildInitial=False, bParentInitial=False).RelativeTransform.Rotation
   )
   &RigUnit_ModifyTransforms(
    ItemToModify="((Item=(Type=Bone,Name=\"calf_twist_01_l\"),Transform=(Rotation=(X=0,Y=0,Z=0,W=1),Translation=(X=0,Y=0,Z=0),Scale3D=(X=1,Y=1,Z=1))),(Item=(Type=Bone,Name=\"calf_twist_02_l\"),Transform=(Rotation=(X=0,Y=0,Z=0,W=1),Translation=(X=0,Y=0,Z=0),Scale3D=(X=1,Y=1,Z=1) ...(truncated)"
    Weight=1
    WeightMinimum=0
    ItemToModify.1.Transform.Rotation=/Script/RigVM.RigVMFunction_MathQuaternionSlerp(B=(X=0,Y=0,Z=0,W=1), T=0.5, A=@n3.Twist) {is=struct RigVMUnitNode, template="Interpolate::Execute(in A,in B,in T,out Result)"}.Result
    ItemToModify.0.Transform.Rotation=@n3.Twist
    ExecutePin=@RigVMFunction_Sequence.D
   )
   n4 = &RigVMFunction_MathQuaternionSwingTwist(
    TwistAxis=(X=1,Y=0,Z=0)
    Input=&RigUnit_GetRelativeTransformForItem(Child=(Type=Bone,Name="foot_r"), Parent=(Type=Null,Name="foot_r_null"), bChildInitial=False, bParentInitial=False).RelativeTransform.Rotation
   )
   > &RigUnit_ModifyTransforms(
    ItemToModify="((Item=(Type=Bone,Name=\"calf_twist_01_r\"),Transform=(Rotation=(X=0,Y=0,Z=0,W=1),Translation=(X=0,Y=0,Z=0),Scale3D=(X=1,Y=1,Z=1))),(Item=(Type=Bone,Name=\"calf_twist_02_r\"),Transform=(Rotation=(X=0,Y=0,Z=0,W=1),Translation=(X=0,Y=0,Z=0),Scale3D=(X=1,Y=1,Z=1) ...(truncated)"
    Weight=1
    WeightMinimum=0
    ItemToModify.1.Transform.Rotation=&RigVMFunction_MathQuaternionSlerp(B=(X=0,Y=0,Z=0,W=1), T=0.5, A=@n4.Twist) {template="Interpolate::Execute(in A,in B,in T,out Result)"}.Result
    ItemToModify.0.Transform.Rotation=@n4.Twist
   )
   n5 = &RigVMFunction_MathQuaternionInverse(Value=&RigVMFunction_MathQuaternionSwingTwist(
     TwistAxis=(X=1,Y=0,Z=0)
     Input=&RigUnit_GetRelativeTransformForItem(Child=(Type=Bone,Name="upperarm_l"), Parent=(Type=Null,Name="upperarm_l_null"), bChildInitial=False, bParentInitial=False).RelativeTransform.Rotation
    ).Twist) {template="Inverse::Execute(in Value,out Result)"}
   &RigUnit_ModifyTransforms(
    ItemToModify="((Item=(Type=Bone,Name=\"upperarm_twist_01_l\"),Transform=(Rotation=(X=0,Y=0,Z=0,W=1),Translation=(X=0,Y=0,Z=0),Scale3D=(X=1,Y=1,Z=1))),(Item=(Type=Bone,Name=\"upperarm_twist_02_l\"),Transform=(Rotation=(X=0,Y=0,Z=0,W=1),Translation=(X=0,Y=0,Z=0),Scale3D=(X=1, ...(truncated)"
    Weight=1
    WeightMinimum=0
    ItemToModify.1.Transform.Rotation=&RigVMFunction_MathQuaternionSlerp(B=(X=0,Y=0,Z=0,W=1), T=0.5, A=@n5.Result) {template="Interpolate::Execute(in A,in B,in T,out Result)"}.Result
    ItemToModify.0.Transform.Rotation=@n5.Result
    ExecutePin=@RigVMFunction_Sequence.C
   )
   n6 = &RigVMFunction_MathQuaternionInverse(Value=&RigVMFunction_MathQuaternionSwingTwist(
     TwistAxis=(X=1,Y=0,Z=0)
     Input=&RigUnit_GetRelativeTransformForItem(Child=(Type=Bone,Name="upperarm_r"), Parent=(Type=Null,Name="upperarm_r_null"), bChildInitial=False, bParentInitial=False).RelativeTransform.Rotation
    ).Twist) {template="Inverse::Execute(in Value,out Result)"}
   Interpolate = &RigVMFunction_MathQuaternionSlerp(B=(X=0,Y=0,Z=0,W=1), T=0.5, A=@n6.Result) {template="Interpolate::Execute(in A,in B,in T,out Result)"}
   > &RigUnit_ModifyTransforms(
    ItemToModify="((Item=(Type=Bone,Name=\"upperarm_twist_01_r\"),Transform=(Rotation=(X=0,Y=0,Z=0,W=1),Translation=(X=0,Y=0,Z=0),Scale3D=(X=1,Y=1,Z=1))),(Item=(Type=Bone,Name=\"upperarm_twist_02_r\"),Transform=(Rotation=(X=0,Y=0,Z=0,W=1),Translation=(X=0,Y=0,Z=0),Scale3D=(X=1, ...(truncated)"
    Weight=1
    WeightMinimum=0
    ItemToModify.1.Transform.Rotation=@Interpolate.Result
    ItemToModify.0.Transform.Rotation=@n6.Result
   )
   RigUnit_ItemArray = &RigUnit_ItemArray(Items="((Type=Bone,Name=\"pelvis\"),(Type=Bone,Name=\"spine_01\"),(Type=Bone,Name=\"spine_02\"),(Type=Bone,Name=\"spine_03\"),(Type=Bone,Name=\"spine_04\"),(Type=Bone,Name=\"spine_05\"),(Type=Bone,Name=\"clavicle_l\"),(Type=Bone,Name=\"upperarm_l\"),(Type=Bone,Name=\"lowerarm_l\") ...(truncated)")
   DISPATCH_RigVMDispatch_ArrayIterator = RigVMDispatchNode(Array=@RigUnit_ItemArray.Items, ExecuteContext=@RigVMFunction_Sequence.A) {template="DISPATCH_RigVMDispatch_ArrayIterator(in Array,out Element,out Index,out Count,out Ratio)"}
   Concat = /Script/RigVM.RigVMFunction_NameConcat(B=_ctrl, A=@DISPATCH_RigVMDispatch_ArrayIterator.Element.Name) {is=struct RigVMUnitNode, template="Concat::Execute(in A,in B,out Result)"}
   > &RigUnit_SetTransform(
    Item=(Type=Bone,Name="spine_01")
    Weight=1
    bInitial=False
    Item.Name=@DISPATCH_RigVMDispatch_ArrayIterator.Element.Name
    Value=&RigUnit_GetTransform(Item=(Type=Control,Name="spine_01_ctrl"), bInitial=False, Item.Name=@Concat.Result).Transform
   )
   n7 = &RigVMFunction_MathQuaternionSwingTwist(
    TwistAxis=(X=1,Y=0,Z=0)
    Input=&RigUnit_GetRelativeTransformForItem(Child=(Type=Bone,Name="hand_l"), Parent=(Type=Null,Name="hand_l_null"), bChildInitial=False, bParentInitial=False).RelativeTransform.Rotation
   )
   &RigUnit_ModifyTransforms(
    ItemToModify="((Item=(Type=Bone,Name=\"lowerarm_twist_01_l\"),Transform=(Rotation=(X=0,Y=0,Z=0,W=1),Translation=(X=0,Y=0,Z=0),Scale3D=(X=1,Y=1,Z=1))),(Item=(Type=Bone,Name=\"lowerarm_twist_02_l\"),Transform=(Rotation=(X=0,Y=0,Z=0,W=1),Translation=(X=0,Y=0,Z=0),Scale3D=(X=1, ...(truncated)"
    Weight=1
    WeightMinimum=0
    ItemToModify.1.Transform.Rotation=&RigVMFunction_MathQuaternionSlerp(B=(X=0,Y=0,Z=0,W=1), T=0.5, A=@n7.Twist) {template="Interpolate::Execute(in A,in B,in T,out Result)"}.Result
    ItemToModify.0.Transform.Rotation=@n7.Twist
    ExecutePin=@RigVMFunction_Sequence.E
   )
   n8 = &RigVMFunction_MathQuaternionSwingTwist(
    TwistAxis=(X=1,Y=0,Z=0)
    Input=&RigUnit_GetRelativeTransformForItem(Child=(Type=Bone,Name="hand_r"), Parent=(Type=Null,Name="hand_r_null"), bChildInitial=False, bParentInitial=False).RelativeTransform.Rotation
   )
   > &RigUnit_ModifyTransforms(
    ItemToModify="((Item=(Type=Bone,Name=\"lowerarm_twist_01_r\"),Transform=(Rotation=(X=0,Y=0,Z=0,W=1),Translation=(X=0,Y=0,Z=0),Scale3D=(X=1,Y=1,Z=1))),(Item=(Type=Bone,Name=\"lowerarm_twist_02_r\"),Transform=(Rotation=(X=0,Y=0,Z=0,W=1),Translation=(X=0,Y=0,Z=0),Scale3D=(X=1, ...(truncated)"
    Weight=1
    WeightMinimum=0
    ItemToModify.1.Transform.Rotation=&RigVMFunction_MathQuaternionSlerp(B=(X=0,Y=0,Z=0,W=1), T=0.5, A=@n8.Twist) {template="Interpolate::Execute(in A,in B,in T,out Result)"}.Result
    ItemToModify.0.Transform.Rotation=@n8.Twist
   )
  ]}
 New Function_ContainedGraph {nodes [
   RigVMFunctionEntryNode()
   > DeltaRotation = RigVMFunctionReferenceNode(Rotation Order=ZYX, child=hand_l, max=100, min=70, onlyTwist=true, parent=hand_l_null, parentType=Null, premult=1, twistAxis=(X=1,Y=0,Z=0))
   > &RigUnit_ModifyTransforms(
    ItemToModify="((Item=(Type=Bone,Name=\"lowerarm_twist_01_l\"),Transform=(Rotation=(X=0,Y=-0.0262,Z=0,W=0.9997),Translation=(X=-0.7,Y=-0.17,Z=1),Scale3D=(X=1,Y=1.1,Z=1.1))),(Item=(Type=Bone,Name=\"lowerarm_twist_02_l\"),Transform=(Rotation=(X=0.342,Y=0,Z=0,W=0.9397),Translat ...(truncated)"
    WeightMinimum=0
    Weight=@DeltaRotation.outputWeight
   )
   > RigVMFunctionReturnNode()
  ]}
 New Function_ContainedGraph {nodes [
   n1 = RigVMFunctionEntryNode()
   RigVMFunction_MathBoolToInteger = &RigVMFunction_MathBoolToInteger(Value=@n1.onlyTwist)
   RigVMFunction_MathQuaternionSwingTwist = &RigVMFunction_MathQuaternionSwingTwist(
    TwistAxis=@n1.twistAxis
    Input=&RigUnit_GetRelativeTransformForItem(
     Child=(Type=Bone,Name="None")
     Parent=(Type=Bone,Name="None")
     bChildInitial=False
     bParentInitial=False
     Child.Name=@n1.child
     Parent.Name=@n1.parent
     Parent.Type=@n1.parentType
    ).RelativeTransform.Rotation
   )
   RigVMFunction_MathQuaternionToEuler = &RigVMFunction_MathQuaternionToEuler(
    Value=RigVMDispatchNode(
     Values=((),())
     Index=@RigVMFunction_MathBoolToInteger.Result
     Values.0=@RigVMFunction_MathQuaternionSwingTwist.Swing
     Values.1=@RigVMFunction_MathQuaternionSwingTwist.Twist
    ) {template="DISPATCH_RigVMDispatch_SelectInt32(in Index,in Values,out Result)"}.Result
    RotationOrder=@n1.Rotation Order
   )
   > DISPATCH_RigVMDispatch_Print = RigVMDispatchNode(ScreenColor=(R=1,G=1,B=1,A=1), ScreenDuration=1.5, Enabled=@n1.debugRotation, Value=@RigVMFunction_MathQuaternionToEuler.Result) {template="DISPATCH_RigVMDispatch_Print(in Prefix,in Value,in Enabled,in ScreenDuration,in ScreenColor)"}
   Clamp = &RigVMFunction_MathIntClamp(Value=@n1.primAxis)
   DISPATCH_RigVMDispatch_SelectInt32 = RigVMDispatchNode(
    Values=(0,0,0)
    Index=@Clamp.Result
    Values.0=@RigVMFunction_MathQuaternionToEuler.Result.X
    Values.1=@RigVMFunction_MathQuaternionToEuler.Result.Y
    Values.2=@RigVMFunction_MathQuaternionToEuler.Result.Z
   ) {template="DISPATCH_RigVMDispatch_SelectInt32(in Index,in Values,out Result)"}
   Remap = &RigVMFunction_MathDoubleRemap(
    TargetMaximum=1
    TargetMinimum=0
    bClamp=False
    Value=@DISPATCH_RigVMDispatch_SelectInt32.Result
    SourceMaximum=@n1.max
    SourceMinimum=@n1.min
   )
   Multiply = &RigVMFunction_MathDoubleMul(A=&RigVMFunction_MathDoubleClamp(Value=@Remap.Result).Result, B=@n1.premult)
   > RigVMFunctionReturnNode(outputWeight=@Multiply.Result)
  ]}
 New Function_ContainedGraph {nodes [
   RigVMFunctionEntryNode()
   > DeltaRotation = RigVMFunctionReferenceNode(Rotation Order=ZYX, child=calf_l, max=180, min=100, parent=calf_l_null, parentType=Null, premult=1, primAxis=2, twistAxis=(X=1,Y=0,Z=0))
   > &RigUnit_ModifyTransforms(
    ItemToModify="((Item=(Type=Bone,Name=\"deform_knee_in_l\"),Transform=(Rotation=(X=0,Y=0,Z=0.342,W=0.9397),Translation=(X=0,Y=0,Z=0),Scale3D=(X=1,Y=1,Z=1))),(Item=(Type=Bone,Name=\"calf_twist_02_l\"),Transform=(Rotation=(X=0,Y=0,Z=0.0872,W=0.9962),Translation=(X=0,Y=4,Z=0),S ...(truncated)"
    WeightMinimum=0
    Weight=@DeltaRotation.outputWeight
   )
   > DeltaRotation_1 = RigVMFunctionReferenceNode(
    Rotation Order=ZYX
    child=calf_r
    debugRotation=false
    max=180
    min=100
    parent=calf_r_null
    parentType=Null
    premult=1
    primAxis=2
    twistAxis=(X=1,Y=0,Z=0)
   )
   > &RigUnit_ModifyTransforms(
    ItemToModify="((Item=(Type=Bone,Name=\"deform_knee_in_r\"),Transform=(Rotation=(X=0,Y=0,Z=0.342,W=0.9397),Translation=(X=0,Y=0,Z=0),Scale3D=(X=1,Y=1,Z=1))),(Item=(Type=Bone,Name=\"calf_twist_02_r\"),Transform=(Rotation=(X=0,Y=0,Z=0.0872,W=0.9962),Translation=(X=0,Y=-4,Z=0), ...(truncated)"
    WeightMinimum=0
    Weight=@DeltaRotation_1.outputWeight
   )
   > RigVMFunctionReturnNode()
  ]}
 New Function_ContainedGraph {nodes [
   RigVMFunctionEntryNode()
   > &RigUnit_ParentConstraint(
    AdvancedSettings=(RotationOrderForFilter=XZY)
    Child=(Type=Bone,Name="stk_tongueShoe_l")
    Parents=((Item=(Type=Bone,Name="ball_l"),Weight=0.0749),(Item=(Type=Bone,Name="calf_twist_01_l"),Weight=0.1188),(Item=(Type=Bone,Name="foot_l"),Weight=0.8063))
    Weight=1
   )
   > &RigUnit_ParentConstraint(
    AdvancedSettings=(RotationOrderForFilter=XZY)
    Child=(Type=Bone,Name="base_shoe_handle_l")
    Parents=((Item=(Type=Bone,Name="calf_twist_02_l"),Weight=0.0006),(Item=(Type=Bone,Name="foot_l"),Weight=0.064),(Item=(Type=Bone,Name="calf_twist_01_l"),Weight=0.9354))
    Weight=1
   )
   > &RigUnit_ParentConstraint(
    AdvancedSettings=(RotationOrderForFilter=XZY)
    Child=(Type=Bone,Name="stk_tongueShoe_r")
    Parents=((Item=(Type=Bone,Name="ball_r"),Weight=0.0749),(Item=(Type=Bone,Name="calf_twist_01_r"),Weight=0.1188),(Item=(Type=Bone,Name="foot_r"),Weight=0.8063))
    Weight=1
   )
   > &RigUnit_ParentConstraint(
    AdvancedSettings=(RotationOrderForFilter=XZY)
    Child=(Type=Bone,Name="base_shoe_handle_r")
    Parents=((Item=(Type=Bone,Name="calf_twist_02_r"),Weight=0.0006),(Item=(Type=Bone,Name="foot_r"),Weight=0.064),(Item=(Type=Bone,Name="calf_twist_01_r"),Weight=0.9354))
    Weight=1
   )
   > DeltaRotation = RigVMFunctionReferenceNode(Rotation Order=ZYX, child=foot_l, max=-35, parent=foot_l_null, parentType=Null, premult=1, primAxis=2, twistAxis=(X=1,Y=0,Z=0))
   > &RigUnit_ModifyTransforms(
    ItemToModify=((Item=(Type=Bone,Name="stk_tongueShoe_l"),Transform=(Rotation=(X=0,Y=0,Z=0.342,W=0.9397),Translation=(X=0,Y=0,Z=0),Scale3D=(X=1,Y=1,Z=1))))
    WeightMinimum=0
    Weight=@DeltaRotation.outputWeight
   )
   > DeltaRotation_1 = RigVMFunctionReferenceNode(Rotation Order=ZYX, child=foot_r, max=-35, parent=foot_r_null, parentType=Null, premult=1, primAxis=2, twistAxis=(X=1,Y=0,Z=0))
   > &RigUnit_ModifyTransforms(
    ItemToModify=((Item=(Type=Bone,Name="stk_tongueShoe_r"),Transform=(Rotation=(X=0,Y=0,Z=0.342,W=0.9397),Translation=(X=0,Y=0,Z=0),Scale3D=(X=1,Y=1,Z=1))))
    WeightMinimum=0
    Weight=@DeltaRotation_1.outputWeight
   )
   > RigVMFunctionReturnNode()
  ]}
 New Function_ContainedGraph {nodes [
   RigVMFunctionEntryNode()
   > &RigUnit_ParentConstraint(
    AdvancedSettings=(RotationOrderForFilter=XZY)
    Child=(Type=Bone,Name="base_armourPadArm_1")
    Parents=((Item=(Type=Bone,Name="deform_upperarm_up_r"),Weight=0.0003),(Item=(Type=Bone,Name="upperarm_r"),Weight=0.0069),(Item=(Type=Bone,Name="upperarm_twist_02_r"),Weight=0.2921),(Item=(Type=Bone,Name="upperarm_twist_01_r"),Weight=0.7007))
    Weight=1
   )
   > DeltaRotation = RigVMFunctionReferenceNode(
    Rotation Order=ZYX
    child=upperarm_r
    debugRotation=false
    max=-60
    min=-30
    parent=upperarm_r_null
    parentType=Null
    premult=1
    primAxis=1
    twistAxis=(X=1,Y=0,Z=0)
   )
   > &RigUnit_ModifyTransforms(
    ItemToModify=((Item=(Type=Bone,Name="base_armourPadArm_bend"),Transform=(Rotation=(X=0,Y=-0.0872,Z=0,W=0.9962),Translation=(X=0,Y=0,Z=3),Scale3D=(X=1,Y=1,Z=1))))
    WeightMinimum=0
    Weight=@DeltaRotation.outputWeight
   )
   > DeltaRotation_1 = RigVMFunctionReferenceNode(
    Rotation Order=ZYX
    child=clavicle_r
    debugRotation=false
    max=-30
    min=0
    onlyTwist=false
    parent=clavicle_r_null
    parentType=Null
    premult=1
    primAxis=1
    twistAxis=(X=1,Y=0,Z=0)
   )
   Multiply = &RigVMFunction_MathDoubleMul(A=@DeltaRotation_1.outputWeight, B=@DeltaRotation.outputWeight)
   > &RigUnit_ModifyTransforms(
    ItemToModify=((Item=(Type=Bone,Name="base_armourPadArm_bend"),Transform=(Rotation=(X=0,Y=0.1736,Z=0,W=0.9848),Translation=(X=-2,Y=0,Z=0),Scale3D=(X=1,Y=1,Z=1))))
    WeightMinimum=0
    Weight=@Multiply.Result
   )
   > n1 = &RigUnit_SphericalPoseReader(
    ActiveRegionScaleFactors=(PositiveWidth=1,NegativeWidth=1,PositiveHeight=1,NegativeHeight=1)
    ActiveRegionSize=0.1
    Debug=(bDrawDebug=false,bDraw2D=False,bDrawLocalAxes=False,DebugScale=25,DebugSegments=20,DebugThickness=0.25)
    DriverAxis=(X=-1,Y=0,Z=0)
    DriverItem=(Type=Bone,Name="upperarm_r")
    FalloffRegionScaleFactors=(PositiveWidth=1,NegativeWidth=1,PositiveHeight=1,NegativeHeight=1)
    FalloffSize=0.22
    FlipHeightScaling=False
    FlipWidthScaling=False
    OptionalParentItem=(Type=Bone,Name="None")
    RotationOffset=(X=-125,Y=30,Z=7)
   )
   > &RigUnit_ModifyTransforms(
    ItemToModify=((Item=(Type=Bone,Name="base_armourPadArm_1"),Transform=(Rotation=(X=-0.4226,Y=0,Z=0,W=0.9063),Translation=(X=0,Y=-3.8,Z=2),Scale3D=(X=1,Y=1,Z=1))))
    WeightMinimum=0
    Weight=@n1.OutputParam
   )
   > n2 = &RigUnit_SphericalPoseReader(
    ActiveRegionScaleFactors=(PositiveWidth=1,NegativeWidth=1,PositiveHeight=1.5,NegativeHeight=1.5)
    ActiveRegionSize=0.1
    Debug=(bDrawDebug=false,bDraw2D=False,bDrawLocalAxes=False,DebugScale=25,DebugSegments=20,DebugThickness=0.25)
    DriverAxis=(X=-1,Y=0,Z=0)
    DriverItem=(Type=Bone,Name="upperarm_r")
    FalloffRegionScaleFactors=(PositiveWidth=1,NegativeWidth=1,PositiveHeight=2,NegativeHeight=2)
    FalloffSize=0.2
    FlipHeightScaling=False
    FlipWidthScaling=False
    OptionalParentItem=(Type=Bone,Name="None")
    RotationOffset=(X=-128,Y=-45,Z=-6)
   )
   > DeltaRotation_2 = RigVMFunctionReferenceNode(
    Rotation Order=ZYX
    child=upperarm_r
    debugRotation=false
    max=-50
    min=30
    onlyTwist=true
    parent=upperarm_r_null
    parentType=Null
    premult=1
    primAxis=0
    twistAxis=(X=1,Y=0,Z=0)
   )
   > n3 = &RigUnit_ModifyTransforms(
    ItemToModify=((Item=(Type=Bone,Name="base_armourPadArm_bend"),Transform=(Rotation=(X=0,Y=-0.0436,Z=0,W=0.999),Translation=(X=0,Y=0,Z=0),Scale3D=(X=1,Y=1,Z=1))))
    WeightMinimum=0
    Weight=&RigVMFunction_MathDoubleMul(A=@DeltaRotation_2.outputWeight, B=@n2.OutputParam).Result
   )
   RigVMFunction_ControlFlowBranch = &RigVMFunction_ControlFlowBranch(ExecuteContext=@n3.ExecutePin, Condition=RigVMVariableNode() {variable=pub isPlayingEmote}.Value)
   n4 = &RigVMFunction_ControlFlowBranch(ExecuteContext=@RigVMFunction_ControlFlowBranch.Completed, Condition=RigVMVariableNode() {variable=pub isPlayingMeeleSwing}.Value)
   RigVMFunctionReturnNode(ExecuteContext=@n4.Completed)
   n5 = &RigUnit_SphericalPoseReader(
    ActiveRegionScaleFactors=(PositiveWidth=1,NegativeWidth=1,PositiveHeight=1,NegativeHeight=1)
    ActiveRegionSize=0.1
    Debug=(bDrawDebug=false,bDraw2D=False,bDrawLocalAxes=False,DebugScale=25,DebugSegments=20,DebugThickness=0.25)
    DriverAxis=(X=-1,Y=0,Z=0)
    DriverItem=(Type=Bone,Name="upperarm_r")
    FalloffRegionScaleFactors=(PositiveWidth=1,NegativeWidth=1,PositiveHeight=1,NegativeHeight=1)
    FalloffSize=0.2
    FlipHeightScaling=False
    FlipWidthScaling=False
    OptionalParentItem=(Type=Bone,Name="spine_05")
    RotationOffset=(X=-10,Y=70,Z=100)
    ExecutePin=@n4.True
   )
   > &RigUnit_ModifyTransforms(
    ItemToModify=((Item=(Type=Bone,Name="base_armourPadArm_1"),Transform=(Rotation=(X=-0.5,Y=0,Z=0,W=0.866),Translation=(X=0,Y=-4,Z=4),Scale3D=(X=1,Y=1,Z=1))))
    WeightMinimum=0
    Weight=@n5.OutputParam
   )
   n6 = &RigUnit_SphericalPoseReader(
    ActiveRegionScaleFactors=(PositiveWidth=1,NegativeWidth=1,PositiveHeight=1.5,NegativeHeight=1.5)
    ActiveRegionSize=0.1
    Debug=(bDrawDebug=false,bDraw2D=False,bDrawLocalAxes=False,DebugScale=25,DebugSegments=20,DebugThickness=0.25)
    DriverAxis=(X=-1,Y=0,Z=0)
    DriverItem=(Type=Bone,Name="upperarm_r")
    FalloffRegionScaleFactors=(PositiveWidth=1,NegativeWidth=1,PositiveHeight=2,NegativeHeight=2)
    FalloffSize=0.25
    FlipHeightScaling=False
    FlipWidthScaling=False
    OptionalParentItem=(Type=Bone,Name="spine_05")
    RotationOffset=(X=-20,Y=210,Z=0)
    ExecutePin=@RigVMFunction_ControlFlowBranch.True
   )
   > n7 = &RigUnit_ModifyTransforms(
    ItemToModify=((Item=(Type=Bone,Name="base_armourPadArm_1"),Transform=(Rotation=(X=-0.7597,Y=0.0332,Z=-0.0283,W=0.6488),Translation=(X=0,Y=-5,Z=4),Scale3D=(X=1,Y=1,Z=1))))
    WeightMinimum=0
    Weight=@n6.OutputParam
   )
   n8 = &RigVMFunction_ControlFlowBranch(ExecuteContext=@n7.ExecutePin, Condition=RigVMVariableNode() {variable=pub isStormAviator}.Value)
   n9 = &RigUnit_SphericalPoseReader(
    ActiveRegionScaleFactors=(PositiveWidth=1,NegativeWidth=1,PositiveHeight=1,NegativeHeight=1)
    ActiveRegionSize=0.1
    Debug=(bDrawDebug=true,bDraw2D=False,bDrawLocalAxes=False,DebugScale=25,DebugSegments=20,DebugThickness=0.25)
    DriverAxis=(X=-1,Y=0,Z=0)
    DriverItem=(Type=Bone,Name="upperarm_r")
    FalloffRegionScaleFactors=(PositiveWidth=1,NegativeWidth=1,PositiveHeight=1,NegativeHeight=1)
    FalloffSize=0.25
    FlipHeightScaling=False
    FlipWidthScaling=False
    OptionalParentItem=(Type=Bone,Name="spine_05")
    RotationOffset=(X=-20,Y=155,Z=15)
    ExecutePin=@n8.True
   )
   > &RigUnit_ModifyTransforms(
    ItemToModify=((Item=(Type=Bone,Name="base_armourPadArm_1"),Transform=(Rotation=(X=-0.4222,Y=-0.0232,Z=0.0084,W=0.9062),Translation=(X=0,Y=-5,Z=3),Scale3D=(X=1,Y=1,Z=1))))
    WeightMinimum=0
    Weight=@n9.OutputParam
   )
  ]}
 New Function_ContainedGraph {nodes [
   RigVMFunctionEntryNode()
   > &RigUnit_ParentConstraint(
    AdvancedSettings=(RotationOrderForFilter=XZY)
    Child=(Type=Bone,Name="base_chestCard_r")
    Parents="((Item=(Type=Bone,Name=\"pelvis\"),Weight=0.0002),(Item=(Type=Bone,Name=\"spine_01\"),Weight=0.0063),(Item=(Type=Bone,Name=\"spine_02\"),Weight=0.0365),(Item=(Type=Bone,Name=\"spine_03\"),Weight=0.0705),(Item=(Type=Bone,Name=\"clavicle_r\"),Weight=0.0993),(Item=(Typ ...(truncated)"
    Weight=1
   )
   > &RigUnit_ParentConstraint(
    AdvancedSettings=(RotationOrderForFilter=XZY)
    Child=(Type=Bone,Name="base_chestCard_c")
    Parents="((Item=(Type=Bone,Name=\"pelvis\"),Weight=0.0003),(Item=(Type=Bone,Name=\"spine_01\"),Weight=0.0091),(Item=(Type=Bone,Name=\"spine_02\"),Weight=0.0495),(Item=(Type=Bone,Name=\"spine_03\"),Weight=0.0846),(Item=(Type=Bone,Name=\"spine_04\"),Weight=0.2868),(Item=(Type= ...(truncated)"
    Weight=1
   )
   > &RigUnit_ParentConstraint(
    AdvancedSettings=(RotationOrderForFilter=XZY)
    Child=(Type=Bone,Name="base_chestCard_l")
    Parents="((Item=(Type=Bone,Name=\"pelvis\"),Weight=0.0002),(Item=(Type=Bone,Name=\"spine_01\"),Weight=0.007),(Item=(Type=Bone,Name=\"spine_02\"),Weight=0.0405),(Item=(Type=Bone,Name=\"spine_03\"),Weight=0.0783),(Item=(Type=Bone,Name=\"spine_04\"),Weight=0.2917),(Item=(Type=B ...(truncated)"
    Weight=1
   )
   > DeltaRotation = RigVMFunctionReferenceNode(Rotation Order=ZYX, child=spine_05, max=30, min=1, parent=spine_04, parentType=Bone, premult=1, primAxis=2, twistAxis=(X=1,Y=0,Z=0))
   > &RigUnit_ModifyTransforms(
    ItemToModify="((Item=(Type=Bone,Name=\"base_chestCard_c\"),Transform=(Rotation=(X=0,Y=-0.1736,Z=0,W=0.9848),Translation=(X=0,Y=0,Z=0),Scale3D=(X=1,Y=1,Z=1))),(Item=(Type=Bone,Name=\"base_chestCard_l\"),Transform=(Rotation=(X=0,Y=-0.1736,Z=0,W=0.9848),Translation=(X=0,Y=0,Z= ...(truncated)"
    WeightMinimum=0
    Weight=@DeltaRotation.outputWeight
   )
   > n1 = &RigUnit_ModifyTransforms(
    ItemToModify=((Item=(Type=Bone,Name="base_beltHipCard_1"),Transform=(Rotation=(X=-0.1608,Y=-0.4774,Z=-0.1927,W=0.8421),Translation=(X=0,Y=0,Z=0),Scale3D=(X=1,Y=1,Z=1))))
    WeightMinimum=0
    Weight=RigVMVariableNode() {variable=r_thigh_rzn100}.Value
   )
   RigVMFunction_ControlFlowBranch = &RigVMFunction_ControlFlowBranch(ExecuteContext=@n1.ExecutePin, Condition=RigVMVariableNode() {variable=pub isPlayingEasterBloc}.Value)
   RigVMFunctionReturnNode(ExecuteContext=@RigVMFunction_ControlFlowBranch.Completed)
   n2 = &RigUnit_SphericalPoseReader(
    ActiveRegionScaleFactors=(PositiveWidth=1,NegativeWidth=1,PositiveHeight=1,NegativeHeight=1)
    ActiveRegionSize=0.1
    Debug=(bDrawDebug=false,bDraw2D=False,bDrawLocalAxes=False,DebugScale=25,DebugSegments=20,DebugThickness=0.25)
    DriverAxis=(X=1,Y=0,Z=0)
    DriverItem=(Type=Bone,Name="thigh_r")
    FalloffRegionScaleFactors=(PositiveWidth=1,NegativeWidth=1,PositiveHeight=1,NegativeHeight=1)
    FalloffSize=0.25
    FlipHeightScaling=False
    FlipWidthScaling=False
    OptionalParentItem=(Type=Bone,Name="None")
    RotationOffset=(X=-25,Y=6,Z=0)
    ExecutePin=@RigVMFunction_ControlFlowBranch.True
   )
   > &RigUnit_ModifyTransforms(
    ItemToModify=((Item=(Type=Bone,Name="base_armourPadLeg_1"),Transform=(Rotation=(X=-0.4082,Y=0.2346,Z=0.1094,W=0.8754),Translation=(X=2,Y=0,Z=-1),Scale3D=(X=1,Y=1,Z=1))))
    WeightMinimum=0
    Weight=@n2.OutputParam
   )
  ]}
 New Function_ContainedGraph {nodes [
   RigVMFunctionEntryNode()
   > &RigUnit_ParentConstraint(
    AdvancedSettings=(RotationOrderForFilter=XZY)
    Child=(Type=Bone,Name="base_belt_1")
    Parents="((Item=(Type=Bone,Name=\"thigh_l\"),Weight=0.0001),(Item=(Type=Bone,Name=\"spine_03\"),Weight=0.0007),(Item=(Type=Bone,Name=\"deform_glute_l\"),Weight=0.0008),(Item=(Type=Bone,Name=\"thigh_twist_01_r\"),Weight=0.0055),(Item=(Type=Bone,Name=\"deform_groin_l\"),Weight ...(truncated)"
    Weight=1
   )
   > &RigUnit_ParentConstraint(
    AdvancedSettings=(RotationOrderForFilter=XZY)
    Child=(Type=Bone,Name="base_belt_2")
    Parents="((Item=(Type=Bone,Name=\"spine_03\"),Weight=0.0047),(Item=(Type=Bone,Name=\"deform_groin_l\"),Weight=0.0076),(Item=(Type=Bone,Name=\"deform_groin_r\"),Weight=0.0076),(Item=(Type=Bone,Name=\"thigh_twist_01_l\"),Weight=0.0193),(Item=(Type=Bone,Name=\"thigh_twist_01_r ...(truncated)"
    Weight=1
   )
   > &RigUnit_ParentConstraint(
    AdvancedSettings=(RotationOrderForFilter=XZY)
    Child=(Type=Bone,Name="base_belt_3")
    Parents="((Item=(Type=Bone,Name=\"deform_glute_r\"),Weight=0.0004),(Item=(Type=Bone,Name=\"deform_groin_r\"),Weight=0.0008),(Item=(Type=Bone,Name=\"thigh_twist_01_l\"),Weight=0.0025),(Item=(Type=Bone,Name=\"spine_03\"),Weight=0.0029),(Item=(Type=Bone,Name=\"thigh_twist_01_r ...(truncated)"
    Weight=1
   )
   > &RigUnit_ParentConstraint(
    AdvancedSettings=(RotationOrderForFilter=XZY)
    Child=(Type=Bone,Name="base_belt_4")
    Parents="((Item=(Type=Bone,Name=\"spine_03\"),Weight=0.0017),(Item=(Type=Bone,Name=\"thigh_twist_01_r\"),Weight=0.0029),(Item=(Type=Bone,Name=\"deform_glute_l\"),Weight=0.0045),(Item=(Type=Bone,Name=\"thigh_twist_01_l\"),Weight=0.0158),(Item=(Type=Bone,Name=\"spine_02\"),Wei ...(truncated)"
    Weight=1
   )
   > &RigUnit_ParentConstraint(
    AdvancedSettings=(RotationOrderForFilter=XZY)
    Child=(Type=Bone,Name="base_belt_5")
    Parents="((Item=(Type=Bone,Name=\"deform_glute_l\"),Weight=0.0003),(Item=(Type=Bone,Name=\"deform_glute_r\"),Weight=0.0003),(Item=(Type=Bone,Name=\"thigh_twist_01_l\"),Weight=0.0011),(Item=(Type=Bone,Name=\"thigh_twist_01_r\"),Weight=0.0011),(Item=(Type=Bone,Name=\"spine_03 ...(truncated)"
    Weight=1
   )
   > &RigUnit_ParentConstraint(
    AdvancedSettings=(RotationOrderForFilter=XZY)
    Child=(Type=Bone,Name="base_belt_6")
    Parents="((Item=(Type=Bone,Name=\"spine_04\"),Weight=0.0007),(Item=(Type=Bone,Name=\"thigh_twist_01_l\"),Weight=0.0013),(Item=(Type=Bone,Name=\"deform_groin_r\"),Weight=0.0031),(Item=(Type=Bone,Name=\"thigh_twist_01_r\"),Weight=0.0068),(Item=(Type=Bone,Name=\"spine_03\"),Wei ...(truncated)"
    Weight=1
   )
   > &RigUnit_ParentConstraint(
    AdvancedSettings=(RotationOrderForFilter=XZY)
    Child=(Type=Bone,Name="base_belt_7")
    Parents="((Item=(Type=Bone,Name=\"spine_03\"),Weight=0.0017),(Item=(Type=Bone,Name=\"thigh_twist_01_l\"),Weight=0.0029),(Item=(Type=Bone,Name=\"deform_glute_r\"),Weight=0.0045),(Item=(Type=Bone,Name=\"thigh_twist_01_r\"),Weight=0.0158),(Item=(Type=Bone,Name=\"spine_02\"),Wei ...(truncated)"
    Weight=1
   )
   > DeltaRotation = RigVMFunctionReferenceNode(Rotation Order=ZYX, child=thigh_l, max=-100, parent=thigh_l_null, parentType=Null, premult=1, primAxis=2, twistAxis=(X=1,Y=0,Z=0))
   > &RigUnit_ModifyTransforms(
    ItemToModify="((Item=(Type=Bone,Name=\"base_belt_1\"),Transform=(Rotation=(X=-0.0888,Y=-0.1715,Z=-0.0323,W=0.9806),Translation=(X=-3.5,Y=3.1,Z=4),Scale3D=(X=1,Y=0.8,Z=1))),(Item=(Type=Bone,Name=\"base_belt_2\"),Transform=(Rotation=(X=0.0098,Y=0.0705,Z=-0.0689,W=0.9951),Tran ...(truncated)"
    WeightMinimum=0
    Weight=@DeltaRotation.outputWeight
   )
   > DeltaRotation_3 = RigVMFunctionReferenceNode(Rotation Order=ZYX, child=thigh_r, max=-100, parent=thigh_r_null, parentType=Null, premult=1, primAxis=2, twistAxis=(X=1,Y=0,Z=0))
   > RigVMVariableNode(Value=@DeltaRotation_3.outputWeight) {variable=r_thigh_rzn100}
   > &RigUnit_ModifyTransforms(
    ItemToModify="((Item=(Type=Bone,Name=\"base_belt_6\"),Transform=(Rotation=(X=0,Y=0,Z=0.0022,W=1),Translation=(X=0,Y=1.45,Z=0.6),Scale3D=(X=1,Y=1,Z=1))),(Item=(Type=Bone,Name=\"base_belt_3\"),Transform=(Rotation=(X=0,Y=0,Z=0,W=1),Translation=(X=0,Y=0.8,Z=0.53),Scale3D=(X=1,Y ...(truncated)"
    WeightMinimum=0
    Weight=@DeltaRotation_3.outputWeight
   )
   > DeltaRotation_2 = RigVMFunctionReferenceNode(Rotation Order=ZYX, child=thigh_l, max=-90, min=-20, parent=thigh_l_null, parentType=Null, premult=1, primAxis=1, twistAxis=(X=1,Y=0,Z=0))
   > &RigUnit_ModifyTransforms(
    ItemToModify=((Item=(Type=Bone,Name="base_beltPad"),Transform=(Rotation=(X=0,Y=-0.342,Z=0,W=0.9397),Translation=(X=0,Y=0,Z=0),Scale3D=(X=1,Y=1,Z=1))))
    WeightMinimum=0
    Weight=@DeltaRotation_2.outputWeight
   )
   > n1 = &RigUnit_SphericalPoseReader(
    ActiveRegionScaleFactors=(PositiveWidth=1,NegativeWidth=1,PositiveHeight=1,NegativeHeight=1)
    ActiveRegionSize=0.05
    Debug=(bDrawDebug=false,bDraw2D=False,bDrawLocalAxes=False,DebugScale=25,DebugSegments=20,DebugThickness=0.25)
    DriverAxis=(X=1,Y=0,Z=0)
    DriverItem=(Type=Bone,Name="thigh_r")
    FalloffRegionScaleFactors=(PositiveWidth=1,NegativeWidth=1,PositiveHeight=1,NegativeHeight=1)
    FalloffSize=0.15
    FlipHeightScaling=False
    FlipWidthScaling=False
    OptionalParentItem=(Type=Bone,Name="None")
    RotationOffset=(X=-64,Y=0,Z=8)
   )
   > &RigUnit_ModifyTransforms(
    ItemToModify="((Item=(Type=Bone,Name=\"base_belt_6\"),Transform=(Rotation=(X=-0.1736,Y=0,Z=0,W=0.9848),Translation=(X=0,Y=0.5,Z=2),Scale3D=(X=1,Y=1,Z=1))),(Item=(Type=Bone,Name=\"base_belt_2\"),Transform=(Rotation=(X=0,Y=0,Z=0,W=1),Translation=(X=0,Y=0,Z=1),Scale3D=(X=1,Y=1 ...(truncated)"
    WeightMinimum=0
    Weight=@n1.OutputParam
   )
   > DeltaRotation_1 = RigVMFunctionReferenceNode(Rotation Order=ZYX, child=spine_05, max=-50, onlyTwist=true, parent=spine_01, parentType=Bone, premult=1, twistAxis=(X=1,Y=0,Z=0))
   > &RigUnit_ModifyTransforms(
    ItemToModify=((Item=(Type=Bone,Name="base_belt_2"),Transform=(Rotation=(X=0,Y=0,Z=0,W=1),Translation=(X=0,Y=0,Z=1),Scale3D=(X=1,Y=1,Z=1))),(Item=(Type=Bone,Name="base_belt_6"),Transform=(Rotation=(X=0,Y=0,Z=0,W=1),Translation=(X=0,Y=0,Z=1),Scale3D=(X=1,Y=1,Z=1))))
    WeightMinimum=0
    Weight=@DeltaRotation_1.outputWeight
   )
   > DeltaRotation_4 = RigVMFunctionReferenceNode(
    Rotation Order=ZYX
    child=spine_02
    debugRotation=false
    max=20
    onlyTwist=false
    parent=pelvis
    parentType=Bone
    premult=1
    primAxis=2
    twistAxis=(X=1,Y=0,Z=0)
   )
   > &RigUnit_ModifyTransforms(
    ItemToModify="((Item=(Type=Bone,Name=\"base_belt_2\"),Transform=(Rotation=(X=0,Y=0,Z=-0.0436,W=0.999),Translation=(X=0,Y=0,Z=1),Scale3D=(X=1,Y=1,Z=1))),(Item=(Type=Bone,Name=\"base_belt_6\"),Transform=(Rotation=(X=-0.1736,Y=0,Z=0,W=0.9848),Translation=(X=0,Y=0,Z=2),Scale3D= ...(truncated)"
    WeightMinimum=0
    Weight=@DeltaRotation_4.outputWeight
   )
   > RigVMFunctionReturnNode()
  ]}
 New Function_ContainedGraph {nodes [
   RigVMFunctionEntryNode()
   > &RigUnit_ParentConstraint(
    AdvancedSettings=(RotationOrderForFilter=XZY)
    Child=(Type=Bone,Name="base_smallBelt_3")
    Parents="((Item=(Type=Bone,Name=\"spine_02\"),Weight=0.0069),(Item=(Type=Bone,Name=\"thigh_twist_01_l\"),Weight=0.0174),(Item=(Type=Bone,Name=\"thigh_twist_01_r\"),Weight=0.0529),(Item=(Type=Bone,Name=\"spine_01\"),Weight=0.1106),(Item=(Type=Bone,Name=\"deform_glute_r\"),Wei ...(truncated)"
    Weight=1
   )
   > &RigUnit_ParentConstraint(
    AdvancedSettings=(RotationOrderForFilter=XZY)
    Child=(Type=Bone,Name="base_smallBelt_2")
    Parents="((Item=(Type=Bone,Name=\"spine_02\"),Weight=0.0015),(Item=(Type=Bone,Name=\"thigh_r\"),Weight=0.0036),(Item=(Type=Bone,Name=\"thigh_twist_01_l\"),Weight=0.0212),(Item=(Type=Bone,Name=\"spine_01\"),Weight=0.0434),(Item=(Type=Bone,Name=\"deform_glute_r\"),Weight=0.046 ...(truncated)"
    Weight=1
   )
   > &RigUnit_ParentConstraint(
    AdvancedSettings=(RotationOrderForFilter=XZY)
    Child=(Type=Bone,Name="base_smallBelt_1")
    Parents="((Item=(Type=Bone,Name=\"spine_03\"),Weight=0.0001),(Item=(Type=Bone,Name=\"thigh_r\"),Weight=0.0001),(Item=(Type=Bone,Name=\"deform_groin_l\"),Weight=0.0013),(Item=(Type=Bone,Name=\"spine_02\"),Weight=0.0115),(Item=(Type=Bone,Name=\"thigh_twist_01_l\"),Weight=0.021 ...(truncated)"
    Weight=1
   )
   > DeltaRotation = RigVMFunctionReferenceNode(Rotation Order=ZYX, child=thigh_r, max=-90, parent=thigh_r_null, parentType=Null, premult=1, primAxis=2, twistAxis=(X=1,Y=0,Z=0))
   > &RigUnit_ModifyTransforms(
    ItemToModify="((Item=(Type=Bone,Name=\"base_smallBelt_1\"),Transform=(Rotation=(X=0,Y=0,Z=0.2299,W=0.9732),Translation=(X=-0.619,Y=-0.1765,Z=1.1623),Scale3D=(X=1,Y=1,Z=1))),(Item=(Type=Bone,Name=\"base_smallBelt_2\"),Transform=(Rotation=(X=0,Y=0,Z=0.1332,W=0.9911),Translati ...(truncated)"
    WeightMinimum=0
    Weight=@DeltaRotation.outputWeight
   )
   > DeltaRotation_3 = RigVMFunctionReferenceNode(Rotation Order=ZYX, child=thigh_r, max=-80, min=-10, parent=thigh_r_null, parentType=Null, premult=1, primAxis=1, twistAxis=(X=1,Y=0,Z=0))
   > RigVMVariableNode(Value=@DeltaRotation_3.outputWeight) {variable=r_thigh_ryn_range_10_80}
   > &RigUnit_ModifyTransforms(
    ItemToModify=((Item=(Type=Bone,Name="base_smallBelt_2"),Transform=(Rotation=(X=0,Y=0,Z=0,W=1),Translation=(X=-1.9,Y=-0.5,Z=2),Scale3D=(X=1,Y=1,Z=1))))
    WeightMinimum=0
    Weight=@DeltaRotation_3.outputWeight
   )
   > RigVMFunctionReturnNode()
  ]}
 New Function_ContainedGraph {nodes [
   RigVMFunctionEntryNode()
   > &RigUnit_ModifyTransforms(
    ItemToModify=((Item=(Type=Bone,Name="base_armourPadLeg_1"),Transform=(Rotation=(X=0,Y=0.5,Z=0,W=0.866),Translation=(X=-1,Y=0,Z=-1),Scale3D=(X=1,Y=1,Z=1))))
    WeightMinimum=0
    Weight=RigVMVariableNode() {variable=r_thigh_ryn_range_10_80}.Value
   )
   > RigVMFunctionReturnNode()
  ]}
 RigVMFunctionLibrary {nodes [
   DeltaRotation = RigVMCollapseNode(ExecuteContext=(), Rotation Order=ZYX, premult=1, twistAxis=(X=1,Y=0,Z=0))
   TwistsFix = RigVMCollapseNode(ExecuteContext=())
   armArmour = RigVMCollapseNode(ExecuteContext=())
   controllers = RigVMCollapseNode(ExecuteContext=(), debug=False)
   customKneeSetup = RigVMCollapseNode(ExecuteContext=())
   legArmour = RigVMCollapseNode(ExecuteContext=())
   mainBelt = RigVMCollapseNode(ExecuteContext=())
   misc = RigVMCollapseNode(ExecuteContext=())
   secBelt = RigVMCollapseNode(ExecuteContext=())
   shoes = RigVMCollapseNode(ExecuteContext=())
  ]}
 RigVMFunction_Sequence_2_ContainedGraph {nodes [
   n1 = RigVMFunctionEntryNode()
   RigVMFunction_Sequence = &RigVMFunction_Sequence(ExecuteContext=@n1.ExecuteContext)
   n2 = &RigVMFunction_Sequence(ExecuteContext=@RigVMFunction_Sequence.B)
   n3 = &RigVMFunction_Sequence(ExecuteContext=@n2.B)
   n4 = &RigVMFunction_Sequence(ExecuteContext=@n3.B)
   RigVMFunctionReturnNode(A=@RigVMFunction_Sequence.A, B=@n2.A, C=@n3.A, D=@n4.A, E=@n4.B)
  ]}
 RigVMModel {nodes [
   RigUnit_BeginExecution = &RigUnit_BeginExecution()
   controllers = RigVMFunctionReferenceNode(debug=false, ExecuteContext=@RigUnit_BeginExecution.ExecutePin)
   > TwistsFix = RigVMFunctionReferenceNode()
   > FN_DeformRig_FNC = RigVMFunctionReferenceNode(
    DeformRig_ Curve="(deformRig_l_upperarm_up_alpha_72_#1=False,deformRig_r_upperarm_up_alpha_73_#2=False,deformRig_l_upperarm_fr_alpha_74_#3=False,deformRig_r_upperarm_fr_alpha_75_CB51B8 ...(truncated)"
    DeformRig_Bool="(upperarm_1_#4=True,elbow_3_#5=True,wrist_7_#6=True,groin_8_#7=True,glute_17_#8=True,knee_9_40C408324CA261 ...(truncated)"
    DeformRig_Debug=(upperarm_debug_18_#4=False,wrist_debug_20_#6=False,lat_debug_26_#9=False,groinGlute_debug_27_#7=False)
    DeformRig_Value="(upperarm_up_18_#10=(Rotation=(X=0,Y=0,Z=0,W=1),Translation=(X=0,Y=0,Z=8),Scale3D=(X=0.7,Y=1,Z=1)),upperarm_fwd_fr_19_#11=(Rotation=(X=0,Y=0,Z=0,W=1),Translation=(X=0,Y=10,Z=0),Scale3D=(X=0.5,Y=1,Z= ...(truncated)"
   )
   > customKneeSetup = RigVMFunctionReferenceNode()
   > mainBelt = RigVMFunctionReferenceNode()
   > secBelt = RigVMFunctionReferenceNode()
   > armArmour = RigVMFunctionReferenceNode()
   > legArmour = RigVMFunctionReferenceNode()
   > shoes = RigVMFunctionReferenceNode()
   > misc = RigVMFunctionReferenceNode()
   > /Script/ControlRig.RigUnit_SetCurveValue(Curve=r_thigh_rz_up, Value=RigVMVariableNode() {variable=r_thigh_rzn100}.Value) {is=struct RigVMUnitNode}
  ]}
]
hierarchy [
 root {loc=<0 0 0>, rot=<0 0 0 1>}
 pelvis {parent=root, loc=<0 -0.5301 98.0571>, rot=<0.0127 0.707 -0.0127 -0.707>}
 spine_01 {parent=pelvis, loc=<6.1208 -0.9328 0>, rot=<0 0 0.034 -0.9994>}
 spine_02 {parent=spine_01, loc=<6.5997 -0.2095 0>, rot=<0 0 0.0238 -0.9997>}
 spine_03 {parent=spine_02, loc=<6.4914 -0.1237 0>, rot=<0 0 0.0149 -0.9999>}
 spine_04 {parent=spine_03, loc=<6.9598 0.0794 0>, rot=<0 0 -0.0132 -0.9999>}
 spine_05 {parent=spine_04, loc=<6.9987 0 0>, rot=<0 0 -0.0015 -1>}
 attach_cape {parent=spine_05, loc=<9.3216 -8.4501 0>, rot=<0.0616 -0.7044 -0.0616 -0.7044>}
 neck_01 {parent=spine_05, loc=<11.3138 0.1529 0>, rot=<0 0 -0.2003 -0.9797>}
 neck_02 {parent=neck_01, loc=<4.5627 0 0>, rot=<0 0 0.0504 -0.9987>}
 head {parent=neck_02, loc=<5.3914 0 0>, rot=<0 0 0.115 -0.9934>}
 attach_backpack {parent=spine_05, loc=<0 -8.35 0>, rot=<-0.0093 -0.707 0.0093 -0.707>}
 clavicle_l {parent=spine_05, loc=<5.4831 -0.1053 -3.2946>, rot=<0.0465 -0.7289 -0.009 -0.6829>}
 upperarm_l {parent=clavicle_l, loc=<9.8658 -1.2136 -0.4105>, rot=<0.0164 -0.3245 -0.0269 -0.9454>}
 lowerarm_l {parent=upperarm_l, loc=<26.465 -0.8345 0.259>, rot=<0.0036 0.0161 -0.2298 -0.9731>}
 hand_l {parent=lowerarm_l, loc=<22.7355 -0.3203 -0.6105>, rot=<-0.6784 -0.0223 0.0735 -0.7307>}
 thumb_01_l {parent=hand_l, loc=<1.3118 -0.3191 -2.459>, rot=<0.6546 -0.4042 0.0664 -0.6354>}
 thumb_02_l {parent=thumb_01_l, loc=<5.9336 -0.0448 0.0401>, rot=<-0.0037 0.0213 0.298 -0.9543>}
 thumb_03_l {parent=thumb_02_l, loc=<2.2223 0.3682 -0.0907>, rot=<0.0118 -0.013 0.0066 -0.9998>}
 index_metacarpal_l {parent=hand_l, loc=<1.6535 0.0183 -2.96>, rot=<0.0158 -0.0124 -0.0173 -0.9996>}
 index_01_l {parent=index_metacarpal_l, loc=<8.3008 0.0357 -0.0524>, rot=<-0.0044 0.0485 0.0924 -0.9945>}
 index_02_l {parent=index_01_l, loc=<3.3587 -0.3004 0.0452>, rot=<-0.002 0.0027 0.3225 -0.9466>}
 index_03_l {parent=index_02_l, loc=<2.5578 -0.123 -0.0252>, rot=<-0.002 0.0067 0.2558 -0.9667>}
 middle_metacarpal_l {parent=hand_l, loc=<2.1856 -0.1725 -0.5984>, rot=<0.006 0.0168 -0.0329 -0.9993>}
 middle_01_l {parent=middle_metacarpal_l, loc=<7.3275 -0.2447 0.0129>, rot=<-0.0208 0.0042 0.0858 -0.9961>}
 middle_02_l {parent=middle_01_l, loc=<4.0123 0.1401 0.1175>, rot=<-0.0218 0.0258 0.3512 -0.9357>}
 middle_03_l {parent=middle_02_l, loc=<2.6077 0.0534 0.1373>, rot=<0.0082 0.0188 0.1708 -0.9851>}
 ring_metacarpal_l {parent=hand_l, loc=<2.2074 -0.3625 1.3638>, rot=<-0.0284 0.0722 -0.0458 -0.9959>}
 ring_01_l {parent=ring_metacarpal_l, loc=<7.0295 -0.6197 -0.1446>, rot=<-0.0489 -0.0252 0.1892 -0.9804>}
 ring_02_l {parent=ring_01_l, loc=<3.6174 0.4647 -0.1027>, rot=<0 0 0.2806 -0.9598>}
 ring_03_l {parent=ring_02_l, loc=<2.4152 0.114 0.0199>, rot=<0 0 0.1189 -0.9929>}
 pinky_metacarpal_l {parent=hand_l, loc=<1.7616 -0.9753 3.3814>, rot=<-0.0397 0.0814 -0.069 -0.9935>}
 pinky_01_l {parent=pinky_metacarpal_l, loc=<7.1709 -0.4831 -0.1094>, rot=<-0.0654 -0.0035 0.2242 -0.9723>}
 pinky_02_l {parent=pinky_01_l, loc=<2.8274 0.0893 -0.1043>, rot=<0.0037 0.0011 0.2854 -0.9584>}
 pinky_03_l {parent=pinky_02_l, loc=<2.2046 0.0307 -0.0078>, rot=<-0.0002 0.0098 0.1445 -0.9895>}
 deform_wrist_dn_base_l {parent=hand_l, loc=<0 -1.5326 0>, rot=<0 0 0 -1>}
 deform_wrist_dn_l {parent=deform_wrist_dn_base_l, loc=<0 0 0>, rot=<0 0 0 1>}
 deform_wrist_up_base_l {parent=hand_l, loc=<0 2.3009 0>, rot=<0 0 0 -1>}
 deform_wrist_up_l {parent=deform_wrist_up_base_l, loc=<0 0 0>, rot=<0 0 0 1>}
 lowerarm_twist_02_l {parent=lowerarm_l, loc=<10.6707 0 0>, rot=<0 0 0 1>}
 lowerarm_twist_01_l {parent=lowerarm_l, loc=<18.316 0 0>, rot=<0 0 0 1>}
 deform_elbow_in_base_l {parent=lowerarm_l, loc=<0.1834 3.6062 -0.2572>, rot=<0 0 0 -1>}
 deform_elbow_in_l {parent=deform_elbow_in_base_l, loc=<0 0 0>, rot=<0 0 0 1>}
 deform_elbow_out_base_l {parent=lowerarm_l, loc=<0 -4 0>, rot=<0 0 0 1>}
 deform_elbow_out_l {parent=deform_elbow_out_base_l, loc=<0 0 0>, rot=<0 0 0 1>}
 hand_l_null {parent=lowerarm_l, type=Null, loc=<22.7355 -0.3203 -0.6105>, rot=<-0.6784 -0.0223 0.0735 -0.7307>}
 upperarm_twist_01_l {parent=upperarm_l, loc=<7.2135 -0.1134 0>, rot=<0 0 0 1>}
 deform_upperarm_bk_base_l {parent=upperarm_twist_01_l, loc=<-6.4053 -4.3746 0.314>, rot=<0 0 0 -1>}
 deform_upperarm_bk_l {parent=deform_upperarm_bk_base_l, loc=<0 0 0>, rot=<0 0 0 1>}
 deform_upperarm_fr_base_l {parent=upperarm_twist_01_l, loc=<-5.6312 5.092 0.3354>, rot=<0 0 0 -1>}
 deform_upperarm_fr_l {parent=deform_upperarm_fr_base_l, loc=<0 0 0>, rot=<0 0 0 1>}
 deform_upperarm_up_base_l {parent=upperarm_twist_01_l, loc=<-8.891 0.4825 3.9517>, rot=<0 0 0 -1>}
 deform_upperarm_up_l {parent=deform_upperarm_up_base_l, loc=<0 0 0>, rot=<0 0 0 1>}
 upperarm_twist_02_l {parent=upperarm_l, loc=<19.7998 -0.1134 0>, rot=<0 0 0 1>}
 upperarm_l_null {parent=clavicle_l, type=Null, loc=<9.8658 -1.2136 -0.4105>, rot=<0.0164 -0.3245 -0.0269 -0.9454>}
 clavicle_r {parent=spine_05, loc=<5.4831 -0.1053 3.2946>, rot=<0.7289 0.0465 -0.6829 0.009>}
 upperarm_r {parent=clavicle_r, loc=<-9.8658 1.2136 0.4105>, rot=<0.0164 -0.3245 -0.0269 -0.9454>}
 lowerarm_r {parent=upperarm_r, loc=<-26.465 0.8345 -0.259>, rot=<0.0036 0.0161 -0.2298 -0.9731>}
 hand_r {parent=lowerarm_r, loc=<-22.7355 0.3203 0.6106>, rot=<-0.6784 -0.0223 0.0735 -0.7307>}
 thumb_01_r {parent=hand_r, loc=<-1.3118 0.3191 2.459>, rot=<0.6546 -0.4042 0.0664 -0.6354>}
 thumb_02_r {parent=thumb_01_r, loc=<-5.9336 0.0448 -0.0401>, rot=<-0.0037 0.0213 0.298 -0.9543>}
 thumb_03_r {parent=thumb_02_r, loc=<-2.2223 -0.3682 0.0907>, rot=<0.0118 -0.013 0.0066 -0.9998>}
 index_metacarpal_r {parent=hand_r, loc=<-1.6535 -0.0183 2.96>, rot=<0.0158 -0.0124 -0.0173 -0.9996>}
 index_01_r {parent=index_metacarpal_r, loc=<-8.3008 -0.0357 0.0524>, rot=<-0.0044 0.0485 0.0924 -0.9945>}
 index_02_r {parent=index_01_r, loc=<-3.3587 0.3004 -0.0452>, rot=<-0.002 0.0027 0.3225 -0.9466>}
 index_03_r {parent=index_02_r, loc=<-2.5578 0.123 0.0252>, rot=<-0.002 0.0067 0.2558 -0.9667>}
 middle_metacarpal_r {parent=hand_r, loc=<-2.1856 0.1725 0.5984>, rot=<0.006 0.0168 -0.0329 -0.9993>}
 middle_01_r {parent=middle_metacarpal_r, loc=<-7.3275 0.2447 -0.0129>, rot=<-0.0208 0.0042 0.0858 -0.9961>}
 middle_02_r {parent=middle_01_r, loc=<-4.0123 -0.1401 -0.1175>, rot=<-0.0218 0.0258 0.3512 -0.9357>}
 middle_03_r {parent=middle_02_r, loc=<-2.6077 -0.0534 -0.1373>, rot=<0.0082 0.0188 0.1708 -0.9851>}
 ring_metacarpal_r {parent=hand_r, loc=<-2.2074 0.3625 -1.3638>, rot=<-0.0284 0.0722 -0.0458 -0.9959>}
 ring_01_r {parent=ring_metacarpal_r, loc=<-7.0295 0.6197 0.1446>, rot=<-0.0489 -0.0252 0.1892 -0.9804>}
 ring_02_r {parent=ring_01_r, loc=<-3.6174 -0.4647 0.1027>, rot=<0 0 0.2806 -0.9598>}
 ring_03_r {parent=ring_02_r, loc=<-2.4152 -0.114 -0.0199>, rot=<0 0 0.1189 -0.9929>}
 pinky_metacarpal_r {parent=hand_r, loc=<-1.7616 0.9753 -3.3814>, rot=<-0.0397 0.0814 -0.069 -0.9935>}
 pinky_01_r {parent=pinky_metacarpal_r, loc=<-7.1709 0.4831 0.1094>, rot=<-0.0654 -0.0035 0.2242 -0.9723>}
 pinky_02_r {parent=pinky_01_r, loc=<-2.8274 -0.0893 0.1043>, rot=<0.0037 0.0011 0.2854 -0.9584>}
 pinky_03_r {parent=pinky_02_r, loc=<-2.2046 -0.0307 0.0078>, rot=<-0.0002 0.0098 0.1445 -0.9895>}
 deform_wrist_dn_base_r {parent=hand_r, loc=<-0.0001 1.5328 0>, rot=<0 0 0 -1>}
 deform_wrist_dn_r {parent=deform_wrist_dn_base_r, loc=<0 0 0>, rot=<0 0 0 1>}
 deform_wrist_up_base_r {parent=hand_r, loc=<-0.0002 -2.3007 -0.0001>, rot=<0 0 0 -1>}
 deform_wrist_up_r {parent=deform_wrist_up_base_r, loc=<0 0 0>, rot=<0 0 0 1>}
 lowerarm_twist_02_r {parent=lowerarm_r, loc=<-10.6707 0 0>, rot=<0 0 0 -1>}
 lowerarm_twist_01_r {parent=lowerarm_r, loc=<-18.316 0 0>, rot=<0 0 0 -1>}
 deform_elbow_in_base_r {parent=lowerarm_r, loc=<-0.1834 -3.6062 0.2572>, rot=<0 0 0 -1>}
 deform_elbow_in_r {parent=deform_elbow_in_base_r, loc=<0 0 0>, rot=<0 0 0 1>}
 deform_elbow_out_base_r {parent=lowerarm_r, loc=<0 4 -0.0001>, rot=<0 0 0 -1>}
 deform_elbow_out_r {parent=deform_elbow_out_base_r, loc=<0 0 0>, rot=<0 0 0 1>}
 hand_r_null {parent=lowerarm_r, type=Null, loc=<-22.7355 0.3203 0.6106>, rot=<-0.6784 -0.0223 0.0735 -0.7307>}
 upperarm_twist_01_r {parent=upperarm_r, loc=<-7.2135 0.1134 0>, rot=<0 0 0 -1>}
 deform_upperarm_bk_base_r {parent=upperarm_twist_01_r, loc=<6.4053 4.3746 -0.314>, rot=<0 0 0 -1>}
 deform_upperarm_bk_r {parent=deform_upperarm_bk_base_r, loc=<0 0 0>, rot=<0 0 0 1>}
 deform_upperarm_fr_base_r {parent=upperarm_twist_01_r, loc=<5.6312 -5.092 -0.3354>, rot=<0 0 0 -1>}
 deform_upperarm_fr_r {parent=deform_upperarm_fr_base_r, loc=<0 0 0>, rot=<0 0 0 1>}
 deform_upperarm_up_base_r {parent=upperarm_twist_01_r, loc=<8.891 -0.4825 -3.9517>, rot=<0 0 0 -1>}
 deform_upperarm_up_r {parent=deform_upperarm_up_base_r, loc=<0 0 0>, rot=<0 0 0 1>}
 upperarm_twist_02_r {parent=upperarm_r, loc=<-19.7998 0.1134 0>, rot=<0 0 0 -1>}
 base_armourPadArm_1 {parent=upperarm_r, loc=<-8.9481 0.9993 -4.0919>, rot=<-0.0125 -0.0075 0.9999 0.0038>}
 dyn_armourPadArm_1 {parent=base_armourPadArm_1, loc=<0 0 0>, rot=<0 0 0 1>}
 dyn_armourPadArm_2 {parent=dyn_armourPadArm_1, loc=<3.7983 0 0>, rot=<0 0 0 1>}
 base_armourPadArm_bend {parent=dyn_armourPadArm_1, loc=<-1.1973 -0.0163 -1.8998>, rot=<0.0024 -1 0 -0.0048>}
 dyn_armourPadArm_bend {parent=base_armourPadArm_bend, loc=<0 0 0>, rot=<0 0 0 1>}
 upperarm_r_null {parent=clavicle_r, type=Null, loc=<-9.8658 1.2136 0.4105>, rot=<0.0164 -0.3245 -0.0269 -0.9454>}
 clavicle_r_null {parent=spine_05, type=Null, loc=<5.4831 -0.1053 3.2946>, rot=<0.7289 0.0465 -0.6829 0.009>}
 deform_pec_aim_base_l {parent=spine_04, loc=<3.2456 9.3309 -1.2237>, rot=<0.0603 -0.4033 0.3318 -0.8507>}
 deform_pec_aim_l {parent=deform_pec_aim_base_l, loc=<0 0 0>, rot=<0 0 0 1>}
 deform_pec_base_l {parent=deform_pec_aim_l, loc=<9.3665 0 0>, rot=<0 0 0 1>}
 deform_pec_l {parent=deform_pec_base_l, loc=<0 0 0>, rot=<0 0 0 1>}
 deform_pec_aim_base_r {parent=spine_04, loc=<3.2452 9.3309 1.2237>, rot=<-0.4033 -0.0603 0.8507 0.3318>}
 deform_pec_aim_r {parent=deform_pec_aim_base_r, loc=<0 0 0>, rot=<0 0 0 1>}
 deform_pec_base_r {parent=deform_pec_aim_r, loc=<-9.3667 -0.0003 -0.0004>, rot=<0 0 0 1>}
 deform_pec_r {parent=deform_pec_base_r, loc=<0 0 0>, rot=<0 0 0 1>}
 base_chestCard_r {parent=spine_04, loc=<6.1103 11.7099 7.5352>, rot=<0.0368 0.5605 0.8245 0.0682>}
 dyn_chestCard_r {parent=base_chestCard_r, loc=<0 0 0>, rot=<0 0 0 1>}
 base_chestCard_c {parent=spine_04, loc=<6.3763 12.9106 0.002>, rot=<0.0648 0.7041 0.7041 0.0648>}
 dyn_chestCard_c {parent=base_chestCard_c, loc=<0 0 0>, rot=<0 0 0 1>}
 base_chestCard_l {parent=spine_04, loc=<6.1104 11.71 -7.5352>, rot=<-0.0682 -0.8245 -0.5605 -0.0368>}
 dyn_chestCard_l {parent=base_chestCard_l, loc=<0 0 0>, rot=<0 0 0 1>}
 base_belt_1 {parent=spine_01, loc=<-1.2092 2.4126 -12.6143>, rot=<-0.6651 -0.7099 -0.2316 -0.0077>}
 dyn_belt_1 {parent=base_belt_1, loc=<0 0 0>, rot=<0 0 0 1>}
 base_beltPad {parent=dyn_belt_1, loc=<-0.0683 3.4579 0.0595>, rot=<0 0 0.7071 -0.7071>}
 dyn_beltPad {parent=base_beltPad, loc=<0 0 0>, rot=<0 0 0 1>}
 base_nade_1 {parent=dyn_beltPad, loc=<2.4598 -2.6156 0.0798>, rot=<-0.1485 0.0053 0.0323 -0.9884>}
 dyn_nade_1 {parent=base_nade_1, loc=<0 0 0>, rot=<0 0 0 1>}
 base_nade_2 {parent=dyn_beltPad, loc=<2.4979 2.9763 0.4274>, rot=<0.1339 -0.0196 -0.0272 -0.9904>}
 dyn_nade_2 {parent=base_nade_2, loc=<0 0 0>, rot=<0 0 0 1>}
 dyn_hangingBelt_1 {parent=dyn_beltPad, loc=<15.4506 2.4305 -0.8496>, rot=<0.0767 -0.0166 -0.0353 -0.9963>}
 dyn_hangingBelt_2 {parent=dyn_hangingBelt_1, loc=<5.2928 0 0>, rot=<-0.0174 -0.0446 0.0103 -0.9988>}
 dyn_hangingBelt_3 {parent=dyn_hangingBelt_2, loc=<5.6733 0 0>, rot=<0.0305 -0.0401 0.0098 -0.9987>}
 dyn_beltHangCard {parent=dyn_hangingBelt_3, loc=<4.9037 0 0.4956>, rot=<0 -0.0293 0 -0.9996>}
 base_belt_2 {parent=spine_01, loc=<0.2642 10.0935 -0.9071>, rot=<0.5652 0.4689 0.4696 -0.4901>}
 dyn_belt_2 {parent=base_belt_2, loc=<0 0 0>, rot=<0 0 0 1>}
 base_belt_3 {parent=spine_01, loc=<2.8108 0.3288 11.759>, rot=<0.1274 -0.1886 0.6616 -0.7145>}
 dyn_belt_3 {parent=base_belt_3, loc=<0 0 0>, rot=<0 0 0 1>}
 base_armourPadLeg_1 {parent=dyn_belt_3, loc=<0.0073 -2.3662 0.9119>, rot=<0.6942 -0.7183 0.0054 -0.0458>}
 dyn_armourPadLeg_1 {parent=base_armourPadLeg_1, loc=<0 0 0>, rot=<0 0 0 1>}
 dyn_armourPadLeg_2 {parent=dyn_armourPadLeg_1, loc=<5.4565 0 0>, rot=<0 0.0475 -0.0127 -0.9988>}
 dyn_armourPadLeg_3 {parent=dyn_armourPadLeg_2, loc=<9.7768 0 0>, rot=<0 0 0 1>}
 dyn_armourPadLeg_0 {parent=dyn_armourPadLeg_1, loc=<-1.6933 -0.1694 -0.4516>, rot=<-0.0059 0 0 -1>}
 base_belt_4 {parent=spine_01, loc=<1.0174 -6.0358 -8.9734>, rot=<-0.4925 -0.7266 0.1153 -0.465>}
 dyn_belt_4 {parent=base_belt_4, loc=<0 0 0>, rot=<0 0 0 1>}
 base_beltHipCard_2 {parent=dyn_belt_4, loc=<0.1059 -1.7767 0.744>, rot=<-0.0739 -0.0991 0.6756 -0.7268>}
 dyn_beltHipCard_2 {parent=base_beltHipCard_2, loc=<0 0 0>, rot=<0 0 0 1>}
 base_belt_5 {parent=spine_01, loc=<2.308 -8.1865 -0.0036>, rot=<-0.2979 -0.6551 0.2904 -0.6307>}
 dyn_belt_5 {parent=base_belt_5, loc=<0 0 0>, rot=<0 0 0 1>}
 base_belt_6 {parent=spine_01, loc=<1.8496 7.562 7.4258>, rot=<0.4047 0.268 0.5851 -0.6497>}
 dyn_belt_6 {parent=base_belt_6, loc=<0 0 0>, rot=<0 0 0 1>}
 base_beltHipCard_1 {parent=dyn_belt_6, loc=<0.9148 -1.6195 0.5605>, rot=<-0.0248 0.0203 0.7251 -0.6879>}
 dyn_beltHipCard_1 {parent=base_beltHipCard_1, loc=<0 0 0>, rot=<0 0 0 1>}
 base_belt_7 {parent=spine_01, loc=<2.6612 -5.4062 8.017>, rot=<-0.1124 -0.4808 0.5329 -0.6872>}
 dyn_belt_7 {parent=base_belt_7, loc=<0 0 0>, rot=<0 0 0 1>}
 thigh_l {parent=pelvis, loc=<-1.0855 -0.2109 -8.7053>, rot=<0.0009 0.0355 0.0155 -0.9993>}
 calf_l {parent=thigh_l, loc=<-42.5284 -0.0019 -0.7297>, rot=<0.0006 -0.0161 -0.0391 -0.9991>}
 foot_l {parent=calf_l, loc=<-46.7401 -2.5014 -0.592>, rot=<0.0427 -0.0238 0.0353 -0.9982>}
 ball_l {parent=foot_l, loc=<-3.5658 10.8066 -0.1544>, rot=<0.0001 -0.0001 -0.7063 -0.7079>}
 base_shoe_handle_l {parent=foot_l, loc=<2.7898 -3.1409 0.5677>, rot=<-0.0437 -0.0124 0.3324 -0.942>}
 dyn_shoe_handle_l {parent=base_shoe_handle_l, loc=<0 0 0>, rot=<0 0 0 1>}
 stk_tongueShoe_l {parent=foot_l, loc=<-0.9706 8.4902 0.3045>, rot=<0.0151 0.0146 0.3498 -0.9366>}
 calf_twist_02_l {parent=calf_l, loc=<-19.1546 -1.7943 -0.2638>, rot=<0.0237 0.0095 -0.0035 -0.9997>}
 calf_twist_01_l {parent=calf_l, loc=<-35.8483 -1.9197 -0.5794>, rot=<0.0237 0.0095 -0.0035 -0.9997>}
 calf_twist_00_l {parent=calf_l, loc=<0 0 0>, rot=<0 0 0 1>}
 deform_knee_in_base_l {parent=calf_l, loc=<0.8031 -5.0231 0.2456>, rot=<0 0 0 -1>}
 deform_knee_in_l {parent=deform_knee_in_base_l, loc=<0 0 0>, rot=<0 0 0 1>}
 deform_knee_out_base_l {parent=calf_l, loc=<0 4 0>, rot=<0 0 0 -1>}
 deform_knee_out_l {parent=deform_knee_out_base_l, loc=<0 0 0>, rot=<0 0 0 1>}
 foot_l_null {parent=calf_l, type=Null, loc=<-46.7401 -2.5014 -0.592>, rot=<0.0427 -0.0238 0.0353 -0.9982>}
 thigh_twist_01_l {parent=thigh_l, loc=<-27.952 0 0>, rot=<0.0013 0 0 -1>}
 deform_groin_base_l {parent=thigh_l, loc=<0.4801 5.6776 -0.2486>, rot=<0 0 0 -1>}
 deform_groin_l {parent=deform_groin_base_l, loc=<0 0 0>, rot=<0 0 0 1>}
 calf_l_null {parent=thigh_l, type=Null, loc=<-42.5284 -0.0019 -0.7297>, rot=<0.0006 -0.0161 -0.0391 -0.9991>}
 thigh_r {parent=pelvis, loc=<-1.0855 -0.2109 8.7053>, rot=<0.0355 -0.0009 0.9993 0.0155>}
 calf_r {parent=thigh_r, loc=<42.5284 0.0019 0.7297>, rot=<0.0006 -0.0161 -0.0391 -0.9991>}
 foot_r {parent=calf_r, loc=<46.7401 2.5014 0.592>, rot=<0.0427 -0.0238 0.0353 -0.9982>}
 ball_r {parent=foot_r, loc=<3.5658 -10.8066 0.1544>, rot=<0.0001 -0.0001 -0.7063 -0.7079>}
 base_shoe_handle_r {parent=foot_r, loc=<-2.7898 3.1409 -0.5677>, rot=<0.0124 -0.0437 0.942 0.3324>}
 dyn_shoe_handle_r {parent=base_shoe_handle_r, loc=<0 0 0>, rot=<0 0 0 1>}
 stk_tongueShoe_r {parent=foot_r, loc=<0.9706 -8.4902 -0.3044>, rot=<-0.0146 0.0151 0.9366 0.3498>}
 calf_twist_02_r {parent=calf_r, loc=<19.1546 1.7943 0.2638>, rot=<0.0237 0.0095 -0.0035 -0.9997>}
 calf_twist_01_r {parent=calf_r, loc=<35.8483 1.9197 0.5794>, rot=<0.0237 0.0095 -0.0035 -0.9997>}
 calf_twist_00_r {parent=calf_r, loc=<0 0 0>, rot=<0 0 0 1>}
 deform_knee_in_base_r {parent=calf_r, loc=<-0.8031 5.0231 -0.2457>, rot=<0 0 0 -1>}
 deform_knee_in_r {parent=deform_knee_in_base_r, loc=<0 0 0>, rot=<0 0 0 1>}
 deform_knee_out_base_r {parent=calf_r, loc=<0 -4 0>, rot=<0 0 0 -1>}
 deform_knee_out_r {parent=deform_knee_out_base_r, loc=<0 0 0>, rot=<0 0 0 1>}
 foot_r_null {parent=calf_r, type=Null, loc=<46.7401 2.5014 0.592>, rot=<0.0427 -0.0238 0.0353 -0.9982>}
 thigh_twist_01_r {parent=thigh_r, loc=<27.952 0 0>, rot=<0.0013 0 0 -1>}
 deform_groin_base_r {parent=thigh_r, loc=<-0.4801 -5.6776 0.2486>, rot=<0 0 0 -1>}
 deform_groin_r {parent=deform_groin_base_r, loc=<0 0 0>, rot=<0 0 0 1>}
 calf_r_null {parent=thigh_r, type=Null, loc=<42.5284 0.0019 0.7297>, rot=<0.0006 -0.0161 -0.0391 -0.9991>}
 deform_glute_base_l {parent=pelvis, loc=<-7.3466 -9.0213 -9.1245>, rot=<0.0172 0.031 -0.4477 -0.8935>}
 deform_glute_l {parent=deform_glute_base_l, loc=<0 0 0>, rot=<0 0 0 1>}
 deform_glute_base_r {parent=pelvis, loc=<-7.3464 -9.0213 9.1245>, rot=<0.031 -0.0172 0.8935 -0.4477>}
 deform_glute_r {parent=deform_glute_base_r, loc=<0 0 0>, rot=<0 0 0 1>}
 base_smallBelt_3 {parent=pelvis, loc=<1.8492 -10.1165 8.2278>, rot=<0.0846 -0.5243 0.8042 -0.2669>}
 dyn_smallBelt_3 {parent=base_smallBelt_3, loc=<0 0 0>, rot=<0 0 0 1>}
 base_smallBelt_2 {parent=pelvis, loc=<-0.6657 -1.7135 14.7101>, rot=<0.1117 -0.0675 0.9914 0.0121>}
 dyn_smallBelt_2 {parent=base_smallBelt_2, loc=<0 0 0>, rot=<0 0 0 1>}
 base_smallBelt_1 {parent=pelvis, loc=<1.9494 7.3508 8.3732>, rot=<0.0182 0.5345 0.8256 0.1798>}
 dyn_smallBelt_1 {parent=base_smallBelt_1, loc=<0 0 0>, rot=<0 0 0 1>}
 thigh_l_null {parent=pelvis, type=Null, loc=<-1.0855 -0.2109 -8.7053>, rot=<0.0009 0.0355 0.0155 -0.9993>}
 thigh_r_null {parent=pelvis, type=Null, loc=<-1.0855 -0.2109 8.7053>, rot=<0.0355 -0.0009 0.9993 0.0155>}
 pelvis_ctrl {controlType=EULER_TRANSFORM, type=Control, loc=<0 0 0>, rot=<0 0 0 1>}
 spine_01_ctrl {controlType=EULER_TRANSFORM, parent=pelvis_ctrl, type=Control, loc=<0 0 0>, rot=<0 0 0 1>}
 spine_02_ctrl {controlType=EULER_TRANSFORM, parent=spine_01_ctrl, type=Control, loc=<0 0 0>, rot=<0 0 0 1>}
 spine_03_ctrl {controlType=EULER_TRANSFORM, parent=spine_02_ctrl, type=Control, loc=<0 0 0>, rot=<0 0 0 1>}
 spine_04_ctrl {controlType=EULER_TRANSFORM, parent=spine_03_ctrl, type=Control, loc=<0 0 0>, rot=<0 0 0 1>}
 spine_05_ctrl {controlType=EULER_TRANSFORM, parent=spine_04_ctrl, type=Control, loc=<0 0 0>, rot=<0 0 0 1>}
 neck_01_ctrl {controlType=EULER_TRANSFORM, parent=spine_05_ctrl, type=Control, loc=<0 0 0>, rot=<0 0 0 1>}
 neck_02_ctrl {controlType=EULER_TRANSFORM, parent=neck_01_ctrl, type=Control, loc=<0 0 0>, rot=<0 0 0 1>}
 head_ctrl {controlType=EULER_TRANSFORM, parent=neck_02_ctrl, type=Control, loc=<0 0 0>, rot=<0 0 0 1>}
 clavicle_l_ctrl {controlType=EULER_TRANSFORM, parent=spine_05_ctrl, type=Control, loc=<0 0 0>, rot=<0 0 0 1>}
 upperarm_l_ctrl {controlType=EULER_TRANSFORM, parent=clavicle_l_ctrl, type=Control, loc=<0 0 0>, rot=<0 0 0 1>}
 lowerarm_l_ctrl {controlType=EULER_TRANSFORM, parent=upperarm_l_ctrl, type=Control, loc=<0 0 0>, rot=<0 0 0 1>}
 hand_l_ctrl {controlType=EULER_TRANSFORM, parent=lowerarm_l_ctrl, type=Control, loc=<0 0 0>, rot=<0 0 0 1>}
 clavicle_r_ctrl {controlType=EULER_TRANSFORM, parent=spine_05_ctrl, type=Control, loc=<0 0 0>, rot=<0 0 0 1>}
 upperarm_r_ctrl {controlType=EULER_TRANSFORM, parent=clavicle_r_ctrl, type=Control, loc=<0 0 0>, rot=<0 0 0 1>}
 lowerarm_r_ctrl {controlType=EULER_TRANSFORM, parent=upperarm_r_ctrl, type=Control, loc=<0 0 0>, rot=<0 0 0 1>}
 hand_r_ctrl {controlType=EULER_TRANSFORM, parent=lowerarm_r_ctrl, type=Control, loc=<0 0 0>, rot=<0 0 0 1>}
 thigh_l_ctrl {controlType=EULER_TRANSFORM, parent=pelvis_ctrl, type=Control, loc=<0 0 0>, rot=<0 0 0 1>}
 calf_l_ctrl {controlType=EULER_TRANSFORM, parent=thigh_l_ctrl, type=Control, loc=<0 0 0>, rot=<0 0 0 1>}
 foot_l_ctrl {controlType=EULER_TRANSFORM, parent=calf_l_ctrl, type=Control, loc=<0 0 0>, rot=<0 0 0 1>}
 thigh_r_ctrl {controlType=EULER_TRANSFORM, parent=pelvis_ctrl, type=Control, loc=<0 0 0>, rot=<0 0 0 1>}
 calf_r_ctrl {controlType=EULER_TRANSFORM, parent=thigh_r_ctrl, type=Control, loc=<0 0 0>, rot=<0 0 0 1>}
 foot_r_ctrl {controlType=EULER_TRANSFORM, parent=calf_r_ctrl, type=Control, loc=<0 0 0>, rot=<0 0 0 1>}
 r_thigh_rz_up {type=Curve, loc=<0 0 0>, rot=<0 0 0 1>}
]
variables [
 pub isPlayingEasterBloc {type=bool, default=False}
 pub isPlayingEmote {type=bool, default=False}
 pub isPlayingMeeleSwing {type=bool, default=False}
 pub isStormAviator {type=bool, default=False}
 r_thigh_ryn_range_10_80 {type=double, default=0}
 r_thigh_rzn100 {type=double, default=0}
]
