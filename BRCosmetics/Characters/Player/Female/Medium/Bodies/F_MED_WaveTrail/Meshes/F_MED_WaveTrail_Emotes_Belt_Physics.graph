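# Physics asset for the hanging belt on the WaveTrail female medium body's emotes variant, covering belt and left thigh bones.
# A constraint chain links the belt pad through three capsule belt segments to a box-shaped hanging card.
# Each joint shares the same swing limits with zero twist limit.
# Three stacked capsules on the left thigh form an unconstrained body, apparently a collider for the swinging belt.
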PhysicsAsset /BRCosmetics/Characters/Player/Female/Medium/Bodies/F_MED_WaveTrail/Meshes/F_MED_WaveTrail_Emotes_Belt_Physics
bodies {
 dyn_beltHangCard [{
   shapeType=Box
   extentX=7.5
   extentY=5.0847
   extentZ=0.6559
   shapeName=dyn_beltHangCard_box
   center=<4.2079 0.1904 0.0688>
   rotation=<rot 0 1.897 0>
  }]
 dyn_beltPad [{shapeType=Capsule, length=2, radius=1, shapeName=dyn_beltPad_capsule}]
 dyn_hangingBelt_1 [{shapeType=Capsule, length=4.4604, radius=1.5396, shapeName=dyn_hangingBelt_1_capsule, center=<2.9458 0 0.6247>, rotation=<rot 90 0 0>}]
 dyn_hangingBelt_2 [{shapeType=Capsule, length=4.4604, radius=1.5396, shapeName=dyn_hangingBelt_2_capsule, center=<2.9458 0 0.9726>, rotation=<rot 90 0 0>}]
 dyn_hangingBelt_3 [{shapeType=Capsule, length=2.9775, radius=1.851, shapeName=dyn_hangingBelt_3_capsule, center=<2.9458 0 1.1363>, rotation=<rot 90 0 0>}]
 thigh_l [
  {
   shapeType=Capsule
   length=22.0558
   radius=10.9158
   shapeName=thigh_l_capsuleBot
   center=<-33.4292 -1.4133 0.5986>
   rotation=<rot 85.0978 179.9998 179.9999>
  }
  {
   shapeType=Capsule
   length=9.954
   radius=11.7419
   shapeName=thigh_l_capsuleMid
   center=<-14.9253 -1.3498 2.7711>
   rotation=<rot 79.8724 165.6567 169.0825>
  }
  {
   shapeType=Capsule
   length=10.9251
   radius=11.3934
   shapeName=thigh_l_capsuleTop
   center=<-5.3341 -1.8957 4.4551>
   rotation=<rot 79.8723 165.6567 169.0825>
  }
 ]
}
constraints [
 {bone1Name=dyn_beltHangCard, bone2Name=dyn_hangingBelt_3, swing1LimitDegrees=30, swing2LimitDegrees=20, twistLimitDegrees=0}
 {bone1Name=dyn_hangingBelt_2, bone2Name=dyn_hangingBelt_1, swing1LimitDegrees=30, swing2LimitDegrees=20, twistLimitDegrees=0}
 {bone1Name=dyn_hangingBelt_3, bone2Name=dyn_hangingBelt_2, swing1LimitDegrees=30, swing2LimitDegrees=20, twistLimitDegrees=0}
 {bone1Name=dyn_hangingBelt_1, bone2Name=dyn_beltPad, swing1LimitDegrees=30, swing2LimitDegrees=20, twistLimitDegrees=0}
]
