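# Physics asset for the WaveTrail female medium body cosmetic, setting up secondary motion for its belt and leg armour pad.
# Two limited-constraint chains hang from small base bodies on spine_01: belt pad through three hanging belt links to a card, and a branching armour pad built mostly from boxes.
# Pelvis, spine, both thighs and right calf carry capsules with free angular constraints, providing body shapes around the dangling parts.
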PhysicsAsset /BRCosmetics/Characters/Player/Female/Medium/Bodies/F_MED_WaveTrail/Meshes/F_MED_WaveTrail_Physics
bodies {
 base_armourPadLeg_1 [{shapeType=Capsule, length=2, radius=1, shapeName=base_beltPad_capsule}]
 base_beltPad [{shapeType=Capsule, length=2, radius=1, shapeName=base_beltPad_capsule}]
 calf_r [{shapeType=Capsule, length=30.2901, radius=6.7181, shapeName=calf_r_capsule, center=<25.3268 1.9259 0.817>, rotation=<rot 0 -90 90>}]
 dyn_armourPadLeg_0 [{shapeType=Capsule, length=2, radius=1, shapeName=base_beltPad_capsule, center=<0.4759 0 -2.4407>, rotation=<rot 90 0 0>}]
 dyn_armourPadLeg_1 [{
   shapeType=Box
   extentX=16.9602
   extentY=3.4389
   extentZ=14.0166
   shapeName=dyn_armourPadLeg_1_box
   center=<5.5673 0.0781 -1.5521>
   rotation=<rot 0.0001 -91.09 -88.2906>
  }]
 dyn_armourPadLeg_2 [{
   shapeType=Box
   extentX=12.2059
   extentY=3.5465
   extentZ=14.6861
   shapeName=dyn_armourPadLeg_2_box
   center=<5.9869 0.0165 -1.2215>
   rotation=<rot -7.3385 0.0001 90>
  }]
 dyn_armourPadLeg_3 [{
   shapeType=Box
   extentX=11.289
   extentY=4.9976
   extentZ=15.2091
   shapeName=dyn_armourPadLeg_3_box
   center=<5.1764 0.0151 -1.8883>
   rotation=<rot -6.905 0.0001 90>
  }]
 dyn_beltHangCard [{
   shapeType=Box
   extentX=7.5
   extentY=5.0847
   extentZ=0.6559
   shapeName=dyn_beltHangCard_box
   center=<4.2079 0.1904 0.0688>
   rotation=<rot 0 1.897 0>
  }]
 dyn_beltPad [
  {
   shapeType=Capsule
   length=12.9886
   radius=1.3044
   shapeName=dyn_beltPad_capsuleA
   center=<6.5112 5.7576 -0.5847>
   rotation=<rot 0.2739 93.3741 -95.5687>
  }
  {
   shapeType=Capsule
   length=13.0611
   radius=1.6176
   shapeName=dyn_beltPad_capsuleB
   center=<8.0521 0.8494 1.1961>
   rotation=<rot 0.4812 91.751 -90.7808>
  }
  {
   shapeType=Capsule
   length=12.3929
   radius=1.6518
   shapeName=dyn_beltPad_capsuleC
   center=<8.3738 -5.7422 -0.8905>
   rotation=<rot 22.3007 83.1728 -94.1545>
  }
  {
   shapeType=Capsule
   length=4.8384
   radius=0.8191
   shapeName=dyn_beltPad_capsuleBackTop
   center=<0.9928 -5.2394 -1.4929>
   rotation=<rot 22.3008 83.1729 -90.0828>
  }
 ]
 dyn_hangingBelt_1 [{shapeType=Capsule, length=4.4604, radius=1.5396, shapeName=dyn_hangingBelt_1_capsule, center=<2.9458 0 0.6247>, rotation=<rot 90 0 0>}]
 dyn_hangingBelt_2 [{shapeType=Capsule, length=4.4604, radius=1.5396, shapeName=dyn_hangingBelt_2_capsule, center=<2.9458 0 0.9726>, rotation=<rot 90 0 0>}]
 dyn_hangingBelt_3 [{shapeType=Capsule, length=2.9775, radius=1.851, shapeName=dyn_hangingBelt_3_capsule, center=<2.9458 0 1.1363>, rotation=<rot 90 0 0>}]
 pelvis [{shapeType=Capsule, length=9.1225, radius=9.3529, shapeName=pelvis_capsule, rotation=<rot 0 170.0516 -175.3657>}]
 spine_01 [{
   shapeType=Capsule
   length=4.8449
   radius=10.8957
   shapeName=spine_01_capsule
   center=<0.0326 -0.3618 1.4519>
   rotation=<rot -1.8553 -90 1.2763>
  }]
 thigh_l [
  {
   shapeType=Capsule
   length=22.0558
   radius=10.9158
   shapeName=thigh_l_capsuleBot
   center=<-33.4292 -1.4133 0.5986>
   rotation=<rot 85.0978 179.9998 179.9999>
  }
  {
   shapeType=Capsule
   length=9.954
   radius=11.7419
   shapeName=thigh_l_capsuleMid
   center=<-14.9253 -1.3498 2.7711>
   rotation=<rot 79.8724 165.6567 169.0825>
  }
  {
   shapeType=Capsule
   length=10.9251
   radius=11.3934
   shapeName=thigh_l_capsuleTop
   center=<-5.3341 -1.8957 4.4551>
   rotation=<rot 79.8723 165.6567 169.0825>
  }
 ]
 thigh_r [
  {
   shapeType=Capsule
   length=23.0796
   radius=10.4464
   shapeName=thigh_l_capsule
   center=<14.8518 0.6172 -2.0283>
   rotation=<rot 82.2947 179.9996 1.7499>
  }
  {
   shapeType=Capsule
   length=14.0357
   radius=10.0813
   shapeName=thigh_r_capsuleSec
   center=<29.828 0.3969 0.7711>
   rotation=<rot 82.2947 179.9996 1.7499>
  }
 ]
}
constraints [
 {bone1Name=dyn_beltHangCard, bone2Name=dyn_hangingBelt_3, swing1LimitDegrees=30, swing2LimitDegrees=20, twistLimitDegrees=0}
 {bone1Name=dyn_hangingBelt_2, bone2Name=dyn_hangingBelt_1, swing1LimitDegrees=30, swing2LimitDegrees=20, twistLimitDegrees=0}
 {bone1Name=dyn_hangingBelt_3, bone2Name=dyn_hangingBelt_2, swing1LimitDegrees=30, swing2LimitDegrees=20, twistLimitDegrees=0}
 {bone1Name=dyn_hangingBelt_1, bone2Name=dyn_beltPad, swing1LimitDegrees=30, swing2LimitDegrees=20, twistLimitDegrees=0}
 {bone1Name=dyn_beltPad, bone2Name=base_beltPad, swing1LimitDegrees=15, swing2LimitDegrees=15, twistLimitDegrees=5}
 {
  bone1Name=base_beltPad
  bone2Name=spine_01
  swing1LimitDegrees=45
  swing1Motion=Free
  swing2LimitDegrees=45
  swing2Motion=Free
  twistLimitDegrees=45
  twistMotion=Free
 }
 {
  bone1Name=thigh_l
  bone2Name=pelvis
  swing1LimitDegrees=45
  swing1Motion=Free
  swing2LimitDegrees=45
  swing2Motion=Free
  twistLimitDegrees=45
  twistMotion=Free
 }
 {
  bone1Name=spine_01
  bone2Name=pelvis
  swing1LimitDegrees=45
  swing1Motion=Free
  swing2LimitDegrees=45
  swing2Motion=Free
  twistLimitDegrees=45
  twistMotion=Free
 }
 {
  bone1Name=thigh_r
  bone2Name=pelvis
  swing1LimitDegrees=45
  swing1Motion=Free
  swing2LimitDegrees=45
  swing2Motion=Free
  twistLimitDegrees=45
  twistMotion=Free
 }
 {bone1Name=dyn_armourPadLeg_3, bone2Name=dyn_armourPadLeg_2, swing1LimitDegrees=0, swing2LimitDegrees=20, twistLimitDegrees=0}
 {
  bone1Name=base_armourPadLeg_1
  bone2Name=spine_01
  swing1LimitDegrees=45
  swing1Motion=Free
  swing2LimitDegrees=45
  swing2Motion=Free
  twistLimitDegrees=45
  twistMotion=Free
 }
 {bone1Name=dyn_armourPadLeg_1, bone2Name=base_armourPadLeg_1, swing1LimitDegrees=20, swing2LimitDegrees=30, twistLimitDegrees=5}
 {bone1Name=dyn_armourPadLeg_0, bone2Name=dyn_armourPadLeg_1, swing1LimitDegrees=0, swing2LimitDegrees=20, twistLimitDegrees=0}
 {bone1Name=dyn_armourPadLeg_2, bone2Name=dyn_armourPadLeg_1, swing1LimitDegrees=0, swing2LimitDegrees=10, twistLimitDegrees=0}
 {
  bone1Name=calf_r
  bone2Name=thigh_r
  swing1LimitDegrees=45
  swing1Motion=Free
  swing2LimitDegrees=45
  swing2Motion=Free
  twistLimitDegrees=45
  twistMotion=Free
 }
]
