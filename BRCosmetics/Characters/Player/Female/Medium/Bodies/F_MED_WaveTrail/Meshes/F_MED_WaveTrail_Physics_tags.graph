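# Physics asset for the WaveTrail female medium cosmetic body, set up for two dangling belt hip cards (the "tags").
# Each box-shaped dyn card hangs from a tiny base anchor body by a constraint with narrow swing limits and zero twist.
# Capsules on the pelvis, right upper thigh and belt pad provide body volumes around the cards.
# The anchors, belt pad and thigh connect to the pelvis through constraints with free angular motion.
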PhysicsAsset /BRCosmetics/Characters/Player/Female/Medium/Bodies/F_MED_WaveTrail/Meshes/F_MED_WaveTrail_Physics_tags
bodies {
 base_beltHipCard_1 [{shapeType=Capsule, length=0.505, radius=0.505, shapeName=base_beltHipCard_1_capsule, rotation=<rot 0 -90 0>}]
 base_beltHipCard_2 [{shapeType=Capsule, length=0.505, radius=0.505, shapeName=base_beltHipCard_2_capsule, rotation=<rot 0 -90 0>}]
 base_beltPad [{
   shapeType=Capsule
   length=12.3929
   radius=2.0153
   shapeName=dyn_beltPad_capsuleC
   center=<7.3534 -5.9105 -0.4109>
   rotation=<rot 22.3007 83.1728 -94.1545>
  }]
 dyn_beltHipCard_1 [{shapeType=Box, extentX=7.6585, extentY=4.6731, extentZ=0.7656, shapeName=dyn_beltHipCard_1_box, center=<4.8073 0 0>}]
 dyn_beltHipCard_2 [{shapeType=Box, extentX=8.883, extentY=5.68, extentZ=0.9658, shapeName=dyn_beltHipCard_2_box, center=<4.9321 0 0>}]
 pelvis [
  {shapeType=Capsule, length=1.9528, radius=9.8323, shapeName=pelvis_capsuleR, center=<0.0503 -2.3107 5.083>, rotation=<rot 0 180 -180>}
  {shapeType=Capsule, length=1.9528, radius=9.8323, shapeName=pelvis_capsuleL, center=<-1.8689 -1.3305 -3.8971>, rotation=<rot 0 180 -180>}
 ]
 thigh_twist_01_r [{
   shapeType=Capsule
   length=24.7347
   radius=9.0284
   shapeName=thigh_l_capsule
   center=<-8.3692 0.588 -0.8494>
   rotation=<rot 86.9779 174.218 177.5401>
  }]
}
constraints [
 {
  bone1Name=base_beltPad
  bone2Name=pelvis
  swing1LimitDegrees=45
  swing1Motion=Free
  swing2LimitDegrees=45
  swing2Motion=Free
  twistLimitDegrees=45
  twistMotion=Free
 }
 {bone1Name=dyn_beltHipCard_2, bone2Name=base_beltHipCard_2, swing1LimitDegrees=10, swing2LimitDegrees=30, twistLimitDegrees=0}
 {
  bone1Name=base_beltHipCard_2
  bone2Name=pelvis
  swing1LimitDegrees=45
  swing1Motion=Free
  swing2LimitDegrees=45
  swing2Motion=Free
  twistLimitDegrees=45
  twistMotion=Free
 }
 {bone1Name=dyn_beltHipCard_1, bone2Name=base_beltHipCard_1, swing1LimitDegrees=10, swing2LimitDegrees=30, twistLimitDegrees=0}
 {
  bone1Name=base_beltHipCard_1
  bone2Name=pelvis
  swing1LimitDegrees=45
  swing1Motion=Free
  swing2LimitDegrees=45
  swing2Motion=Free
  twistLimitDegrees=45
  twistMotion=Free
 }
 {
  bone1Name=thigh_twist_01_r
  bone2Name=pelvis
  swing1LimitDegrees=45
  swing1Motion=Free
  swing2LimitDegrees=45
  swing2Motion=Free
  twistLimitDegrees=45
  twistMotion=Free
 }
]
