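# Post-process style Anim Blueprint for the WaveTrail female body cosmetic, adding secondary motion to the pose copied from its parent mesh.
# Cached stages chain Control Rig, bone offsets, AnimDynamics on small accessories, RigidBody simulation and post-sim belt fixes.
# RigidBody branches switch between emote and regular physics assets before shared tag physics and further AnimDynamics.
# The event graph updates gravity, emote, glider, pistol and frame-rate-scaled curve variables.
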AnimBlueprint /BRCosmetics/Characters/Player/Female/Medium/Bodies/F_MED_WaveTrail/Meshes/F_MED_WaveTrail_AnimBP
target_skeleton=&/F_MED_WaveTrail_Skeleton
graphs [
 AnimGraph {nodes [
   Output Pose(Result=Use cached pose 'postSimFixes'()) {layerGroup=None}
   n1 = Control Rig!(
    Source=Copy Pose From Mesh() {bUseAttachedParent=true}
    Alpha
    pub isPlayingEasterBloc:bool=Get isPlayingEasterBloc() -> (isPlayingEasterBloc:bool)
    pub isStormAviator:bool=Get isPlayingStormAviator() -> (isPlayingStormAviator:bool)
    pub isPlayingEmote:bool=Get bIsPlayingEmote()
    pub isPlayingMeeleSwing:bool=Get bIsPlayingMeleeAnim()
   ) -> (Pose) {
    alphaBoolBlend {blendOption=Linear}
    controlRigAssetReference {blueprintRigClass=&/F_MED_WaveTrail_CtrlRig.F_MED_WaveTrail_CtrlRig_C}
   }
   Save cached pose 'controlRigCP'(Pose=@n1) {cacheName=controlRigCP}
   n2 = RigidBody!(ComponentPose=Local To Component(LocalPose=Use cached pose 'animDynamicsCP'()), Alpha, OverrideWorldGravity:Vector="0,0,0", ExternalForce) -> (Pose) {
    bOverrideWorldGravity=true
    overridePhysicsAsset=&/F_MED_WaveTrail_Emotes_Physics
    simulationSpace=BaseBoneSpace
    alphaBoolBlend {blendOption=Linear}
    baseBoneRef {boneName=pelvis}
    componentAppliedLinearAccClamp=<3000 3000 3000>
    componentLinearAccScale=<0.3 0.4 0.3>
    componentLinearVelScale=<0.3 0.4 0.3>
    simSpaceSettings {maxAngularVelocity=3, maxLinearAcceleration=700, maxLinearVelocity=500, velocityScaleZ=0, worldAlpha=0.65, externalLinearDragV=<0.2 0.2 0.2>}
   }
   n3 = RigidBody(ComponentPose=@n2, OverrideWorldGravity="0,0,0") {
    bOverrideWorldGravity=true
    overridePhysicsAsset=&/F_MED_WaveTrail_Emotes_Belt_Physics
    simulationSpace=BaseBoneSpace
    alphaBoolBlend {blendOption=Linear}
    baseBoneRef {boneName=thigh_l}
    componentAppliedLinearAccClamp=<3000 3000 3000>
    componentLinearAccScale=<0.3 0.4 0.3>
    componentLinearVelScale=<0.3 0.4 0.3>
    simSpaceSettings {maxAngularVelocity=3, maxLinearAcceleration=700, maxLinearVelocity=500, velocityScaleZ=0, worldAlpha=0.65, externalLinearDragV=<0.2 0.2 0.2>}
   }
   n4 = Component To Local(ComponentPose=@n3)
   n5 = RigidBody(ComponentPose=Local To Component(LocalPose=Use cached pose 'animDynamicsCP'()), OverrideWorldGravity="0,0,0") {
    bOverrideWorldGravity=true
    overridePhysicsAsset=&/F_MED_WaveTrail_Physics
    simulationSpace=BaseBoneSpace
    alphaBoolBlend {blendOption=Linear}
    baseBoneRef {boneName=pelvis}
    componentAppliedLinearAccClamp=<3000 3000 3000>
    componentLinearAccScale=<0.3 0.4 0.3>
    componentLinearVelScale=<0.3 0.4 0.3>
    simSpaceSettings {maxAngularVelocity=3, maxLinearAcceleration=700, maxLinearVelocity=500, velocityScaleZ=0, worldAlpha=0.65, externalLinearDragV=<0.2 0.2 0.2>}
   }
   n6 = Component To Local(ComponentPose=@n5)
   n7 = Get bIsPlayingEmote()
   n8 = RigidBody(
    ComponentPose=Local To Component(LocalPose=Blend Poses by bool(BlendPose_0=@n4, BlendPose_1=@n6, bActiveValue=@n7) {blendType=Linear})
    OverrideWorldGravity="0,0,0"
   ) {
    bOverrideWorldGravity=true
    overridePhysicsAsset=&/F_MED_WaveTrail_Physics_tags
    simulationSpace=BaseBoneSpace
    alphaBoolBlend {blendOption=Linear}
    baseBoneRef {boneName=pelvis}
    componentAppliedLinearAccClamp=<3000 3000 3000>
    componentLinearAccScale=<0.3 0.4 0.3>
    componentLinearVelScale=<0.3 0.4 0.3>
    simSpaceSettings {maxAngularVelocity=3, maxLinearAcceleration=700, maxLinearVelocity=500, velocityScaleZ=0, worldAlpha=0.65, externalLinearDragV=<0.2 0.2 0.2>}
   }
   n9 = Anim Dynamics - Bone: dyn_nade_1!(ComponentPose=@n8, Alpha=1, GravityOverride:Vector="0,0,-980", ExternalForce="0,0,0") -> (Pose) {
    angularDampingOverride=1
    angularSpringConstant=12000
    bAngularSpring=true
    bOverrideAngularDamping=true
    bUseGravityOverride=true
    gravityScale=0.35
    simulationSpace=RootRelative
    alphaBoolBlend {blendOption=Linear}
    boundBone {boneName=dyn_nade_1}
    componentAppliedLinearAccClamp=<3000 3000 3000>
    componentLinearAccScale=<0.2 0.2 0.2>
    componentLinearVelScale=<0.2 0.2 0.2>
    gravityOverride {z=-980}
    physicsBodyDefinitions [{
      boundBone {boneName=dyn_nade_1}
      boxExtents=<20 30 30>
      constraintSetup {angularLimitsMax=<0 0 7>, angularLimitsMin=<0 -10 -7>, angularTarget=<1 0 0>}
      localJointOffset=<2 0 0>
     }]
   }
   n10 = Anim Dynamics - Bone: dyn_nade_2!(ComponentPose=@n9, Alpha=1, GravityOverride:Vector="0,0,-980", ExternalForce="0,0,0") -> (Pose) {
    angularDampingOverride=1
    angularSpringConstant=12000
    bAngularSpring=true
    bOverrideAngularDamping=true
    bUseGravityOverride=true
    gravityScale=0.35
    simulationSpace=RootRelative
    alphaBoolBlend {blendOption=Linear}
    boundBone {boneName=dyn_nade_2}
    componentAppliedLinearAccClamp=<3000 3000 3000>
    componentLinearAccScale=<0.2 0.2 0.2>
    componentLinearVelScale=<0.2 0.2 0.2>
    gravityOverride {z=-980}
    physicsBodyDefinitions [{
      boundBone {boneName=dyn_nade_2}
      boxExtents=<20 30 30>
      constraintSetup {angularLimitsMax=<0 0 7>, angularLimitsMin=<0 -10 -7>, angularTarget=<1 0 0>}
      localJointOffset=<2 0 0>
     }]
   }
   Save cached pose 'dynamicsCP'(Pose=Component To Local(ComponentPose=@n10)) {cacheName=dynamicsCP}
   n11 = "Transform (Modify) Bone - Bone: base_armourPadLeg_1"!(
    ComponentPose=Local To Component(LocalPose=Use cached pose 'controlRigCP'())
    bAlphaBoolEnabled:bool=True
    Translation="0,0,-2"
    Rotation="-70,0,-30"
   ) -> (Pose) {
    alphaInputType=Bool
    rotationMode=BMM_Additive
    rotationSpace=BCS_BoneSpace
    translationMode=BMM_Additive
    translationSpace=BCS_BoneSpace
    alphaBoolBlend {blendInTime=0.5, blendOption=Cubic, blendOutTime=0.2}
    boneToModify {boneName=base_armourPadLeg_1}
    rotation {pitch=-70, roll=-30}
    translation {z=-2}
   }
   Save cached pose 'preSimTweaks'(Pose=Component To Local(ComponentPose=@n11)) {cacheName=preSimTweaks}
   n12 = "Constraint - attach_cape (P)"(ComponentPose=Local To Component(LocalPose=Use cached pose 'preSimTweaks'()), Alpha=1, ConstraintWeights_0=1) {
    alphaBoolBlend {blendOption=Linear}
    boneToModify {boneName=attach_cape}
    constraintSetup [{targetBone {boneName=spine_05}}]
    constraintWeights [1]
   }
   n13 = "Transform (Modify) Bone - Bone: attach_cape"!(ComponentPose=@n12, bAlphaBoolEnabled:bool=True, Translation="0,-1,0", Rotation="0,0,16") -> (Pose) {
    alphaInputType=Bool
    rotationMode=BMM_Additive
    rotationSpace=BCS_BoneSpace
    translationMode=BMM_Additive
    translationSpace=BCS_BoneSpace
    alphaBoolBlend {blendInTime=0.5, blendOption=CircularOut, blendOutTime=0.5}
    boneToModify {boneName=attach_cape}
    rotation {roll=16}
    translation {y=-1}
   }
   n14 = "Transform (Modify) Bone - Bone: attach_cape"(ComponentPose=@n13, bAlphaBoolEnabled=True, Translation="0,-2,0", Rotation="0,0,20") {
    alphaInputType=Bool
    rotationMode=BMM_Additive
    rotationSpace=BCS_BoneSpace
    translationMode=BMM_Additive
    translationSpace=BCS_BoneSpace
    alphaBoolBlend {blendInTime=0.5, blendOption=CircularOut, blendOutTime=0.5}
    boneToModify {boneName=attach_cape}
    rotation {roll=20}
    translation {y=-2}
   }
   n15 = "Transform (Modify) Bone - Bone: attach_cape"(ComponentPose=@n14, bAlphaBoolEnabled=True, Translation="0,0,0", Rotation="0,0,15") {
    alphaInputType=Bool
    rotationMode=BMM_Additive
    rotationSpace=BCS_BoneSpace
    translationMode=BMM_Additive
    translationSpace=BCS_BoneSpace
    alphaBoolBlend {blendInTime=0.5, blendOption=Cubic, blendOutTime=0.5}
    boneToModify {boneName=attach_cape}
    rotation {roll=15}
   }
   n16 = "Transform (Modify) Bone - Bone: base_armourPadArm_1"!(ComponentPose=@n15, bAlphaBoolEnabled:bool=True, Translation="0,-4,0", Rotation="0,0,30") -> (Pose) {
    alphaInputType=Bool
    rotationMode=BMM_Additive
    rotationSpace=BCS_BoneSpace
    translationMode=BMM_Additive
    translationSpace=BCS_BoneSpace
    alphaBoolBlend {blendInTime=0.5, blendOption=Cubic, blendOutTime=0.5}
    boneToModify {boneName=base_armourPadArm_1}
    rotation {roll=30}
    translation {y=-4}
   }
   n17 = "Transform (Modify) Bone - Bone: base_armourPadArm_1"(ComponentPose=@n16, bAlphaBoolEnabled=True, Translation="0,-1,0", Rotation="0,0,15") {
    alphaInputType=Bool
    rotationMode=BMM_Additive
    rotationSpace=BCS_BoneSpace
    translationMode=BMM_Additive
    translationSpace=BCS_BoneSpace
    alphaBoolBlend {blendInTime=0.5, blendOption=Cubic, blendOutTime=0.5}
    boneToModify {boneName=base_armourPadArm_1}
    rotation {roll=15}
    translation {y=-1}
   }
   Save cached pose 'generalAdjuistementsCP'(Pose=Component To Local(ComponentPose=@n17)) {cacheName=generalAdjuistementsCP}
   n18 = Anim Dynamics - Bone: dyn_shoe_handle_l!(
    ComponentPose=Local To Component(LocalPose=Use cached pose 'generalAdjuistementsCP'())
    Alpha=1
    GravityOverride:Vector="0,0,-980"
    ExternalForce="0,0,0"
   ) -> (Pose) {
    angularDampingOverride=1
    angularSpringConstant=8000
    bAngularSpring=true
    bOverrideAngularDamping=true
    bUseGravityOverride=true
    gravityScale=0.35
    simulationSpace=RootRelative
    alphaBoolBlend {blendOption=Linear}
    boundBone {boneName=dyn_shoe_handle_l}
    componentAppliedLinearAccClamp=<3000 3000 3000>
    componentLinearAccScale=<1 1 1>
    componentLinearVelScale=<1 1 1>
    gravityOverride {z=-980}
    physicsBodyDefinitions [{
      boundBone {boneName=dyn_shoe_handle_l}
      boxExtents=<20 20 20>
      constraintSetup {angularLimitsMax=<0 15 40>, angularLimitsMin=<0 -15 -20>, angularTarget=<1 0 0>}
      localJointOffset=<5 0 0>
     }]
   }
   n19 = Anim Dynamics - Bone: dyn_shoe_handle_r!(ComponentPose=@n18, Alpha=1, GravityOverride:Vector="0,0,-980", ExternalForce="0,0,0") -> (Pose) {
    angularDampingOverride=1
    angularSpringConstant=8000
    bAngularSpring=true
    bOverrideAngularDamping=true
    bUseGravityOverride=true
    gravityScale=0.35
    simulationSpace=RootRelative
    alphaBoolBlend {blendOption=Linear}
    boundBone {boneName=dyn_shoe_handle_r}
    componentAppliedLinearAccClamp=<3000 3000 3000>
    componentLinearAccScale=<1 1 1>
    componentLinearVelScale=<1 1 1>
    gravityOverride {z=-980}
    physicsBodyDefinitions [{
      boundBone {boneName=dyn_shoe_handle_r}
      boxExtents=<20 20 20>
      constraintSetup {angularLimitsMax=<0 15 40>, angularLimitsMin=<0 -15 -20>, angularTarget=<1 0 0>}
      localJointOffset=<5 0 0>
     }]
   }
   n20 = Anim Dynamics - Bone: dyn_chestCard_r!(ComponentPose=@n19, Alpha=1, GravityOverride:Vector="0,0,-980", ExternalForce="0,0,0") -> (Pose) {
    angularDampingOverride=0.999
    angularSpringConstant=17000
    bAngularSpring=true
    bOverrideAngularDamping=true
    bUseGravityOverride=true
    gravityScale=0.35
    simulationSpace=RootRelative
    alphaBoolBlend {blendOption=Linear}
    boundBone {boneName=dyn_chestCard_r}
    componentAppliedLinearAccClamp=<3000 3000 3000>
    componentLinearAccScale=<1 1 1>
    componentLinearVelScale=<1 1 1>
    gravityOverride {z=-980}
    physicsBodyDefinitions [{
      boundBone {boneName=dyn_chestCard_r}
      boxExtents=<20 30 30>
      constraintSetup {angularLimitsMax=<0 0 5>, angularLimitsMin=<0 -20 -5>, angularTarget=<1 0 0>}
      localJointOffset=<4 0 0>
     }]
   }
   n21 = Anim Dynamics - Bone: dyn_chestCard_c!(ComponentPose=@n20, Alpha=1, GravityOverride:Vector="0,0,-980", ExternalForce="0,0,0") -> (Pose) {
    angularDampingOverride=0.999
    angularSpringConstant=17000
    bAngularSpring=true
    bOverrideAngularDamping=true
    bUseGravityOverride=true
    gravityScale=0.35
    simulationSpace=RootRelative
    alphaBoolBlend {blendOption=Linear}
    boundBone {boneName=dyn_chestCard_c}
    componentAppliedLinearAccClamp=<3000 3000 3000>
    componentLinearAccScale=<1 1 1>
    componentLinearVelScale=<1 1 1>
    gravityOverride {z=-980}
    physicsBodyDefinitions [{
      boundBone {boneName=dyn_chestCard_c}
      boxExtents=<20 30 30>
      constraintSetup {angularLimitsMax=<0 0 5>, angularLimitsMin=<0 -20 -5>, angularTarget=<1 0 0>}
      localJointOffset=<4 0 0>
     }]
   }
   n22 = Anim Dynamics - Bone: dyn_chestCard_l!(ComponentPose=@n21, Alpha=1, GravityOverride:Vector="0,0,-980", ExternalForce="0,0,0") -> (Pose) {
    angularDampingOverride=0.999
    angularSpringConstant=17000
    bAngularSpring=true
    bOverrideAngularDamping=true
    bUseGravityOverride=true
    gravityScale=0.35
    simulationSpace=RootRelative
    alphaBoolBlend {blendOption=Linear}
    boundBone {boneName=dyn_chestCard_l}
    componentAppliedLinearAccClamp=<3000 3000 3000>
    componentLinearAccScale=<1 1 1>
    componentLinearVelScale=<1 1 1>
    gravityOverride {z=-980}
    physicsBodyDefinitions [{
      boundBone {boneName=dyn_chestCard_l}
      boxExtents=<20 30 30>
      constraintSetup {angularLimitsMax=<0 0 5>, angularLimitsMin=<0 -20 -5>, angularTarget=<1 0 0>}
      localJointOffset=<4 0 0>
     }]
   }
   n23 = Anim Dynamics - Bone: dyn_armourPadArm_1!(ComponentPose=@n22, Alpha=0.75, GravityOverride:Vector="0,0,-980", ExternalForce="0,0,0") -> (Pose) {
    angularDampingOverride=1
    angularSpringConstant=5000
    bAngularSpring=true
    bOverrideAngularDamping=true
    bUseGravityOverride=true
    gravityScale=0.35
    simulationSpace=RootRelative
    alphaBoolBlend {blendOption=Linear}
    boundBone {boneName=dyn_armourPadArm_1}
    componentAppliedLinearAccClamp=<3000 3000 3000>
    componentLinearAccScale=<0.1 0.1 0.1>
    componentLinearVelScale=<0.1 0.1 0.1>
    gravityOverride {z=-980}
    physicsBodyDefinitions [{
      boundBone {boneName=dyn_armourPadArm_1}
      boxExtents=<50 50 50>
      constraintSetup {angularLimitsMax=<0 4 0>, angularLimitsMin=<0 -4 0>, angularTarget=<1 0 0>}
      localJointOffset=<10 0 0>
     }]
   }
   n24 = Anim Dynamics - Bone: dyn_armourPadArm_2!(ComponentPose=@n23, Alpha=0.75, GravityOverride:Vector="0,0,-980", ExternalForce="0,0,0") -> (Pose) {
    angularDampingOverride=1
    angularSpringConstant=10000
    bAngularSpring=true
    bOverrideAngularDamping=true
    bUseGravityOverride=true
    gravityScale=0.35
    simulationSpace=RootRelative
    alphaBoolBlend {blendOption=Linear}
    boundBone {boneName=dyn_armourPadArm_2}
    componentAppliedLinearAccClamp=<3000 3000 3000>
    componentLinearAccScale=<0.25 0.25 0.25>
    componentLinearVelScale=<0.25 0.25 0.25>
    gravityOverride {z=-980}
    physicsBodyDefinitions [{
      boundBone {boneName=dyn_armourPadArm_2}
      boxExtents=<40 40 40>
      constraintSetup {angularLimitsMax=<0 5 0>, angularLimitsMin=<0 -5 0>, angularTarget=<1 0 0>}
      localJointOffset=<5 0 0>
     }]
   }
   Save cached pose 'animDynamicsCP'(Pose=Component To Local(ComponentPose=@n24)) {cacheName=animDynamicsCP}
   n25 = "Transform (Modify) Bone - Bone: dyn_hangingBelt_2"!(
    ComponentPose=Local To Component(LocalPose=Use cached pose 'dynamicsCP'())
    bAlphaBoolEnabled:bool=True
    Translation="0,0,0"
    Rotation="15,0,0"
   ) -> (Pose) {
    alphaInputType=Bool
    rotationMode=BMM_Additive
    rotationSpace=BCS_BoneSpace
    translationMode=BMM_Additive
    translationSpace=BCS_BoneSpace
    alphaBoolBlend {blendInTime=0.5, blendOption=Cubic, blendOutTime=0.2}
    boneToModify {boneName=dyn_hangingBelt_2}
    rotation {pitch=15}
   }
   n26 = "Transform (Modify) Bone - Bone: dyn_hangingBelt_3"!(ComponentPose=@n25, bAlphaBoolEnabled:bool=True, Translation="0,0,0", Rotation="10,0,0") -> (Pose) {
    alphaInputType=Bool
    rotationMode=BMM_Additive
    rotationSpace=BCS_BoneSpace
    translationMode=BMM_Additive
    translationSpace=BCS_BoneSpace
    alphaBoolBlend {blendInTime=0.5, blendOption=Cubic, blendOutTime=0.2}
    boneToModify {boneName=dyn_hangingBelt_3}
    rotation {pitch=10}
   }
   n27 = "Transform (Modify) Bone - Bone: dyn_beltHangCard"!(ComponentPose=@n26, bAlphaBoolEnabled:bool=True, Translation="0,0,0", Rotation="5,0,0") -> (Pose) {
    alphaInputType=Bool
    rotationMode=BMM_Additive
    rotationSpace=BCS_BoneSpace
    translationMode=BMM_Additive
    translationSpace=BCS_BoneSpace
    alphaBoolBlend {blendInTime=0.5, blendOption=Cubic, blendOutTime=0.2}
    boneToModify {boneName=dyn_beltHangCard}
    rotation {pitch=5}
   }
   n28 = "Transform (Modify) Bone - Bone: dyn_armourPadLeg_1"!(ComponentPose=@n27, Alpha=1, Translation="0,0,-1.5", Rotation="-5,0,0") -> (Pose) {
    rotationMode=BMM_Additive
    rotationSpace=BCS_BoneSpace
    translationMode=BMM_Additive
    translationSpace=BCS_BoneSpace
    alphaBoolBlend {blendOption=Linear}
    boneToModify {boneName=dyn_armourPadLeg_1}
    rotation {pitch=-5}
    translation {z=-1.5}
   }
   Save cached pose 'postSimFixes'(Pose=Component To Local(ComponentPose=@n28)) {cacheName=postSimFixes}
  ]}
 EventGraph {nodes [
   n1 = Event BlueprintUpdateAnimation()
   > Set remap 60to30fps(remap 60to30fps:double=Remap Float(Input=@n1.DeltaTimeX, OldRange=(X=0.0167,Y=0.0333), NewRange=(X=0,Y=1), Clamp=true) -> (Output=~)) -> (Output_Get:double)
   > n2 = Sequence!(execute) -> (then_0, then_1, then_2:Exec)
   > n3 = Gravity Override State Function DEPRECATED(CharacterPartAnimInstance=Self-Reference(), GameTimeInSeconds=GetGameTimeInSeconds(), GravityOverrideParams=Get Gravity Override Params())
   > Set GravityOverride(GravityOverride=@n3)
   n4 = GetAnimInstance(self=GetOwnerSkeletalMeshComponent())
   Is Valid(exec=@n2.then_1, InputObject=@n4)
   > Set isPlayingEasterBloc(isPlayingEasterBloc:bool=Montage_IsPlaying(self=@n4, Montage=/BRCosmetics/Animation/Game/MainPlayer/Emotes/Eastern_Bloc/Emote_EasternBloc_CMF_M)) -> (Output_Get:bool)
   > Set isPlayingStormAviator(isPlayingStormAviator:bool=Montage_IsPlaying(self=@n4, Montage=/BRCosmetics/Animation/Game/MainPlayer/Emotes/StormAviator/CMF/Emote_StormAviator_CMF_M)) -> (Output_Get:bool)
   Set isCrouchASDPistol(
    execute=@n2.then_2
    isCrouchASDPistol:bool=AND Boolean!(A=Get bIsTargeting(), B=Get isUsingPistols() -> (isUsingPistols:bool), C:bool=Get bIsCrouching()) -> (ReturnValue)
   ) -> (Output_Get:bool)
   > Set r_thigh_rz_up_trigger(r_thigh_rz_up_trigger:double=float * float(A=GetCurveValue(CurveName=r_thigh_rz_up), B=Get remap 60to30fps() -> (remap 60to30fps:double)) {is=title K2Node_PromotableOperator}) -> (Output_Get:double)
   Event OnParachuteOpened()
   > Switch on EGliderType(Selection=Get PlayerGliderType() -> (PlayerGliderType=HangGlider))
   > Set isUsingHangGliding(isUsingHangGliding:bool=true) -> (Output_Get:bool)
   > Set weightNades(weightNades:double=0.5) -> (Output_Get:double)
   Event OnParachuteClosed()
   > Set isUsingHangGliding(isUsingHangGliding=false)
   > Set weightNades(weightNades=1)
   Event OnWeaponChanged()
   > Set isUsingPistols(isUsingPistols:bool=false) -> (Output_Get:bool)
   > n5 = Switch on EFortWeaponCoreAnimation!(execute, Selection:EFortWeaponCoreAnimation Enum=Get WeaponCoreAnim!() -> (WeaponCoreAnim)) -> (
    Melee:Exec
    Pistol:Exec
    Shotgun:Exec
    PaperBlueprint:Exec
    Rifle:Exec
    MeleeOneHand:Exec
    MachinePistol:Exec
    RocketLauncher:Exec
    GrenadeLauncher:Exec
    GoingCommando:Exec
    AssaultRifle:Exec
    TacticalShotgun:Exec
    SniperRifle:Exec
    TrapPlacement:Exec
    ShoulderLauncher:Exec
    AbilityDecoTool:Exec
    Crossbow:Exec
    C4:Exec
    RemoteControl:Exec
    DualWield:Exec
    AR_BullPup:Exec
    AR_ForwardGrip:Exec
    MedPackPaddles:Exec
    SMG_P90:Exec
    AR_DrumGun:Exec
    Consumable_Small:Exec
    Consumable_Large:Exec
    Balloon:Exec
    MountedTurret:Exec
    CreativeTool:Exec
    ExplosiveBow:Exec
    AshtonIndigo:Exec
    AshtonChicago:Exec
    MeleeDualWield:Exec
    Unarmed:Exec
   )
   Set isUsingPistols(execute=@n5.Pistol, isUsingPistols=true)
  ]}
]
variables [
 Gravity Override Params {type=GravityOverrideParamsStruct}
 GravityOverride {type=Vector}
 isCrouchASDPistol {type=boolean}
 isPlayingEasterBloc {type=boolean}
 isPlayingStormAviator {type=boolean}
 isUsingHangGliding {type=boolean}
 isUsingPistols {type=boolean}
 r_thigh_rz_up_trigger {type=number}
 remap 60to30fps {type=number}
 weightNades {type=number, default=1}
]
