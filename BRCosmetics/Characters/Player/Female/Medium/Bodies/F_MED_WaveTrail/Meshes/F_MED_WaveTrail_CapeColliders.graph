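# Physics asset of capsule colliders for the F_MED_WaveTrail player body, named as the collision set for its cape.
# Capsules cover the pelvis, spine, clavicles, head and feet, with several spine bodies pairing two differently oriented capsules.
# Dynamic grenade and leg armour pad bones also carry capsules, while arm, thigh, calf and neck bodies have no shapes.
# Constraints follow the skeleton, attaching the accessory chains to spine_01, with all swing and twist motion free.
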PhysicsAsset /BRCosmetics/Characters/Player/Female/Medium/Bodies/F_MED_WaveTrail/Meshes/F_MED_WaveTrail_CapeColliders
bodies {
 calf_l []
 calf_r []
 clavicle_l [{shapeType=Capsule, length=6.5623, radius=7.5194, center=<3.9566 -1.0862 -2.1127>, rotation=<rot -77.5566 177.3915 7.7314>}]
 clavicle_r [{shapeType=Capsule, length=6.5623, radius=7.5194, center=<-3.9566 1.0862 2.1127>, rotation=<rot -77.5566 177.3915 -172.2686>}]
 dyn_armourPadLeg_1 [{shapeType=Capsule, length=9.6034, radius=4.2674, shapeName=dyn_armourPadLeg_1_capsule, center=<0.7738 0 -2.543>, rotation=<rot 0 0 90>}]
 dyn_armourPadLeg_2 [{
   shapeType=Capsule
   length=10.5448
   radius=4.8077
   shapeName=dyn_armourPadLeg_2_capsule
   center=<3.5432 -0.521 -1.3027>
   rotation=<rot 0 0 90>
  }]
 dyn_armourPadLeg_3 [{shapeType=Capsule, length=8.1356, radius=5.6224, shapeName=dyn_armourPadLeg_3_capsule, center=<5.2974 0 -1.4239>, rotation=<rot 0 0 90>}]
 dyn_nade_1 [{
   shapeType=Capsule
   length=6.352
   radius=7.819
   shapeName=dyn_nade_1_capsule
   center=<4.454 -0.187 3.7291>
   rotation=<rot 0.5983 -89.9253 -89.8465>
  }]
 dyn_nade_2 [{
   shapeType=Capsule
   length=6.3527
   radius=7.1035
   shapeName=dyn_nade_2_capsule
   center=<4.4543 -0.1434 3.7185>
   rotation=<rot 0 -89.8386 -89.8767>
  }]
 foot_l [{shapeType=Capsule, length=11.0765, radius=8.0344, center=<-4.1811 6.0027 -0.338>, rotation=<rot 0.0388 1.4701 -89.3982>}]
 foot_r [{shapeType=Capsule, length=11.0765, radius=8.0344, center=<4.1811 -6.0027 0.338>, rotation=<rot 0.0388 1.4701 90.6018>}]
 head [{shapeType=Capsule, length=0, radius=18, center=<9.0697 0.0746 0>, rotation=<rot 0 -90 0>}]
 lowerarm_l []
 lowerarm_r []
 neck_02 []
 pelvis [{shapeType=Capsule, length=9.6042, radius=11.1441, center=<-2.4471 -2.5289 -0.1981>}]
 spine_01 [{shapeType=Capsule, length=9.3712, radius=14.3638, center=<-8.9021 -0.3237 0>}]
 spine_02 [
  {shapeType=Capsule, length=2.2119, radius=10.0417, center=<-2.6885 -1.2641 0>}
  {shapeType=Capsule, length=20.6875, radius=9.3125, center=<-1.8317 1.9362 0>, rotation=<rot 90 0 0>}
 ]
 spine_03 [
  {shapeType=Capsule, length=6.1863, radius=9.0029, center=<0.0574 0.4416 0>}
  {shapeType=Capsule, length=20.7188, radius=9.8921, center=<-1.2163 0.549 0>, rotation=<rot 90 180 -175.0732>}
 ]
 spine_04 [
  {shapeType=Capsule, length=9.0076, radius=10.3916, center=<5.6234 -0.6779 0>}
  {shapeType=Capsule, length=10.0312, radius=9.2812, center=<-1.2326 1.308 0>, rotation=<rot 90 0 0>}
 ]
 spine_05 [{shapeType=Capsule, length=9.678, radius=12.3571, center=<0.6969 1.91 0>}]
 thigh_l []
 thigh_r []
 upperarm_l []
 upperarm_r []
}
constraints [
 {
  bone1Name=spine_01
  bone2Name=pelvis
  swing1LimitDegrees=45
  swing1Motion=Free
  swing2LimitDegrees=45
  swing2Motion=Free
  twistLimitDegrees=45
  twistMotion=Free
 }
 {
  bone1Name=spine_02
  bone2Name=spine_01
  swing1LimitDegrees=45
  swing1Motion=Free
  swing2LimitDegrees=45
  swing2Motion=Free
  twistLimitDegrees=45
  twistMotion=Free
 }
 {
  bone1Name=spine_03
  bone2Name=spine_02
  swing1LimitDegrees=45
  swing1Motion=Free
  swing2LimitDegrees=45
  swing2Motion=Free
  twistLimitDegrees=45
  twistMotion=Free
 }
 {
  bone1Name=spine_04
  bone2Name=spine_03
  swing1LimitDegrees=45
  swing1Motion=Free
  swing2LimitDegrees=45
  swing2Motion=Free
  twistLimitDegrees=45
  twistMotion=Free
 }
 {
  bone1Name=spine_05
  bone2Name=spine_04
  swing1LimitDegrees=45
  swing1Motion=Free
  swing2LimitDegrees=45
  swing2Motion=Free
  twistLimitDegrees=45
  twistMotion=Free
 }
 {
  bone1Name=clavicle_l
  bone2Name=spine_05
  swing1LimitDegrees=45
  swing1Motion=Free
  swing2LimitDegrees=45
  swing2Motion=Free
  twistLimitDegrees=45
  twistMotion=Free
 }
 {
  bone1Name=upperarm_l
  bone2Name=clavicle_l
  swing1LimitDegrees=45
  swing1Motion=Free
  swing2LimitDegrees=45
  swing2Motion=Free
  twistLimitDegrees=45
  twistMotion=Free
 }
 {
  bone1Name=lowerarm_l
  bone2Name=upperarm_l
  swing1LimitDegrees=45
  swing1Motion=Free
  swing2LimitDegrees=45
  swing2Motion=Free
  twistLimitDegrees=45
  twistMotion=Free
 }
 {
  bone1Name=neck_02
  bone2Name=spine_05
  swing1LimitDegrees=45
  swing1Motion=Free
  swing2LimitDegrees=45
  swing2Motion=Free
  twistLimitDegrees=45
  twistMotion=Free
 }
 {
  bone1Name=head
  bone2Name=neck_02
  swing1LimitDegrees=45
  swing1Motion=Free
  swing2LimitDegrees=45
  swing2Motion=Free
  twistLimitDegrees=45
  twistMotion=Free
 }
 {
  bone1Name=thigh_l
  bone2Name=pelvis
  swing1LimitDegrees=45
  swing1Motion=Free
  swing2LimitDegrees=45
  swing2Motion=Free
  twistLimitDegrees=45
  twistMotion=Free
 }
 {
  bone1Name=foot_l
  bone2Name=calf_l
  swing1LimitDegrees=45
  swing1Motion=Free
  swing2LimitDegrees=45
  swing2Motion=Free
  twistLimitDegrees=45
  twistMotion=Free
 }
 {
  bone1Name=lowerarm_r
  bone2Name=upperarm_r
  swing1LimitDegrees=45
  swing1Motion=Free
  swing2LimitDegrees=45
  swing2Motion=Free
  twistLimitDegrees=45
  twistMotion=Free
 }
 {
  bone1Name=upperarm_r
  bone2Name=clavicle_r
  swing1LimitDegrees=45
  swing1Motion=Free
  swing2LimitDegrees=45
  swing2Motion=Free
  twistLimitDegrees=45
  twistMotion=Free
 }
 {
  bone1Name=clavicle_r
  bone2Name=spine_05
  swing1LimitDegrees=45
  swing1Motion=Free
  swing2LimitDegrees=45
  swing2Motion=Free
  twistLimitDegrees=45
  twistMotion=Free
 }
 {
  bone1Name=calf_l
  bone2Name=thigh_l
  swing1LimitDegrees=45
  swing1Motion=Free
  swing2LimitDegrees=45
  swing2Motion=Free
  twistLimitDegrees=45
  twistMotion=Free
 }
 {
  bone1Name=thigh_r
  bone2Name=pelvis
  swing1LimitDegrees=45
  swing1Motion=Free
  swing2LimitDegrees=45
  swing2Motion=Free
  twistLimitDegrees=45
  twistMotion=Free
 }
 {
  bone1Name=foot_r
  bone2Name=calf_r
  swing1LimitDegrees=45
  swing1Motion=Free
  swing2LimitDegrees=45
  swing2Motion=Free
  twistLimitDegrees=45
  twistMotion=Free
 }
 {
  bone1Name=calf_r
  bone2Name=thigh_r
  swing1LimitDegrees=45
  swing1Motion=Free
  swing2LimitDegrees=45
  swing2Motion=Free
  twistLimitDegrees=45
  twistMotion=Free
 }
 {
  bone1Name=dyn_nade_2
  bone2Name=spine_01
  swing1LimitDegrees=45
  swing1Motion=Free
  swing2LimitDegrees=45
  swing2Motion=Free
  twistLimitDegrees=45
  twistMotion=Free
 }
 {
  bone1Name=dyn_nade_1
  bone2Name=spine_01
  swing1LimitDegrees=45
  swing1Motion=Free
  swing2LimitDegrees=45
  swing2Motion=Free
  twistLimitDegrees=45
  twistMotion=Free
 }
 {
  bone1Name=dyn_armourPadLeg_1
  bone2Name=spine_01
  swing1LimitDegrees=45
  swing1Motion=Free
  swing2LimitDegrees=45
  swing2Motion=Free
  twistLimitDegrees=45
  twistMotion=Free
 }
 {
  bone1Name=dyn_armourPadLeg_2
  bone2Name=dyn_armourPadLeg_1
  swing1LimitDegrees=45
  swing1Motion=Free
  swing2LimitDegrees=45
  swing2Motion=Free
  twistLimitDegrees=45
  twistMotion=Free
 }
 {
  bone1Name=dyn_armourPadLeg_3
  bone2Name=dyn_armourPadLeg_2
  swing1LimitDegrees=45
  swing1Motion=Free
  swing2LimitDegrees=45
  swing2Motion=Free
  twistLimitDegrees=45
  twistMotion=Free
 }
]
