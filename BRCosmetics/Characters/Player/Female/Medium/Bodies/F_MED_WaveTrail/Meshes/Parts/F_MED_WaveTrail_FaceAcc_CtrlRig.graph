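# Control Rig for the F_MED_WaveTrail face accessory that rotates the ponytail bones correctively when the character looks up while crouched.
# DeltaRotation remaps head-to-neck rotation into head-up and extreme head-up weights.
# These combine with crouch weights, some derived from pelvis height through shaping curves, to weight layered bone rotations and a CrouchLookUp curve.
# The main graph calls only Hair_DEF; PonyTailHeadDownAdjust sits unused in the library.
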ControlRigBlueprint /BRCosmetics/Characters/Player/Female/Medium/Bodies/F_MED_WaveTrail/Meshes/Parts/F_MED_WaveTrail_FaceAcc_CtrlRig
graphs [
 Multiply_1_ContainedGraph {nodes [
   n1 = RigVMFunctionEntryNode()
   Multiply = &RigVMFunction_MathDoubleMul(A=@n1.A, B=@n1.B)
   RigVMFunctionReturnNode(Result=&RigVMFunction_MathDoubleMul(B=@n1.C, A=@Multiply.Result).Result)
  ]}
 Multiply_3_ContainedGraph {nodes [
   n1 = RigVMFunctionEntryNode()
   RigVMFunctionReturnNode(Result=&RigVMFunction_MathDoubleMul(B=@n1.C, A=&RigVMFunction_MathDoubleMul(A=@n1.A, B=@n1.B).Result).Result)
  ]}
 New Function_ContainedGraph {nodes [
   n1 = RigVMFunctionEntryNode()
   RigVMFunction_MathQuaternionToEuler = &RigVMFunction_MathQuaternionToEuler(
    RotationOrder=ZYX
    Value=&RigUnit_GetRelativeTransformForItem(
     Child=(Type=Bone,Name="hand_l")
     Parent=(Type=Bone,Name="lowerarm_l")
     bChildInitial=False
     bParentInitial=False
     Child.Name=@n1.Child
     Parent.Name=@n1.Parent
    ).RelativeTransform.Rotation
   )
   Remap = &RigVMFunction_MathDoubleRemap(
    TargetMaximum=1
    TargetMinimum=0
    Value=@RigVMFunction_MathQuaternionToEuler.Result.Z
    SourceMaximum=@n1.Source Maximum
    SourceMinimum=@n1.Source Minimum
   )
   > RigVMFunctionReturnNode(Delta=@Remap.Result)
  ]}
 New Function_ContainedGraph {nodes [
   RigVMFunctionEntryNode()
   > RigVMFunction_Sequence = RigVMAggregateNode()
   Maximum = &RigVMFunction_MathDoubleMax(B=RigVMVariableNode() {variable=IsCrouching}.Value, A=RigVMVariableNode() {variable=IsCrouchWalking}.Value)
   Multiply = &RigVMFunction_MathDoubleMul(B=@Maximum.Result, A=RigVMVariableNode() {variable=IsHeadUpValue}.Value)
   n1 = &RigVMFunction_AnimEvalRichCurve(
    Curve=(EditorCurveData=(Keys=((InterpMode=RCIM_Cubic),(InterpMode=RCIM_Cubic,Time=0.5973,Value=0.3461,ArriveTangent=1,LeaveTangent=1),(InterpMode=RCIM_Cubic,TangentMode=RCTM_User,Time=1,Value=1,ArriveTangent=1.9385,LeaveTangent=1.9385))))
    SourceMinimum=0
    TargetMinimum=0
    Value=@Multiply.Result
   )
   &RigUnit_ModifyTransforms(
    ItemToModify=((Item=(Type=Bone,Name="base_ponytail_main"),Transform=(Rotation=(X=0,Y=-0.1736,Z=0,W=0.9848),Translation=(X=0,Y=0,Z=0),Scale3D=(X=1,Y=1,Z=1))))
    WeightMinimum=0
    Weight=@n1.Result
    ExecutePin=@RigVMFunction_Sequence.B
   )
   > &RigUnit_ModifyTransforms(
    ItemToModify=((Item=(Type=Bone,Name="dyn_ponytail_Mid_main"),Transform=(Rotation=(X=0,Y=-0.1305,Z=0,W=0.9914),Translation=(X=0,Y=0,Z=0),Scale3D=(X=1,Y=1,Z=1))))
    WeightMinimum=0
    Weight=@n1.Result
   )
   > /Script/ControlRig.RigUnit_SetCurveValue(Curve=CrouchLookUp, Value=@n1.Result) {is=struct RigVMUnitNode}
   RigVMFunctionReturnNode(ExecuteContext=@RigVMFunction_Sequence.D)
   n2 = &RigVMFunction_MathDoubleMul(A=RigVMVariableNode() {variable=IsCrouchWalking}.Value, B=RigVMVariableNode() {variable=IsHeadUpExtreme}.Value)
   &RigUnit_ModifyTransforms(
    ItemToModify=((Item=(Type=Bone,Name="base_ponytail_main"),Transform=(Rotation=(X=0,Y=-0.1736,Z=0,W=0.9848),Translation=(X=0,Y=0,Z=0),Scale3D=(X=1,Y=1,Z=1))))
    WeightMinimum=0
    Weight=@n2.Result
    ExecutePin=@RigVMFunction_Sequence.C
   )
   > &RigUnit_ModifyTransforms(
    ItemToModify=((Item=(Type=Bone,Name="dyn_ponytail_Mid_main"),Transform=(Rotation=(X=0,Y=-0.0872,Z=0,W=0.9962),Translation=(X=0,Y=0,Z=0),Scale3D=(X=1,Y=1,Z=1))))
    WeightMinimum=0
    Weight=@n2.Result
   )
   DeltaRotation = RigVMFunctionReferenceNode(Child=head, Parent=neck_01, Source Maximum=-55, Source Minimum=-20, ExecuteContext=@RigVMFunction_Sequence.A)
   > RigVMVariableNode(Value=@DeltaRotation.Delta) {variable=IsHeadUpValue}
   RigUnit_GetRelativeTransformForItem = &RigUnit_GetRelativeTransformForItem(Child=(Type=Bone,Name="pelvis"), Parent=(Type=Bone,Name="root"), bChildInitial=False, bParentInitial=False)
   Remap = &RigVMFunction_MathDoubleRemap(
    SourceMaximum=90
    SourceMinimum=70
    TargetMaximum=0
    TargetMinimum=1
    Value=@RigUnit_GetRelativeTransformForItem.RelativeTransform.Translation.Z
   )
   RigVMFunction_AnimEvalRichCurve = &RigVMFunction_AnimEvalRichCurve(
    Curve="(EditorCurveData=(Keys=((InterpMode=RCIM_Cubic,TangentMode=RCTM_User,ArriveTangent=-0.0184,LeaveTangent=-0.0184),(InterpMode=RCIM_Cubic,TangentMode=RCTM_User,Time=0.0618,Value=0.276,ArriveTangent=-0.2478,LeaveTangent=-0.2478),(InterpMode=RCIM_Cubic,Tangent ...(truncated)"
    SourceMinimum=0
    TargetMinimum=0
    Value=@Remap.Result
   )
   > RigVMVariableNode(Value=@RigVMFunction_AnimEvalRichCurve.Result) {variable=IsCrouching}
   n3 = &RigVMFunction_AnimEvalRichCurve(
    Curve=(EditorCurveData=(Keys=((),(Time=0.5,Value=1),(Time=0.8))))
    SourceMaximum=2
    SourceMinimum=0
    TargetMinimum=0
    Value=&RigVMFunction_MathDoubleRemap(
     SourceMaximum=75
     SourceMinimum=45
     TargetMaximum=1
     TargetMinimum=0
     bClamp=false
     Value=&RigUnit_GetRelativeTransformForItem(Child=(Type=Bone,Name="pelvis"), Parent=(Type=Bone,Name="root"), bChildInitial=False, bParentInitial=False).RelativeTransform.Translation.Z
    ).Result
   )
   > RigVMVariableNode(Value=@n3.Result) {variable=IsCrouchWalking}
   > DeltaRotation_1 = RigVMFunctionReferenceNode(Child=head, Parent=neck_01, Source Maximum=-65, Source Minimum=-50)
   > RigVMVariableNode(Value=@DeltaRotation_1.Delta) {variable=IsHeadUpExtreme}
  ]}
 New Function_ContainedGraph {nodes [
   RigVMFunctionEntryNode()
   > RigVMFunction_Sequence = RigVMAggregateNode()
   RigVMFunctionReturnNode(ExecuteContext=@RigVMFunction_Sequence.C)
   DeltaRotation = RigVMFunctionReferenceNode(Child=head, Parent=neck_01, Source Maximum=-43, Source Minimum=-20, ExecuteContext=@RigVMFunction_Sequence.A)
   > RigVMVariableNode(Value=@DeltaRotation.Delta) {variable=isHeadUpDelta}
   > DeltaRotation_1 = RigVMFunctionReferenceNode(Child=head, Parent=neck_01, Source Maximum=-89, Source Minimum=-50)
   > RigVMVariableNode(Value=@DeltaRotation_1.Delta) {variable=isHeadUpExtremeDelta}
   n1 = &RigVMFunction_AnimEvalRichCurve(
    Curve=(EditorCurveData=(Keys=((InterpMode=RCIM_Cubic),(InterpMode=RCIM_Cubic,Time=0.5,Value=1),(Time=0.8))))
    SourceMaximum=2
    SourceMinimum=0
    TargetMinimum=0
    Value=&RigVMFunction_MathDoubleRemap(
     SourceMaximum=75
     SourceMinimum=45
     TargetMaximum=1
     TargetMinimum=0
     bClamp=false
     Value=&RigUnit_GetRelativeTransformForItem(Child=(Type=Bone,Name="pelvis"), Parent=(Type=Bone,Name="root"), bChildInitial=False, bParentInitial=False).RelativeTransform.Translation.Z
    ).Result
   )
   > RigVMVariableNode(Value=@n1.Result) {variable=PelvisGroundRemap}
   n2 = &RigVMFunction_MathDoubleMul(A=RigVMVariableNode() {variable=isHeadUpDelta}.Value, B=RigVMVariableNode() {variable=isCrouching}.Value)
   &RigUnit_ModifyTransforms(
    ItemToModify=((Item=(Type=Bone,Name="base_ponytail_main"),Transform=(Rotation=(X=0,Y=-0.2164,Z=0,W=0.9763),Translation=(X=0,Y=0,Z=0),Scale3D=(X=1,Y=1,Z=1))))
    WeightMinimum=0
    Weight=@n2.Result
    ExecutePin=@RigVMFunction_Sequence.B
   )
   > &RigUnit_ModifyTransforms(
    ItemToModify=((Item=(Type=Bone,Name="dyn_ponytail_Mid_main"),Transform=(Rotation=(X=0,Y=-0.1736,Z=0,W=0.9848),Translation=(X=0,Y=0,Z=0),Scale3D=(X=1,Y=1,Z=1))))
    WeightMinimum=0
    Weight=@n2.Result
   )
   Multiply = RigVMAggregateNode(
    A=RigVMVariableNode() {variable=isHeadUpExtremeDelta}.Value
    B=RigVMVariableNode() {variable=PelvisGroundRemap}.Value
    C=RigVMVariableNode() {variable=isCrouching}.Value
   )
   > &RigUnit_ModifyTransforms(
    ItemToModify=((Item=(Type=Bone,Name="base_ponytail_main"),Transform=(Rotation=(X=0,Y=-0.1736,Z=0,W=0.9848),Translation=(X=0,Y=0,Z=0),Scale3D=(X=1,Y=1,Z=1))))
    WeightMinimum=0
    Weight=@Multiply.Result
   )
   > &RigUnit_ModifyTransforms(
    ItemToModify=((Item=(Type=Bone,Name="dyn_ponytail_Mid_main"),Transform=(Rotation=(X=0,Y=-0.1736,Z=0,W=0.9848),Translation=(X=0,Y=0,Z=0),Scale3D=(X=1,Y=1,Z=1))))
    WeightMinimum=0
    Weight=@Multiply.Result
   )
   Multiply_1 = RigVMAggregateNode(
    C=RigVMVariableNode() {variable=isCrouching}.Value
    A=RigVMVariableNode() {variable=isHeadUpExtremeDelta}.Value
    B=RigVMVariableNode() {variable=isGunOut}.Value
   )
   > &RigUnit_ModifyTransforms(
    ItemToModify=((Item=(Type=Bone,Name="base_ponytail_main"),Transform=(Rotation=(X=0,Y=-0.0872,Z=0,W=0.9962),Translation=(X=0,Y=0,Z=0),Scale3D=(X=1,Y=1,Z=1))))
    WeightMinimum=0
    Weight=@Multiply_1.Result
   )
   > &RigUnit_ModifyTransforms(
    ItemToModify=((Item=(Type=Bone,Name="dyn_ponytail_Mid_main"),Transform=(Rotation=(X=0,Y=-0.1305,Z=0,W=0.9914),Translation=(X=0,Y=0,Z=0),Scale3D=(X=1,Y=1,Z=1))))
    WeightMinimum=0
    Weight=@Multiply_1.Result
   )
  ]}
 RigVMFunctionLibrary {nodes [
   DeltaRotation = RigVMCollapseNode(ExecuteContext=())
   Hair_DEF = RigVMCollapseNode(ExecuteContext=())
   PonyTailHeadDownAdjust = RigVMCollapseNode(ExecuteContext=())
  ]}
 RigVMFunction_Sequence_1_ContainedGraph {nodes [
   n1 = RigVMFunctionEntryNode()
   RigVMFunction_Sequence = &RigVMFunction_Sequence(ExecuteContext=@n1.ExecuteContext)
   n2 = &RigVMFunction_Sequence(ExecuteContext=@RigVMFunction_Sequence.B)
   n3 = &RigVMFunction_Sequence(ExecuteContext=@n2.B)
   RigVMFunctionReturnNode(A=@RigVMFunction_Sequence.A, B=@n2.A, C=@n3.A, D=@n3.B)
  ]}
 RigVMFunction_Sequence_1_ContainedGraph {nodes [
   n1 = RigVMFunctionEntryNode()
   RigVMFunction_Sequence = &RigVMFunction_Sequence(ExecuteContext=@n1.ExecuteContext)
   n2 = &RigVMFunction_Sequence(ExecuteContext=@RigVMFunction_Sequence.B)
   RigVMFunctionReturnNode(A=@RigVMFunction_Sequence.A, B=@n2.A, C=@n2.B)
  ]}
 RigVMModel {nodes [
   RigUnit_BeginExecution = &RigUnit_BeginExecution()
   Hair_DEF = RigVMFunctionReferenceNode(ExecuteContext=@RigUnit_BeginExecution.ExecutePin)
  ]}
]
hierarchy [
 root {loc=<0 0 0>, rot=<0 0 0 1>}
 pelvis {parent=root, loc=<0 -0.5301 98.0571>, rot=<0.0127 0.707 -0.0127 -0.707>}
 spine_01 {parent=pelvis, loc=<6.1208 -0.9328 0>, rot=<0 0 0.034 -0.9994>}
 spine_02 {parent=spine_01, loc=<6.5997 -0.2095 0>, rot=<0 0 0.0238 -0.9997>}
 spine_03 {parent=spine_02, loc=<6.4914 -0.1237 0>, rot=<0 0 0.0149 -0.9999>}
 spine_04 {parent=spine_03, loc=<6.9598 0.0794 0>, rot=<0 0 -0.0132 -0.9999>}
 spine_05 {parent=spine_04, loc=<6.9987 0 0>, rot=<0 0 -0.0015 -1>}
 neck_01 {parent=spine_05, loc=<11.3138 0.1529 0>, rot=<0 0 -0.2003 -0.9797>}
 neck_02 {parent=neck_01, loc=<4.5627 0 0>, rot=<0 0 0.0504 -0.9987>}
 head {parent=neck_02, loc=<5.3914 0 0>, rot=<0 0 0.115 -0.9934>}
 base_HairStrand_l {parent=head, loc=<6.0075 4.3269 -6.2964>, rot=<0.0743 -0.9956 0.0038 -0.0575>}
 dyn_HairStrand_l {parent=base_HairStrand_l, loc=<0 0 0>, rot=<0 0 0 1>}
 dyn_HairStrand_B_01_l {parent=dyn_HairStrand_l, loc=<4.6174 1.5204 -0.3781>, rot=<0.1467 -0.0336 0.0128 -0.9885>}
 dyn_HairStrand_B_02_l {parent=dyn_HairStrand_B_01_l, loc=<4.6604 -0.003 0.262>, rot=<-0.0098 0.0541 0.0751 -0.9957>}
 dyn_HairStrand_B_Front_03_l {parent=dyn_HairStrand_B_02_l, loc=<2.9065 -0.0186 0.2712>, rot=<-0.0263 0.1369 0.0101 -0.9902>}
 dyn_HairStrand_C_01_l {parent=dyn_HairStrand_l, loc=<3.7413 -1.0432 0.4125>, rot=<0.0039 -0.0287 -0.0056 -0.9996>}
 dyn_HairStrand_C_02_l {parent=dyn_HairStrand_C_01_l, loc=<5.4888 0 0>, rot=<0.0476 0.1892 0.2225 -0.9552>}
 base_HairStrand_r {parent=head, loc=<6.2564 3.9528 6.3215>, rot=<0.0325 -0.0005 0.9992 -0.023>}
 dyn_HairStrand_r {parent=base_HairStrand_r, loc=<0 0 0>, rot=<0 0 0 1>}
 dyn_HairStrand_B_01_r {parent=dyn_HairStrand_r, loc=<3.9642 -0.1425 0.2305>, rot=<-0.1737 -0.0172 0.0336 -0.9841>}
 dyn_HairStrand_B_02_r {parent=dyn_HairStrand_B_01_r, loc=<4.6487 0.2041 -0.3931>, rot=<0.0871 0.0704 -0.0398 -0.9929>}
 dyn_HairStrand_B_Front_03_r {parent=dyn_HairStrand_B_02_r, loc=<5.0206 -1.4303 0.1626>, rot=<-0.0078 0.1612 -0.0106 -0.9868>}
 dyn_HairStrand_B_Back_03_r {parent=dyn_HairStrand_B_02_r, loc=<4.929 -0.0891 0.1019>, rot=<-0.002 0.0336 0.0321 -0.9989>}
 dyn_HairStrand_C_02_r {parent=dyn_HairStrand_B_02_r, loc=<2.7597 1.715 0.4241>, rot=<-0.0736 0.0486 0.0392 -0.9953>}
 base_ponytail_main {parent=head, loc=<6.6306 -8.4323 -0.1067>, rot=<0.4797 -0.5195 0.5195 -0.4797>}
 dyn_ponytail_main {parent=base_ponytail_main, loc=<0 0 0>, rot=<0 0 0 1>}
 dyn_ponytail_Mid_main {parent=dyn_ponytail_main, loc=<5.549 -0.0331 -0.0078>, rot=<0 -0.3184 0 -0.9479>}
 dyn_ponytail_strand_01 {parent=dyn_ponytail_Mid_main, loc=<6.1413 4.5642 1.3908>, rot=<0.0712 -0.0298 -0.0737 -0.9943>}
 dyn_ponytail_strand_02 {parent=dyn_ponytail_Mid_main, loc=<6.466 3.3735 0.1462>, rot=<0.0088 -0.2477 0.0041 -0.9688>}
 dyn_ponytail_strand_03 {parent=dyn_ponytail_Mid_main, loc=<6.6328 1.526 -1.6151>, rot=<0.0142 -0.1907 0.0291 -0.9811>}
 dyn_ponytail_strand_04 {parent=dyn_ponytail_Mid_main, loc=<7.0038 -1.2074 -2.2005>, rot=<-0.0731 -0.2832 0.0174 -0.9561>}
 dyn_ponytail_strand_05 {parent=dyn_ponytail_Mid_main, loc=<4.9694 -3.6765 -2.8473>, rot=<-0.1214 -0.2771 0.079 -0.9499>}
 dyn_ponytail_strand_06 {parent=dyn_ponytail_Mid_main, loc=<7.2224 -3.2492 -0.0792>, rot=<-0.0785 -0.0979 -0.0515 -0.9908>}
 dyn_ponytail_strand_07 {parent=dyn_ponytail_Mid_main, loc=<4.805 -5.3861 3.2882>, rot=<-0.0329 0.0427 0.0847 -0.9949>}
 dyn_ponytail_strand_08 {parent=dyn_ponytail_Mid_main, loc=<5.5691 -2.4803 3.8583>, rot=<-0.0498 -0.0362 -0.0702 -0.9956>}
 dyn_ponytail_strand_09 {parent=dyn_ponytail_Mid_main, loc=<7.0682 1.8897 1.8213>, rot=<0.0444 -0.2007 -0.0481 -0.9775>}
 dyn_ponytail_strand_10 {parent=dyn_ponytail_Mid_main, loc=<3.08 4.4715 3.8363>, rot=<0.0429 0.0602 -0.1145 -0.9907>}
 dyn_ponytail_strand_11 {parent=dyn_ponytail_Mid_main, loc=<4.4372 1.8766 4.6476>, rot=<0.0318 0.0729 -0.0742 -0.9941>}
 dyn_ponytail_strand_12 {parent=dyn_ponytail_Mid_main, loc=<2.7334 -0.825 5.722>, rot=<-0.0434 0.1378 -0.0361 -0.9888>}
 dyn_ponytail_strand_13 {parent=dyn_ponytail_Mid_main, loc=<6.6638 -0.6755 0.4833>, rot=<-0.031 -0.1417 -0.0417 -0.9885>}
 clavicle_l {parent=spine_05, loc=<5.4831 -0.1053 -3.2946>, rot=<0.0465 -0.7289 -0.009 -0.6829>}
 clavicle_r {parent=spine_05, loc=<5.4831 -0.1053 3.2946>, rot=<0.7289 0.0465 -0.6829 0.009>}
 upperarm_r {parent=clavicle_r, loc=<-9.8658 1.2136 0.4105>, rot=<0.0164 -0.3245 -0.0269 -0.9454>}
 base_armourPadArm_1 {parent=upperarm_r, loc=<-8.9481 0.9993 -4.0919>, rot=<-0.0125 -0.0075 0.9999 0.0038>}
 dyn_armourPadArm_1 {parent=base_armourPadArm_1, loc=<0 0 0>, rot=<0 0 0 1>}
 base_armourPadArm_bend {parent=dyn_armourPadArm_1, loc=<-1.1973 -0.0163 -1.8998>, rot=<0.0024 -1 0 -0.0048>}
 dyn_armourPadArm_bend {parent=base_armourPadArm_bend, loc=<0 0 0>, rot=<0 0 0 1>}
 CrouchLookUp {type=Curve, loc=<0 0 0>, rot=<0 0 0 1>}
]
variables [
 CrouchHeadUpValue {type=double, default=0}
 IsCrouchWalking {type=double, default=0}
 isCrouching {type=double, default=0}
 isGunOut {type=double, default=0}
]
