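# Physics asset for the F_MED_WaveTrail female character's face accessory part, setting up secondary motion for hair strands and an arm armour pad.
# Left and right hair capsules form branching chains from the head, with narrow swing limits and twist at or near zero.
# Two dynamic pad bodies, a capsule and a box, attach to the pad base with free swing and twist.
# Head and right upper-arm capsules provide body collision shapes.
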PhysicsAsset /BRCosmetics/Characters/Player/Female/Medium/Bodies/F_MED_WaveTrail/Meshes/Parts/F_MED_WaveTrail_FaceAcc_Physics
bodies {
 base_armourPadArm_1 [
  {
   shapeType=Capsule
   length=10.255
   radius=3.255
   shapeName=base_armourPadArm_1_capsule
   center=<6.0862 0.0001 -0.9934>
   rotation=<rot -90 180 90.0005>
  }
  {
   shapeType=Capsule
   length=13.25
   radius=2.75
   shapeName=base_armourPadArm_1_capsule
   center=<3.4479 0 -2.1658>
   rotation=<rot 0 0.0002 -89.9998>
  }
 ]
 dyn_HairStrand_B_01_l [{
   shapeType=Capsule
   length=1.5656
   radius=1.25
   shapeName=dyn_HairStrand_B_01_l_capsule
   center=<1.5905 0.1072 0.7843>
   rotation=<rot -1.23 89.5022 -89.8937>
  }]
 dyn_HairStrand_B_01_r [
  {
   shapeType=Capsule
   length=4.2149
   radius=0.8346
   shapeName=dyn_HairStrand_B_01_r_capsule
   center=<2.0019 0.0118 -0.0054>
   rotation=<rot -2.4201 -0.7984 -82.9675>
  }
  {
   shapeType=Capsule
   length=4.5015
   radius=0.7904
   shapeName=dyn_HairStrand_B_01_r_capsule
   center=<3.0561 0.1872 -0.076>
   rotation=<rot 0 -0.0001 -81.9715>
  }
 ]
 dyn_HairStrand_B_02_l [{
   shapeType=Capsule
   length=0.6351
   radius=1.0924
   shapeName=dyn_HairStrand_B_02_l_capsule
   center=<0.5751 0.019 0.145>
   rotation=<rot -1.4443 -4.2456 -92.4682>
  }]
 dyn_HairStrand_B_02_r [{
   shapeType=Capsule
   length=4.6378
   radius=1
   shapeName=dyn_HairStrand_B_02_r_capsule
   center=<1.0824 0.4024 0.4528>
   rotation=<rot -0.1713 -179.9011 -88.3689>
  }]
 dyn_HairStrand_B_Back_03_r [{
   shapeType=Capsule
   length=2.2868
   radius=0.5613
   shapeName=dyn_HairStrand_B_Back_03_r_capsule
   center=<2.4129 0.1 -0.08>
   rotation=<rot -0.2464 93.332 -87.1826>
  }]
 dyn_HairStrand_B_Front_03_l [{
   shapeType=Capsule
   length=2.1291
   radius=1.0303
   shapeName=dyn_HairStrand_B_Front_03_l_capsule
   center=<1.2603 0 0>
   rotation=<rot -0.5297 87.9761 -86.9942>
  }]
 dyn_HairStrand_B_Front_03_r [{
   shapeType=Capsule
   length=2.3402
   radius=0.7214
   shapeName=dyn_HairStrand_B_Front_03_r_capsule
   center=<1.119 -0.05 -0.06>
   rotation=<rot 0.0801 84.9735 -89.0466>
  }]
 dyn_HairStrand_C_01_l [{
   shapeType=Capsule
   length=1.5799
   radius=1.2242
   shapeName=dyn_HairStrand_C_01_l_capsule
   center=<2.4478 0.005 0.556>
   rotation=<rot 0.4875 -89.2657 -89.4185>
  }]
 dyn_HairStrand_C_02_l [{
   shapeType=Capsule
   length=3.1814
   radius=1.1533
   shapeName=dyn_HairStrand_C_02_l_capsule
   center=<2.2209 0 0>
   rotation=<rot 0 -87.8216 -86.1545>
  }]
 dyn_HairStrand_C_02_r [{
   shapeType=Capsule
   length=3.2209
   radius=1.25
   shapeName=dyn_HairStrand_C_02_r_capsule
   center=<2.4179 0.115 0.856>
   rotation=<rot -0.7332 -85.5244 -108.9672>
  }]
 dyn_HairStrand_l [{
   shapeType=Capsule
   length=3.8983
   radius=1.3756
   shapeName=dyn_HairStrand_l_capsule
   center=<2.2424 0.35 0.65>
   rotation=<rot -0.0005 -5.1493 -84.0881>
  }]
 dyn_HairStrand_r [{
   shapeType=Capsule
   length=3.9331
   radius=1.1181
   shapeName=dyn_HairStrand_r_capsule
   center=<2.0339 -0.4228 0.9553>
   rotation=<rot 4.5946 179.0896 -80.9518>
  }]
 dyn_armourPadArm_2 [{
   shapeType=Capsule
   length=10.255
   radius=3.005
   shapeName=dyn_armourPadArm_2_capsule
   center=<5.9227 -0.1441 -0.2084>
   rotation=<rot 89.4954 90.0092 -179.9964>
  }]
 dyn_armourPadArm_bend [{shapeType=Box, extentX=15.51, extentY=17.01, extentZ=1.51, shapeName=dyn_armourPadArm_bend_box, center=<3.5813 0 0>, rotation=<rot 0 -90 0>}]
 head [
  {
   shapeType=Capsule
   length=7.1647
   radius=6.0932
   shapeName=head_capsule
   center=<0.0607 3.1243 0.2042>
   rotation=<rot 0.5498 89.9283 -89.6826>
  }
  {shapeType=Capsule, length=2.25, radius=2.5, shapeName=head_capsule, center=<2.4162 -1.8485 -6.5751>, rotation=<rot -89.0094 179.9743 -180>}
  {shapeType=Capsule, length=3.5, radius=2, shapeName=head_capsule, center=<2.4891 -1.3462 7.1442>, rotation=<rot 89.8215 -0.072 0>}
 ]
 upperarm_twist_01_r [{
   shapeType=Capsule
   length=19.5502
   radius=4.167
   shapeName=upperarm_twist_01_r_capsule
   center=<-1.4496 0 1.0586>
   rotation=<rot -0.0002 -89.9998 -90>
  }]
}
constraints [
 {bone1Name=dyn_HairStrand_l, bone2Name=head, swing1LimitDegrees=10, swing2LimitDegrees=5, twistLimitDegrees=0}
 {bone1Name=dyn_HairStrand_B_01_l, bone2Name=dyn_HairStrand_l, swing1LimitDegrees=15, swing2LimitDegrees=3, twistLimitDegrees=0}
 {bone1Name=dyn_HairStrand_B_02_l, bone2Name=dyn_HairStrand_B_01_l, swing1LimitDegrees=15, swing2LimitDegrees=2, twistLimitDegrees=0}
 {bone1Name=dyn_HairStrand_B_Front_03_l, bone2Name=dyn_HairStrand_B_02_l, swing1LimitDegrees=15, swing2LimitDegrees=2, twistLimitDegrees=5}
 {bone1Name=dyn_HairStrand_r, bone2Name=head, swing1LimitDegrees=10, swing2LimitDegrees=5, twistLimitDegrees=0}
 {bone1Name=dyn_HairStrand_B_01_r, bone2Name=dyn_HairStrand_r, swing1LimitDegrees=15, swing2LimitDegrees=3, twistLimitDegrees=0}
 {bone1Name=dyn_HairStrand_B_02_r, bone2Name=dyn_HairStrand_B_01_r, swing1LimitDegrees=12, swing2LimitDegrees=2, twistLimitDegrees=0}
 {bone1Name=dyn_HairStrand_B_Front_03_r, bone2Name=dyn_HairStrand_B_02_r, swing1LimitDegrees=10, swing2LimitDegrees=2, twistLimitDegrees=0}
 {bone1Name=dyn_HairStrand_C_01_l, bone2Name=dyn_HairStrand_l, swing1LimitDegrees=15, swing2LimitDegrees=3, twistLimitDegrees=0}
 {bone1Name=dyn_HairStrand_C_02_l, bone2Name=dyn_HairStrand_C_01_l, swing1LimitDegrees=15, swing2LimitDegrees=2, twistLimitDegrees=0}
 {bone1Name=dyn_HairStrand_B_Back_03_r, bone2Name=dyn_HairStrand_B_02_r, swing1LimitDegrees=15, swing2LimitDegrees=2, twistLimitDegrees=0}
 {bone1Name=dyn_HairStrand_C_02_r, bone2Name=dyn_HairStrand_B_02_r, swing1LimitDegrees=15, swing2LimitDegrees=2, twistLimitDegrees=0}
 {
  bone1Name=dyn_armourPadArm_2
  bone2Name=base_armourPadArm_1
  swing1LimitDegrees=45
  swing1Motion=Free
  swing2LimitDegrees=45
  swing2Motion=Free
  twistLimitDegrees=45
  twistMotion=Free
 }
 {
  bone1Name=dyn_armourPadArm_bend
  bone2Name=base_armourPadArm_1
  swing1LimitDegrees=45
  swing1Motion=Free
  swing2LimitDegrees=45
  swing2Motion=Free
  twistLimitDegrees=45
  twistMotion=Free
 }
]
