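# Anim Blueprint for the F_MED_WaveTrail face accessory mesh, animating its ponytail and arm armor pad bones.
# Parent pose, layered with Body_Mesh shoulder and pad bones, feeds a Control Rig driven by crouch and gun-out state.
# Five Modify Bone nodes then replace pad transforms, before chained AnimDynamics ponytail springs and a head-relative RigidBody solve.
# The event graph caches body pad socket transforms and updates gravity override through a deprecated function.
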AnimBlueprint /BRCosmetics/Characters/Player/Female/Medium/Bodies/F_MED_WaveTrail/Meshes/Parts/F_MED_WaveTrail_FaceAcc_AnimBP
target_skeleton=&/F_MED_WaveTrail_FaceAcc_Skeleton
graphs [
 AnimGraph {nodes [
   n1 = Use cached pose 'CRCache_Dynamics'()
   Output Pose(Result=Component To Local(ComponentPose=Local To Component(LocalPose=@n1))) {layerGroup=None}
   n2 = Layered blend per bone(BasePose=Copy Pose From Mesh() {bUseAttachedParent=true}, BlendPoses_0=Copy Pose From Mesh(SourceMeshComponent=Get Body_Mesh())) {layerSetup [{branchFilters [
       {boneName=clavicle_l}
       {boneName=clavicle_r}
       {boneName=base_armourPadArm_1}
       {boneName=dyn_armourPadArm_1}
       {boneName=base_armourPadArm_bend}
       {boneName=dyn_armourPadArm_2}
       {boneName=dyn_armourPadArm_bend}
       {boneName=dyn_armourPadArm_bend_end}
       {boneName=dyn_armourPadArm_2_end}
       {boneName=upperarm_r}
      ]}]}
   n3 = Control Rig!(Source=@n2, Alpha, isCrouching:double=Get fCrouching() -> (fCrouching:double), isGunOut:double=Get isGunOut() -> (isGunOut:double)) -> (Pose) {
    alphaBoolBlend {blendOption=Linear}
    controlRigAssetReference {blueprintRigClass=&/F_MED_WaveTrail_FaceAcc_CtrlRig.F_MED_WaveTrail_FaceAcc_CtrlRig_C}
   }
   Save cached pose 'CRCache_Input'(Pose=@n3) {cacheName=CRCache_Input}
   n4 = "Transform (Modify) Bone - Bone: base_armourPadArm_1"!(
    ComponentPose=Local To Component(LocalPose=Use cached pose 'CRCache_Input'())
    bAlphaBoolEnabled:bool=True
    Translation="0,0,0"
    Rotation="0,0,0"
   ) -> (Pose) {
    alphaInputType=Bool
    rotationMode=BMM_Replace
    rotationSpace=BCS_ParentBoneSpace
    translationMode=BMM_Replace
    translationSpace=BCS_ParentBoneSpace
    alphaBoolBlend {blendOption=Linear}
    boneToModify {boneName=base_armourPadArm_1}
   }
   n5 = "Transform (Modify) Bone - Bone: dyn_armourPadArm_1"!(ComponentPose=@n4, bAlphaBoolEnabled:bool=True, Translation="0,0,0", Rotation="0,0,0") -> (Pose) {
    alphaInputType=Bool
    rotationMode=BMM_Replace
    rotationSpace=BCS_ParentBoneSpace
    translationMode=BMM_Replace
    translationSpace=BCS_ParentBoneSpace
    alphaBoolBlend {blendOption=Linear}
    boneToModify {boneName=dyn_armourPadArm_1}
   }
   n6 = "Transform (Modify) Bone - Bone: base_armourPadArm_bend"!(ComponentPose=@n5, bAlphaBoolEnabled:bool=True, Translation="0,0,0", Rotation="0,0,0") -> (Pose) {
    alphaInputType=Bool
    rotationMode=BMM_Replace
    rotationSpace=BCS_ParentBoneSpace
    translationMode=BMM_Replace
    translationSpace=BCS_ParentBoneSpace
    alphaBoolBlend {blendOption=Linear}
    boneToModify {boneName=base_armourPadArm_bend}
   }
   n7 = "Transform (Modify) Bone - Bone: dyn_armourPadArm_bend"!(ComponentPose=@n6, bAlphaBoolEnabled:bool=True, Translation="0,0,0", Rotation="0,0,0") -> (Pose) {
    alphaInputType=Bool
    rotationMode=BMM_Replace
    rotationSpace=BCS_ParentBoneSpace
    translationMode=BMM_Replace
    translationSpace=BCS_ParentBoneSpace
    alphaBoolBlend {blendOption=Linear}
    boneToModify {boneName=dyn_armourPadArm_bend}
   }
   n8 = "Transform (Modify) Bone - Bone: dyn_armourPadArm_2"!(ComponentPose=@n7, bAlphaBoolEnabled:bool=True, Translation="0,0,0", Rotation="0,0,0") -> (Pose) {
    alphaInputType=Bool
    rotationMode=BMM_Replace
    rotationSpace=BCS_ParentBoneSpace
    translationMode=BMM_Replace
    translationSpace=BCS_ParentBoneSpace
    alphaBoolBlend {blendOption=Linear}
    boneToModify {boneName=dyn_armourPadArm_2}
   }
   Save cached pose 'CRCache_TransferTransforms'(Pose=Component To Local(ComponentPose=@n8)) {cacheName=CRCache_TransferTransforms}
   n9 = Anim Dynamics - Bone: dyn_ponytail_main!(ComponentPose=Local To Component(LocalPose=Use cached pose 'CRCache_TransferTransforms'()), Alpha=1, GravityOverride:Vector="0,0,-980") -> (Pose) {
    angularDampingOverride=1
    angularSpringConstant=10000
    bAngularSpring=true
    bOverrideAngularDamping=true
    bUseGravityOverride=true
    gravityScale=0.35
    simulationSpace=World
    alphaBoolBlend {blendOption=Linear}
    boundBone {boneName=dyn_ponytail_main}
    componentAppliedLinearAccClamp=<3000 3000 3000>
    componentLinearAccScale=<0.3 0.3 0.3>
    componentLinearVelScale=<0.3 0.3 0.3>
    gravityOverride {z=-980}
    physicsBodyDefinitions [{
      boundBone {boneName=dyn_ponytail_main}
      boxExtents=<25 35 35>
      constraintSetup {angularLimitsMax=<20 0 15>, angularLimitsMin=<-20 -30 -15>, angularTarget=<1 0 0>}
      localJointOffset=<2 0 0>
     }]
   }
   n10 = Anim Dynamics - Bone: dyn_ponytail_Mid_main!(ComponentPose=@n9, Alpha=1, GravityOverride:Vector="0,0,-980") -> (Pose) {
    angularDampingOverride=1
    angularSpringConstant=10000
    bAngularSpring=true
    bOverrideAngularDamping=true
    bUseGravityOverride=true
    gravityScale=0.35
    simulationSpace=RootRelative
    alphaBoolBlend {blendOption=Linear}
    boundBone {boneName=dyn_ponytail_Mid_main}
    componentAppliedLinearAccClamp=<3000 3000 3000>
    componentLinearAccScale=<0.3 0.3 0.3>
    componentLinearVelScale=<0.3 0.3 0.3>
    gravityOverride {z=-980}
    physicsBodyDefinitions [{
      boundBone {boneName=dyn_ponytail_Mid_main}
      boxExtents=<25 30 30>
      constraintSetup {angularLimitsMax=<20 5 15>, angularLimitsMin=<-20 -45 -15>, angularTarget=<1 0 0>}
      localJointOffset=<2 0 0>
     }]
   }
   n11 = Anim Dynamics - Bone: dyn_ponytail_strand_01!(ComponentPose=@n10, Alpha=1, GravityOverride:Vector="0,0,-980") -> (Pose) {
    angularDampingOverride=1
    angularSpringConstant=20000
    bAngularSpring=true
    bOverrideAngularDamping=true
    bUseGravityOverride=true
    gravityScale=0.35
    simulationSpace=BoneRelative
    alphaBoolBlend {blendOption=Linear}
    boundBone {boneName=dyn_ponytail_strand_01}
    componentAppliedLinearAccClamp=<3500 3500 3500>
    componentLinearAccScale=<0.3 0.6 0.3>
    componentLinearVelScale=<0.3 0.6 0.3>
    gravityOverride {z=-980}
    physicsBodyDefinitions [{
      boundBone {boneName=dyn_ponytail_strand_01}
      boxExtents=<15 25 25>
      constraintSetup {angularLimitsMax=<15 15 15>, angularLimitsMin=<-15 -60 -15>, angularTarget=<1 0 0>}
      localJointOffset=<4 0 0>
     }]
    relativeSpaceBone {boneName=spine_03}
   }
   n12 = Anim Dynamics - Bone: dyn_ponytail_strand_02!(ComponentPose=@n11, Alpha=1, GravityOverride:Vector="0,0,-980") -> (Pose) {
    angularDampingOverride=1
    angularSpringConstant=20000
    bAngularSpring=true
    bOverrideAngularDamping=true
    bUseGravityOverride=true
    gravityScale=0.35
    simulationSpace=BoneRelative
    alphaBoolBlend {blendOption=Linear}
    boundBone {boneName=dyn_ponytail_strand_02}
    componentAppliedLinearAccClamp=<3500 3500 3500>
    componentLinearAccScale=<0.3 0.6 0.3>
    componentLinearVelScale=<0.3 0.6 0.3>
    gravityOverride {z=-980}
    physicsBodyDefinitions [{
      boundBone {boneName=dyn_ponytail_strand_02}
      boxExtents=<15 25 25>
      constraintSetup {angularLimitsMax=<15 15 15>, angularLimitsMin=<-15 -60 -15>, angularTarget=<1 0 0>}
      localJointOffset=<4 0 0>
     }]
    relativeSpaceBone {boneName=spine_03}
   }
   n13 = Anim Dynamics - Bone: dyn_ponytail_strand_03!(ComponentPose=@n12, Alpha=1, GravityOverride:Vector="0,0,-980") -> (Pose) {
    angularDampingOverride=1
    angularSpringConstant=20000
    bAngularSpring=true
    bOverrideAngularDamping=true
    bUseGravityOverride=true
    gravityScale=0.35
    simulationSpace=BoneRelative
    alphaBoolBlend {blendOption=Linear}
    boundBone {boneName=dyn_ponytail_strand_03}
    componentAppliedLinearAccClamp=<3500 3500 3500>
    componentLinearAccScale=<0.3 0.6 0.3>
    componentLinearVelScale=<0.3 0.6 0.3>
    gravityOverride {z=-980}
    physicsBodyDefinitions [{
      boundBone {boneName=dyn_ponytail_strand_03}
      boxExtents=<15 25 25>
      constraintSetup {angularLimitsMax=<15 15 15>, angularLimitsMin=<-15 -60 -15>, angularTarget=<1 0 0>}
      localJointOffset=<4 0 0>
     }]
    relativeSpaceBone {boneName=spine_03}
   }
   n14 = Anim Dynamics - Bone: dyn_ponytail_strand_04!(ComponentPose=@n13, Alpha=1, GravityOverride:Vector="0,0,-980") -> (Pose) {
    angularDampingOverride=1
    angularSpringConstant=20000
    bAngularSpring=true
    bOverrideAngularDamping=true
    bUseGravityOverride=true
    gravityScale=0.35
    simulationSpace=BoneRelative
    alphaBoolBlend {blendOption=Linear}
    boundBone {boneName=dyn_ponytail_strand_04}
    componentAppliedLinearAccClamp=<3500 3500 3500>
    componentLinearAccScale=<0.3 0.6 0.3>
    componentLinearVelScale=<0.3 0.6 0.3>
    gravityOverride {z=-980}
    physicsBodyDefinitions [{
      boundBone {boneName=dyn_ponytail_strand_04}
      boxExtents=<15 25 25>
      constraintSetup {angularLimitsMax=<15 15 15>, angularLimitsMin=<-15 -60 -15>, angularTarget=<1 0 0>}
      localJointOffset=<4 0 0>
     }]
    relativeSpaceBone {boneName=spine_03}
   }
   n15 = Anim Dynamics - Bone: dyn_ponytail_strand_05!(ComponentPose=@n14, Alpha=1, GravityOverride:Vector="0,0,-980") -> (Pose) {
    angularDampingOverride=1
    angularSpringConstant=20000
    bAngularSpring=true
    bOverrideAngularDamping=true
    bUseGravityOverride=true
    gravityScale=0.35
    simulationSpace=BoneRelative
    alphaBoolBlend {blendOption=Linear}
    boundBone {boneName=dyn_ponytail_strand_05}
    componentAppliedLinearAccClamp=<3500 3500 3500>
    componentLinearAccScale=<0.3 0.6 0.3>
    componentLinearVelScale=<0.3 0.6 0.3>
    gravityOverride {z=-980}
    physicsBodyDefinitions [{
      boundBone {boneName=dyn_ponytail_strand_05}
      boxExtents=<17 27 27>
      constraintSetup {angularLimitsMax=<15 15 15>, angularLimitsMin=<-15 -60 -15>, angularTarget=<1 0 0>}
      localJointOffset=<4 0 0>
     }]
    relativeSpaceBone {boneName=spine_03}
   }
   n16 = Anim Dynamics - Bone: dyn_ponytail_strand_06!(ComponentPose=@n15, Alpha=1, GravityOverride:Vector="0,0,-980") -> (Pose) {
    angularDampingOverride=1
    angularSpringConstant=20000
    bAngularSpring=true
    bOverrideAngularDamping=true
    bUseGravityOverride=true
    gravityScale=0.35
    simulationSpace=BoneRelative
    alphaBoolBlend {blendOption=Linear}
    boundBone {boneName=dyn_ponytail_strand_06}
    componentAppliedLinearAccClamp=<3500 3500 3500>
    componentLinearAccScale=<0.3 0.6 0.3>
    componentLinearVelScale=<0.3 0.6 0.3>
    gravityOverride {z=-980}
    physicsBodyDefinitions [{
      boundBone {boneName=dyn_ponytail_strand_06}
      boxExtents=<17 27 27>
      constraintSetup {angularLimitsMax=<15 15 15>, angularLimitsMin=<-15 -60 -15>, angularTarget=<1 0 0>}
      localJointOffset=<4 0 0>
     }]
    relativeSpaceBone {boneName=spine_03}
   }
   n17 = Anim Dynamics - Bone: dyn_ponytail_strand_07!(ComponentPose=@n16, Alpha=1, GravityOverride:Vector="0,0,-980") -> (Pose) {
    angularDampingOverride=1
    angularSpringConstant=20000
    bAngularSpring=true
    bOverrideAngularDamping=true
    bUseGravityOverride=true
    gravityScale=0.35
    simulationSpace=BoneRelative
    alphaBoolBlend {blendOption=Linear}
    boundBone {boneName=dyn_ponytail_strand_07}
    componentAppliedLinearAccClamp=<3500 3500 3500>
    componentLinearAccScale=<0.3 0.6 0.3>
    componentLinearVelScale=<0.3 0.6 0.3>
    gravityOverride {z=-980}
    physicsBodyDefinitions [{
      boundBone {boneName=dyn_ponytail_strand_07}
      boxExtents=<17 27 27>
      constraintSetup {angularLimitsMax=<15 15 15>, angularLimitsMin=<-15 -60 -15>, angularTarget=<1 0 0>}
      localJointOffset=<4 0 0>
     }]
    relativeSpaceBone {boneName=spine_03}
   }
   n18 = Anim Dynamics - Bone: dyn_ponytail_strand_08!(ComponentPose=@n17, Alpha=1, GravityOverride:Vector="0,0,-980") -> (Pose) {
    angularDampingOverride=1
    angularSpringConstant=20000
    bAngularSpring=true
    bOverrideAngularDamping=true
    bUseGravityOverride=true
    gravityScale=0.35
    simulationSpace=BoneRelative
    alphaBoolBlend {blendOption=Linear}
    boundBone {boneName=dyn_ponytail_strand_08}
    componentAppliedLinearAccClamp=<3500 3500 3500>
    componentLinearAccScale=<0.3 0.6 0.3>
    componentLinearVelScale=<0.3 0.6 0.3>
    gravityOverride {z=-980}
    physicsBodyDefinitions [{
      boundBone {boneName=dyn_ponytail_strand_08}
      boxExtents=<17 27 27>
      constraintSetup {angularLimitsMax=<15 15 15>, angularLimitsMin=<-15 -60 -15>, angularTarget=<1 0 0>}
      localJointOffset=<4 0 0>
     }]
    relativeSpaceBone {boneName=spine_03}
   }
   n19 = Anim Dynamics - Bone: dyn_ponytail_strand_09!(ComponentPose=@n18, Alpha=1, GravityOverride:Vector="0,0,-980") -> (Pose) {
    angularDampingOverride=1
    angularSpringConstant=20000
    bAngularSpring=true
    bOverrideAngularDamping=true
    bUseGravityOverride=true
    gravityScale=0.35
    simulationSpace=BoneRelative
    alphaBoolBlend {blendOption=Linear}
    boundBone {boneName=dyn_ponytail_strand_09}
    componentAppliedLinearAccClamp=<3500 3500 3500>
    componentLinearAccScale=<0.3 0.6 0.3>
    componentLinearVelScale=<0.3 0.6 0.3>
    gravityOverride {z=-980}
    physicsBodyDefinitions [{
      boundBone {boneName=dyn_ponytail_strand_09}
      boxExtents=<13 23 23>
      constraintSetup {angularLimitsMax=<15 15 15>, angularLimitsMin=<-15 -60 -15>, angularTarget=<1 0 0>}
      localJointOffset=<4 0 0>
     }]
    relativeSpaceBone {boneName=spine_03}
   }
   n20 = Anim Dynamics - Bone: dyn_ponytail_strand_10!(ComponentPose=@n19, Alpha=1, GravityOverride:Vector="0,0,-980") -> (Pose) {
    angularDampingOverride=1
    angularSpringConstant=20000
    bAngularSpring=true
    bOverrideAngularDamping=true
    bUseGravityOverride=true
    gravityScale=0.35
    simulationSpace=BoneRelative
    alphaBoolBlend {blendOption=Linear}
    boundBone {boneName=dyn_ponytail_strand_10}
    componentAppliedLinearAccClamp=<3500 3500 3500>
    componentLinearAccScale=<0.3 0.6 0.3>
    componentLinearVelScale=<0.3 0.6 0.3>
    gravityOverride {z=-980}
    physicsBodyDefinitions [{
      boundBone {boneName=dyn_ponytail_strand_10}
      boxExtents=<13 23 23>
      constraintSetup {angularLimitsMax=<15 15 15>, angularLimitsMin=<-15 -60 -15>, angularTarget=<1 0 0>}
      localJointOffset=<4 0 0>
     }]
    relativeSpaceBone {boneName=spine_03}
   }
   n21 = Anim Dynamics - Bone: dyn_ponytail_strand_11!(ComponentPose=@n20, Alpha=1, GravityOverride:Vector="0,0,-980") -> (Pose) {
    angularDampingOverride=1
    angularSpringConstant=20000
    bAngularSpring=true
    bOverrideAngularDamping=true
    bUseGravityOverride=true
    gravityScale=0.35
    simulationSpace=BoneRelative
    alphaBoolBlend {blendOption=Linear}
    boundBone {boneName=dyn_ponytail_strand_11}
    componentAppliedLinearAccClamp=<3500 3500 3500>
    componentLinearAccScale=<0.3 0.6 0.3>
    componentLinearVelScale=<0.3 0.6 0.3>
    gravityOverride {z=-980}
    physicsBodyDefinitions [{
      boundBone {boneName=dyn_ponytail_strand_11}
      boxExtents=<13 23 23>
      constraintSetup {angularLimitsMax=<15 15 15>, angularLimitsMin=<-15 -60 -15>, angularTarget=<1 0 0>}
      localJointOffset=<4 0 0>
     }]
    relativeSpaceBone {boneName=spine_03}
   }
   n22 = Anim Dynamics - Bone: dyn_ponytail_strand_12!(ComponentPose=@n21, Alpha=1, GravityOverride:Vector="0,0,-980") -> (Pose) {
    angularDampingOverride=1
    angularSpringConstant=20000
    bAngularSpring=true
    bOverrideAngularDamping=true
    bUseGravityOverride=true
    gravityScale=0.35
    simulationSpace=BoneRelative
    alphaBoolBlend {blendOption=Linear}
    boundBone {boneName=dyn_ponytail_strand_12}
    componentAppliedLinearAccClamp=<3500 3500 3500>
    componentLinearAccScale=<0.3 0.6 0.3>
    componentLinearVelScale=<0.3 0.6 0.3>
    gravityOverride {z=-980}
    physicsBodyDefinitions [{
      boundBone {boneName=dyn_ponytail_strand_12}
      boxExtents=<13 23 23>
      constraintSetup {angularLimitsMax=<15 15 15>, angularLimitsMin=<-15 -60 -15>, angularTarget=<1 0 0>}
      localJointOffset=<4 0 0>
     }]
    relativeSpaceBone {boneName=spine_03}
   }
   n23 = Anim Dynamics - Bone: dyn_ponytail_strand_13!(ComponentPose=@n22, Alpha=1, GravityOverride:Vector="0,0,-980") -> (Pose) {
    angularDampingOverride=1
    angularSpringConstant=20000
    bAngularSpring=true
    bOverrideAngularDamping=true
    bUseGravityOverride=true
    gravityScale=0.35
    simulationSpace=RootRelative
    alphaBoolBlend {blendOption=Linear}
    boundBone {boneName=dyn_ponytail_strand_13}
    componentAppliedLinearAccClamp=<3500 3500 3500>
    componentLinearAccScale=<0.3 0.4 0.3>
    componentLinearVelScale=<0.3 0.4 0.3>
    gravityOverride {z=-980}
    physicsBodyDefinitions [{
      boundBone {boneName=dyn_ponytail_strand_13}
      boxExtents=<13 23 23>
      constraintSetup {angularLimitsMax=<15 0 15>, angularLimitsMin=<-15 -60 -15>, angularTarget=<1 0 0>}
      localJointOffset=<4 0 0>
     }]
   }
   n24 = RigidBody(ComponentPose=@n23) {
    bOverrideWorldGravity=true
    overridePhysicsAsset=&/F_MED_WaveTrail_FaceAcc_Physics
    simulationSpace=BaseBoneSpace
    alphaBoolBlend {blendOption=Linear}
    baseBoneRef {boneName=head}
    componentAppliedLinearAccClamp=<3000 3000 3000>
    componentLinearAccScale=<0.4 0.3 0.4>
    componentLinearVelScale=<0.4 0.5 0.4>
    simSpaceSettings {
     maxAngularAcceleration=1000
     maxAngularVelocity=3
     maxLinearAcceleration=700
     maxLinearVelocity=50
     velocityScaleZ=0
     worldAlpha=0.65
     externalLinearDragV=<0.8 0.2 0.8>
    }
   }
   Save cached pose 'CRCache_Dynamics'(Pose=Component To Local(ComponentPose=@n24)) {cacheName=CRCache_Dynamics}
  ]}
 EventGraph {nodes [
   Event BlueprintUpdateAnimation()
   > n1 = Gravity Override State Function DEPRECATED(CharacterPartAnimInstance=Self-Reference(), GameTimeInSeconds=GetGameTimeInSeconds(), GravityOverrideParams=Get Gravity Override Params())
   > Set GravityOverride(GravityOverride=@n1)
   n2 = Get Body_Mesh()
   > Is Valid(InputObject=@n2)
   n3 = GetSocketBoneName!(self:Skinned Mesh Component Object Reference=@n2, InSocketName:Name=upperarm_r) -> (ReturnValue:Name=None)
   > n4 = Set isBodyAttached(isBodyAttached:bool="Not Equal Exactly (String)"(A="To String (Name)"(InName=@n3), B=None)) -> (Output_Get:bool)
   > Branch(Condition=@n4)
   n5 = BreakTransform(InTransform=GetSocketTransform(self=Get Body_Mesh(), InSocketName=base_armourPadArm_1, TransformSpace=RTS_ParentBoneSpace))
   > Set base_PadArm_1_Translation(base_PadArm_1_Translation:Vector=@n5) -> (Output_Get:Vector)
   > Set base_PadArm_1_Rotation(base_PadArm_1_Rotation:Rotator=@n5.Rotation) -> (Output_Get:Rotator)
   n6 = BreakTransform(InTransform=GetSocketTransform(self=Get Body_Mesh(), InSocketName=dyn_armourPadArm_1, TransformSpace=RTS_ParentBoneSpace))
   > Set dyn_PadArm_1_Translation(dyn_PadArm_1_Translation:Vector=@n6) -> (Output_Get:Vector)
   > Set dyn_PadArm_1_Rotation(dyn_PadArm_1_Rotation:Rotator=@n6.Rotation) -> (Output_Get:Rotator)
   n7 = BreakTransform(InTransform=GetSocketTransform(self=Get Body_Mesh(), InSocketName=base_armourPadArm_bend, TransformSpace=RTS_ParentBoneSpace))
   > Set base_PadArm_bend_Translation(base_PadArm_bend_Translation:Vector=@n7) -> (Output_Get:Vector)
   > Set base_PadArm_bend_Rotation(base_PadArm_bend_Rotation:Rotator=@n7.Rotation) -> (Output_Get:Rotator)
   n8 = BreakTransform(InTransform=GetSocketTransform(self=Get Body_Mesh(), InSocketName=dyn_armourPadArm_bend, TransformSpace=RTS_ParentBoneSpace))
   > Set dyn_PadArm_bend_Tranform(dyn_PadArm_bend_Tranform:Vector=@n8) -> (Output_Get:Vector)
   > Set dyn_PadArm_bend_Rotation(dyn_PadArm_bend_Rotation:Rotator=@n8.Rotation) -> (Output_Get:Rotator)
   n9 = BreakTransform(InTransform=GetSocketTransform(self=Get Body_Mesh(), InSocketName=dyn_armourPadArm_2, TransformSpace=RTS_ParentBoneSpace))
   > Set dyn_PadArm_2_Translation(dyn_PadArm_2_Translation:Vector=@n9) -> (Output_Get:Vector)
   > Set dyn_PadArm_2_Rotation(dyn_PadArm_2_Rotation:Rotator=@n9.Rotation) -> (Output_Get:Rotator)
   Event OnBeginCrouching()
   > Set fCrouching(fCrouching:double=1) -> (Output_Get:double)
   Event OnEndCrouching()
   > Set fCrouching(fCrouching=0)
   Event OnWeaponChanged()
   > Set isGunOut(isGunOut:double=1) -> (Output_Get:double)
   > n10 = Switch on EFortWeaponCoreAnimation!(execute, Selection:EFortWeaponCoreAnimation Enum=Get WeaponCoreAnim!() -> (WeaponCoreAnim)) -> (
    Melee:Exec
    Pistol:Exec
    Shotgun:Exec
    PaperBlueprint:Exec
    Rifle:Exec
    MeleeOneHand:Exec
    MachinePistol:Exec
    RocketLauncher:Exec
    GrenadeLauncher:Exec
    GoingCommando:Exec
    AssaultRifle:Exec
    TacticalShotgun:Exec
    SniperRifle:Exec
    TrapPlacement:Exec
    ShoulderLauncher:Exec
    AbilityDecoTool:Exec
    Crossbow:Exec
    C4:Exec
    RemoteControl:Exec
    DualWield:Exec
    AR_BullPup:Exec
    AR_ForwardGrip:Exec
    MedPackPaddles:Exec
    SMG_P90:Exec
    AR_DrumGun:Exec
    Consumable_Small:Exec
    Consumable_Large:Exec
    Balloon:Exec
    MountedTurret:Exec
    CreativeTool:Exec
    ExplosiveBow:Exec
    AshtonIndigo:Exec
    AshtonChicago:Exec
    MeleeDualWield:Exec
    Unarmed:Exec
   )
   Set isGunOut(execute=[@n10.Melee, @n10.MeleeDualWield, @n10.MeleeOneHand, @n10.Unarmed], isGunOut=0)
   n11 = Event OnCharacterCustomizationLoaded()
   > Set Body_Mesh(Body_Mesh=Get Array Item with EFort Custom Part Type(Array=@n11.CustomMeshParts))
  ]}
]
variables [
 base_PadArm_1_Rotation {type=Rotator}
 base_PadArm_1_Translation {type=Vector}
 base_PadArm_bend_Rotation {type=Rotator}
 base_PadArm_bend_Translation {type=Vector}
 Body_Mesh {type=SkeletalMeshComponent}
 dyn_PadArm_1_Rotation {type=Rotator}
 dyn_PadArm_1_Translation {type=Vector}
 dyn_PadArm_2_Rotation {type=Rotator}
 dyn_PadArm_2_Translation {type=Vector}
 dyn_PadArm_bend_Rotation {type=Rotator}
 dyn_PadArm_bend_Tranform {type=Vector}
 DynCrouchAlpha {type=number}
 fCrouching {type=number}
 Gravity Override Params {type=GravityOverrideParamsStruct, default {rail_Grinding {jointName=spine_03}, ziplining {jointName=spine_03}}}
 GravityOverride {type=Vector}
 HangGliding {type=boolean}
 isBodyAttached {type=boolean}
 isGunOut {type=number}
 SkelMesh {type=SkeletalMeshComponent}
]
